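; Audio volume actor for Sideways encounters that plays the bubble ambient sound while inside its sphere.
; The AudioShapeSphere inside-state event adds or removes a "Sideways" entry on the ambient audio subsystem.
; SetRadius sizes the sphere from a given radius plus a padding variable.
; IsOptimized reports mobile or Switch platforms but is not called within this graph.
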(event OnInsideStateChanged(AudioShapeSphere)
  (params (bIsInside :type Boolean))
  (bind _returnvalue (WorldSubsystems|GetAmbientAudioSubsystem))
  (if bIsInside
    (AmbientAudio|AddAmbientEntry _returnvalue "Sideways" "/SidewaysGameplay/Sounds/Bubble/POIAudio_Sideways_Bubble.POIAudio_Sideways_Bubble" 10 0.25)
    (else
      (AmbientAudio|RemoveAmbientEntry _returnvalue "Sideways" -1.0))))

(fn IsOptimized
  (return (or (Fortnite|IsMobilePlatform) (TouchPlatform|IsNintendoSwitch))))

(fn SetRadius
  (params (InRadius :type Float%20%28double-precision%29))
  (Control|SetRadius (Variables|Default|GetAudioShapeSphere) (+ InRadius (Variables|Default|GetRadiusPadding)) :!class /Script/AudioShapes.AudioShapeSphereComponent))

(fn ConstructionScript)
