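; Audio volume actor for the small Sideways encounter that switches bubble ambience and edge filtering as the inside state changes.
; Entering or leaving adds or removes the "Sideways" entry in the ambient audio subsystem.
; A timeline plays on BeginPlay and exit and reverses on entry, driving the edge component's low-pass frequency.
; SetRadius syncs the SFX sphere to the main sphere and starts the source bus; optimized mode skips SFX activation and inside/outside fades.
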(event EventBeginPlay
  (|CallParentFunction)
  (Utilities|FlowControl|Sequence
    (:then_0
      (if (not (CallFunction|IsOptimized))
        (Components|Activation|Activate (Variables|Default|GetAudioShapeSphere_SFX) false)))
    (:then_1
      (bind (_direction _fader) (Timeline|Timeline_0 Play)
        (:Update
          (Audio|Components|Audio|SetLowPassFilterFrequency (Control|GetEdgeAudioComponent (Variables|Default|GetAudioShapeSphere)) (Math|Float|MapRangeClamped _fader 0.0 1.0 0.0 3e+02) :!class /Script/Engine.AudioComponent))))))

(event OnInsideStateChanged(AudioShapeSphere)
  (params (bIsInside :type Boolean))
  (CallFunction|EnableAmbientAudio bIsInside)
  (if (CallFunction|IsOptimized)
    _
    (elif bIsInside
      (Timeline|Timeline_0 Reverse)
      (else
        (Timeline|Timeline_0 Play)))))

(fn EnableAmbientAudio
  (params (Enable :type Boolean))
  (bind _returnvalue (WorldSubsystems|GetAmbientAudioSubsystem))
  (if Enable
    (AmbientAudio|AddAmbientEntry _returnvalue "Sideways" "/SidewaysGameplay/Sounds/Bubble/POIAudio_Sideways_Bubble.POIAudio_Sideways_Bubble" 10 0.25)
    (else
      (AmbientAudio|RemoveAmbientEntry _returnvalue "Sideways" -1.0))))

(fn SetRadius
  (params (InRadius :type Float%20%28double-precision%29))
  (|CallParentFunction InRadius)
  (Control|SetRadius (Variables|Default|GetAudioShapeSphere_SFX) (Variables|/Script/AudioShapes.AudioShapeSphereComponent|GetRadius (Variables|Default|GetAudioShapeSphere)) :!class /Script/AudioShapes.AudioShapeSphereComponent)
  (if (not (Audio|Components|Audio|IsPlaying (Variables|Default|GetSideways_SourceBus)))
    (Audio|Components|Audio|Play (Variables|Default|GetSideways_SourceBus) 0.0)))

(fn ConstructionScript
  (|CallParentFunction))
