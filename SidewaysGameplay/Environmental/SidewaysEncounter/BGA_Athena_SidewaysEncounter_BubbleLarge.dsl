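; Large Sideways encounter bubble actor that never dies and plays visual and audio feedback when crossed or struck.
; BuildBlockingVolume overlaps use gameplay tags to separate vehicles from pawns, pulsing the EncounterEdge material at the actor's location.
; Pawn crossings also toggle BubbleMask after a delay, on the edge and the parent EnvCoreLarge's surface smoke.
; Hits and damage spawn impact FX, sound and a pulse; the ThingsHitBubble hook is empty.
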(event EventBeginPlay)

(event Custom|PawnExitBubble
  (params (EnterLocation :type Vector))
  (bind _encounteredge (Variables|Default|GetEncounterEdge))
  ;[1 1 1] Pawn Entered / Exited the bubble
  (Development|PrintString "PawnEnterBubble")
  (Rendering|Material|SetScalarParameterValueonMaterials _encounteredge "PulseDuration" 1.0)
  (Rendering|Material|SetScalarParameterValueonMaterials _encounteredge "BrightnessPulseRadius" 512.0)
  (Rendering|Material|SetVectorParameterValueonMaterials _encounteredge "PawnIntersectLocation" EnterLocation)
  (Rendering|Material|SetScalarParameterValueonMaterials _encounteredge "GameTimeFromBP" (Utilities|Time|GetGameTimeinSeconds))
  (Utilities|FlowControl|Delay 0.2)
  (Rendering|Material|SetScalarParameterValueonMaterials _encounteredge "BubbleMask" 0.0)
  (bind _asbga_athena_sideways_encounter_env_core_large (Utilities|Casting|CastToBGA_Athena_SidewaysEncounter_EnvCoreLarge (Actor|GetParentActor)))
  (Effects|Components|ParticleSystem|SetFloatParameter (Variables|/SidewaysGameplay/Environmental/SidewaysEncounter/BGA_Athena_SidewaysEncounter_EnvCoreLarge.BGA_Athena_SidewaysEncounter_EnvCoreLarge_C|GetNS_SideWays_SurfaceSmoke _asbga_athena_sideways_encounter_env_core_large) "BubbleMask" 0.0))

(event Custom|PawnEnterBubble
  (params (EnterLocation :type Vector))
  (bind _encounteredge (Variables|Default|GetEncounterEdge))
  ;[1 1 1] Pawn Entered / Exited the bubble
  (Development|PrintString "PawnEnterBubble")
  (Rendering|Material|SetScalarParameterValueonMaterials _encounteredge "PulseDuration" 1.0)
  (Rendering|Material|SetScalarParameterValueonMaterials _encounteredge "BrightnessPulseRadius" 512.0)
  (Rendering|Material|SetVectorParameterValueonMaterials _encounteredge "PawnIntersectLocation" EnterLocation)
  (Rendering|Material|SetScalarParameterValueonMaterials _encounteredge "GameTimeFromBP" (Utilities|Time|GetGameTimeinSeconds))
  (Utilities|FlowControl|Delay 0.2)
  (Rendering|Material|SetScalarParameterValueonMaterials _encounteredge "BubbleMask" 1.0)
  (bind _asbga_athena_sideways_encounter_env_core_large (Utilities|Casting|CastToBGA_Athena_SidewaysEncounter_EnvCoreLarge (Actor|GetParentActor)))
  (Effects|Components|ParticleSystem|SetFloatParameter (Variables|/SidewaysGameplay/Environmental/SidewaysEncounter/BGA_Athena_SidewaysEncounter_EnvCoreLarge.BGA_Athena_SidewaysEncounter_EnvCoreLarge_C|GetNS_SideWays_SurfaceSmoke _asbga_athena_sideways_encounter_env_core_large) "BubbleMask" 1.0))

(event Custom|VehicleEnterBubble
  (params (EnterLocation :type Vector))
  (bind _encounteredge (Variables|Default|GetEncounterEdge))
  ;[1 1 1] Vehicle Entered / Exited the bubble
  (Development|PrintString "VehicleEnterBubble")
  (Rendering|Material|SetScalarParameterValueonMaterials _encounteredge "PulseDuration" 1.0)
  (Rendering|Material|SetScalarParameterValueonMaterials _encounteredge "BrightnessPulseRadius" 7.5e+02)
  (Rendering|Material|SetVectorParameterValueonMaterials _encounteredge "PawnIntersectLocation" EnterLocation)
  (Rendering|Material|SetScalarParameterValueonMaterials _encounteredge "GameTimeFromBP" (Utilities|Time|GetGameTimeinSeconds)))

(event Custom|VehicleExitBubble
  (params (EnterLocation :type Vector))
  (bind _encounteredge (Variables|Default|GetEncounterEdge))
  ;[1 1 1] Vehicle Entered / Exited the bubble
  (Development|PrintString "VehicleEnterBubble")
  (Rendering|Material|SetScalarParameterValueonMaterials _encounteredge "PulseDuration" 1.0)
  (Rendering|Material|SetScalarParameterValueonMaterials _encounteredge "BrightnessPulseRadius" 7.5e+02)
  (Rendering|Material|SetVectorParameterValueonMaterials _encounteredge "PawnIntersectLocation" EnterLocation)
  (Rendering|Material|SetScalarParameterValueonMaterials _encounteredge "GameTimeFromBP" (Utilities|Time|GetGameTimeinSeconds)))

(event Custom|ThingsHitBubble
  (params (Hit :type Hit%20Result%20Structure) (OtherActor :type Actor%20Object%20Reference)))

(event EventOnDamagePlayEffects
  (params (Damage :type Float%20%28single-precision%29) (DamageTags :type Gameplay%20Tag%20Container%20Structure%20%28by%20ref%29) (Momentum :type Vector :default "0, 0, 0") (HitInfo :type Hit%20Result%20Structure%20%28by%20ref%29) (InstigatedBy :type Fort%20Pawn%20Object%20Reference) (DamageCauser :type Actor%20Object%20Reference) (EffectContext :type Gameplay%20Effect%20Context%20Handle%20Structure))
  (bind (_bblockinghit _binitialoverlap _time _distance _location _impactpoint _normal _impactnormal) (Collision|BreakHitResult HitInfo))
  (bind _encounteredge (Variables|Default|GetEncounterEdge))
  ;[1 1 1] Detect if something hit the bubble (hit for projectiles, damage for traces)
  (CallFunction|ThingsHitBubble HitInfo DamageCauser)
  (Effects|Components|ParticleSystem|SpawnFXAtLocation "/SidewaysGameplay/Environmental/SidewaysEncounter/Effects/NS_Sideways_Bubble_Impact.NS_Sideways_Bubble_Impact" _location (Math|Conversions|RotationFromXVector _impactnormal))
  (Rendering|Material|SetScalarParameterValueonMaterials _encounteredge "PulseDuration" 0.5)
  (Rendering|Material|SetScalarParameterValueonMaterials _encounteredge "BrightnessPulseRadius" 256.0)
  (Rendering|Material|SetVectorParameterValueonMaterials _encounteredge "PawnIntersectLocation" _location)
  (Rendering|Material|SetScalarParameterValueonMaterials _encounteredge "GameTimeFromBP" (Utilities|Time|GetGameTimeinSeconds))
  (Audio|SpawnSoundatLocation "/SidewaysGameplay/Sounds/Bubble/Grabber_Bubble_Impact_Cue.Grabber_Bubble_Impact_Cue" _location))

(event OnComponentHit(EncounterEdge)
  (params (HitComponent :type Primitive%20Component%20Object%20Reference) (OtherActor :type Actor%20Object%20Reference) (OtherComp :type Primitive%20Component%20Object%20Reference) (NormalImpulse :type Vector :default "0, 0, 0") (Hit :type Hit%20Result%20Structure%20%28by%20ref%29))
  (bind (_bblockinghit _binitialoverlap _time _distance _location _impactpoint _normal _impactnormal) (Collision|BreakHitResult Hit))
  (bind _encounteredge (Variables|Default|GetEncounterEdge))
  ;[1 1 1] Detect if something hit the bubble (hit for projectiles, damage for traces)
  (CallFunction|ThingsHitBubble Hit OtherActor)
  (Effects|Components|ParticleSystem|SpawnFXAtLocation "/SidewaysGameplay/Environmental/SidewaysEncounter/Effects/NS_Sideways_Bubble_Impact.NS_Sideways_Bubble_Impact" _location (Math|Conversions|RotationFromXVector _impactnormal))
  (Rendering|Material|SetScalarParameterValueonMaterials _encounteredge "PulseDuration" 1.0)
  (Rendering|Material|SetScalarParameterValueonMaterials _encounteredge "BrightnessPulseRadius" 512.0)
  (Rendering|Material|SetVectorParameterValueonMaterials _encounteredge "PawnIntersectLocation" _location)
  (Rendering|Material|SetScalarParameterValueonMaterials _encounteredge "GameTimeFromBP" (Utilities|Time|GetGameTimeinSeconds))
  (Audio|SpawnSoundatLocation "/SidewaysGameplay/Sounds/Bubble/Grabber_Bubble_Impact_Cue.Grabber_Bubble_Impact_Cue" _location))

(event OnComponentBeginOverlap(BuildBlockingVolume)
  (params (OverlappedComponent :type Primitive%20Component%20Object%20Reference) (OtherActor :type Actor%20Object%20Reference) (OtherComp :type Primitive%20Component%20Object%20Reference) (OtherBodyIndex :type Integer) (bFromSweep :type Boolean) (SweepResult :type Hit%20Result%20Structure%20%28by%20ref%29))
  (bind _returnvalue (Transformation|GetActorLocation OtherActor))
  ;[1 1 1] Detect if a pawn or vehicle enters/exits the bubble
  (bind _returnvalue_2 (GameplayTags|HasAnyMatchingGameplayTags(Message) OtherActor (Variables|Default|GetTC_Vehicle)))
  (if _returnvalue_2
    (CallFunction|VehicleEnterBubble _returnvalue)
    (else
      (bind _returnvalue_3 (GameplayTags|HasMatchingGameplayTag(Message) OtherActor (Variables|Default|GetT_Pawn)))
      (if _returnvalue_3
        (CallFunction|PawnEnterBubble _returnvalue)))))

(event OnComponentEndOverlap(BuildBlockingVolume)
  (params (OverlappedComponent :type Primitive%20Component%20Object%20Reference) (OtherActor :type Actor%20Object%20Reference) (OtherComp :type Primitive%20Component%20Object%20Reference) (OtherBodyIndex :type Integer))
  (bind _returnvalue (Transformation|GetActorLocation OtherActor))
  ;[1 1 1] Detect if a pawn or vehicle enters/exits the bubble
  (bind _returnvalue_2 (GameplayTags|HasAnyMatchingGameplayTags(Message) OtherActor (Variables|Default|GetTC_Vehicle)))
  (if _returnvalue_2
    (CallFunction|VehicleExitBubble _returnvalue)
    (else
      (bind _returnvalue_3 (GameplayTags|HasMatchingGameplayTag(Message) OtherActor (Variables|Default|GetT_Pawn)))
      (if _returnvalue_3
        (CallFunction|PawnExitBubble _returnvalue)))))

(fn GetSidewaysPlayspace
  (bind _aslarge_sideways_playspace_volume (Utilities|Casting|CastToLargeSidewaysPlayspaceVolume (Variables|/Script/Engine.ChildActorComponent|GetChildActor (Variables|Default|GetPlayspaceGameplayVolume)))
    (:then
      (bind _aslarge_sideways_playspace (Utilities|Casting|CastToLargeSidewaysPlayspace (Class|GameplayVolume|GetPlayspace _aslarge_sideways_playspace_volume))
        (:then
          (return _aslarge_sideways_playspace))
        (:CastFailed
          _)))
    (:CastFailed
      _))
  (return 0))

(fn ShouldDie
  (params (Damage :type Float%20%28single-precision%29) (EventInstigator :type Controller%20Object%20Reference) (DamageCauser :type Actor%20Object%20Reference) (bIsTest :type Boolean) (EffectContext :type Gameplay%20Effect%20Context%20Handle%20Structure%20%28by%20ref%29))
  (return false))

(fn ConstructionScript)
