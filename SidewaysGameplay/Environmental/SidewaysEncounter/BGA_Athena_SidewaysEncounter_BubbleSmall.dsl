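; Small Sideways encounter bubble, a child of the EnvCoreSmall actor, that forms the indestructible, visible edge of the encounter area.
; Overlaps, hits and damage pulse the edge material at the contact point; impacts add FX and sound, while pawn crossings toggle BubbleMask on material and core smoke.
; Tracked players inside receive server-sent quest stats for completed stages and encounters.
; OnRep_BubbleVisible syncs collision, visibility, core VFX and client scale; a timed authority sphere check damages player-built structures.
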(event EventBeginPlay
  (if (Networking|IsServer)
    (bind _asbga_athena_sideways_encounter_env_core_small (Utilities|Casting|CastToBGA_Athena_SidewaysEncounter_EnvCoreSmall (Actor|GetParentActor)))
    (Variables|Default|SetCore _asbga_athena_sideways_encounter_env_core_small)
    (else
      ;[1 0.3938 0.01266] Deal with Replay Weirdness - DISABLED FOR NOW
      (ReplayContext|BindEventtoOnReplayScrubStart (BlueprintContext|GetFortReplayContext) (delegate Custom|HandleReplayScrubStart)))))

(event Custom|HandleReplayScrubStart)

(event Custom|PawnExitBubble
  (params (EnterLocation :type Vector))
  (bind _encounteredge (Variables|Default|GetEncounterEdge))
  ;[1 1 1] Pawn Entered / Exited the bubble
  (Development|PrintString "PawnEnterBubble")
  (Rendering|Material|SetScalarParameterValueonMaterials _encounteredge "PulseDuration" 1.0)
  (Rendering|Material|SetScalarParameterValueonMaterials _encounteredge "BrightnessPulseRadius" 512.0)
  (Rendering|Material|SetVectorParameterValueonMaterials _encounteredge "PawnIntersectLocation" EnterLocation)
  (Rendering|Material|SetScalarParameterValueonMaterials _encounteredge "GameTimeFromBP" (Utilities|Time|GetGameTimeinSeconds))
  (Utilities|FlowControl|Delay 0.2)
  (Rendering|Material|SetScalarParameterValueonMaterials _encounteredge "BubbleMask" 0.0)
  (bind _asbga_athena_sideways_encounter_env_core_small (Utilities|Casting|CastToBGA_Athena_SidewaysEncounter_EnvCoreSmall (Actor|GetParentActor)))
  (Effects|Components|ParticleSystem|SetFloatParameter (Variables|/SidewaysGameplay/Environmental/SidewaysEncounter/BGA_Athena_SidewaysEncounter_EnvCoreSmall.BGA_Athena_SidewaysEncounter_EnvCoreSmall_C|GetNS_SideWays_SurfaceSmoke_SM _asbga_athena_sideways_encounter_env_core_small) "BubbleMask" 0.0))

(event Custom|PawnEnterBubble
  (params (EnterLocation :type Vector))
  (bind _encounteredge (Variables|Default|GetEncounterEdge))
  ;[1 1 1] Pawn Entered / Exited the bubble
  (Development|PrintString "PawnEnterBubble")
  (Rendering|Material|SetScalarParameterValueonMaterials _encounteredge "PulseDuration" 1.0)
  (Rendering|Material|SetScalarParameterValueonMaterials _encounteredge "BrightnessPulseRadius" 512.0)
  (Rendering|Material|SetVectorParameterValueonMaterials _encounteredge "PawnIntersectLocation" EnterLocation)
  (Rendering|Material|SetScalarParameterValueonMaterials _encounteredge "GameTimeFromBP" (Utilities|Time|GetGameTimeinSeconds))
  (Utilities|FlowControl|Delay 0.2)
  (Rendering|Material|SetScalarParameterValueonMaterials _encounteredge "BubbleMask" 1.0)
  (bind _asbga_athena_sideways_encounter_env_core_small (Utilities|Casting|CastToBGA_Athena_SidewaysEncounter_EnvCoreSmall (Actor|GetParentActor)))
  (Effects|Components|ParticleSystem|SetFloatParameter (Variables|/SidewaysGameplay/Environmental/SidewaysEncounter/BGA_Athena_SidewaysEncounter_EnvCoreSmall.BGA_Athena_SidewaysEncounter_EnvCoreSmall_C|GetNS_SideWays_SurfaceSmoke_SM _asbga_athena_sideways_encounter_env_core_small) "BubbleMask" 1.0))

(event Custom|VehicleEnterBubble
  (params (EnterLocation :type Vector))
  (bind _encounteredge (Variables|Default|GetEncounterEdge))
  ;[1 1 1] Vehicle Entered / Exited the bubble
  (Development|PrintString "VehicleEnterBubble")
  (Rendering|Material|SetScalarParameterValueonMaterials _encounteredge "PulseDuration" 1.0)
  (Rendering|Material|SetScalarParameterValueonMaterials _encounteredge "BrightnessPulseRadius" 7.5e+02)
  (Rendering|Material|SetVectorParameterValueonMaterials _encounteredge "PawnIntersectLocation" EnterLocation)
  (Rendering|Material|SetScalarParameterValueonMaterials _encounteredge "GameTimeFromBP" (Utilities|Time|GetGameTimeinSeconds)))

(event Custom|VehicleExitBubble
  (params (EnterLocation :type Vector))
  (bind _encounteredge (Variables|Default|GetEncounterEdge))
  ;[1 1 1] Vehicle Entered / Exited the bubble
  (Development|PrintString "VehicleEnterBubble")
  (Rendering|Material|SetScalarParameterValueonMaterials _encounteredge "PulseDuration" 1.0)
  (Rendering|Material|SetScalarParameterValueonMaterials _encounteredge "BrightnessPulseRadius" 7.5e+02)
  (Rendering|Material|SetVectorParameterValueonMaterials _encounteredge "PawnIntersectLocation" EnterLocation)
  (Rendering|Material|SetScalarParameterValueonMaterials _encounteredge "GameTimeFromBP" (Utilities|Time|GetGameTimeinSeconds)))

(event Custom|ThingsHitBubble
  (params (Hit :type Hit%20Result%20Structure) (OtherActor :type Actor%20Object%20Reference))
  ;[1 1 1] A projectile or trace weapon hit the bubble
  (Development|PrintString "ThingHitBubble"))

(event EventOnDamagePlayEffects
  (params (Damage :type Float%20%28single-precision%29) (DamageTags :type Gameplay%20Tag%20Container%20Structure%20%28by%20ref%29) (Momentum :type Vector :default "0, 0, 0") (HitInfo :type Hit%20Result%20Structure%20%28by%20ref%29) (InstigatedBy :type Fort%20Pawn%20Object%20Reference) (DamageCauser :type Actor%20Object%20Reference) (EffectContext :type Gameplay%20Effect%20Context%20Handle%20Structure))
  (bind (_bblockinghit _binitialoverlap _time _distance _location _impactpoint _normal _impactnormal) (Collision|BreakHitResult HitInfo))
  (bind _encounteredge (Variables|Default|GetEncounterEdge))
  ;[1 1 1] Detect if something hit the bubble (hit for projectiles, damage for traces)
  (CallFunction|ThingsHitBubble HitInfo DamageCauser)
  (Effects|Components|ParticleSystem|SpawnFXAtLocation "/SidewaysGameplay/Environmental/SidewaysEncounter/Effects/NS_Sideways_Bubble_Impact.NS_Sideways_Bubble_Impact" _location (Math|Conversions|RotationFromXVector _impactnormal))
  (Rendering|Material|SetScalarParameterValueonMaterials _encounteredge "PulseDuration" 0.5)
  (Rendering|Material|SetScalarParameterValueonMaterials _encounteredge "BrightnessPulseRadius" 256.0)
  (Rendering|Material|SetVectorParameterValueonMaterials _encounteredge "PawnIntersectLocation" _location)
  (Rendering|Material|SetScalarParameterValueonMaterials _encounteredge "GameTimeFromBP" (Utilities|Time|GetGameTimeinSeconds))
  (Audio|SpawnSoundatLocation "/SidewaysGameplay/Sounds/Bubble/Grabber_Bubble_Impact_Cue.Grabber_Bubble_Impact_Cue" _location))

(event OnComponentHit(EncounterEdge)
  (params (HitComponent :type Primitive%20Component%20Object%20Reference) (OtherActor :type Actor%20Object%20Reference) (OtherComp :type Primitive%20Component%20Object%20Reference) (NormalImpulse :type Vector :default "0, 0, 0") (Hit :type Hit%20Result%20Structure%20%28by%20ref%29))
  (bind (_bblockinghit _binitialoverlap _time _distance _location _impactpoint _normal _impactnormal) (Collision|BreakHitResult Hit))
  (bind _encounteredge (Variables|Default|GetEncounterEdge))
  ;[1 1 1] Detect if something hit the bubble (hit for projectiles, damage for traces)
  (CallFunction|ThingsHitBubble Hit OtherActor)
  (Effects|Components|ParticleSystem|SpawnFXAtLocation "/SidewaysGameplay/Environmental/SidewaysEncounter/Effects/NS_Sideways_Bubble_Impact.NS_Sideways_Bubble_Impact" _location (Math|Conversions|RotationFromXVector _impactnormal))
  (Rendering|Material|SetScalarParameterValueonMaterials _encounteredge "PulseDuration" 1.0)
  (Rendering|Material|SetScalarParameterValueonMaterials _encounteredge "BrightnessPulseRadius" 512.0)
  (Rendering|Material|SetVectorParameterValueonMaterials _encounteredge "PawnIntersectLocation" _location)
  (Rendering|Material|SetScalarParameterValueonMaterials _encounteredge "GameTimeFromBP" (Utilities|Time|GetGameTimeinSeconds))
  (Audio|SpawnSoundatLocation "/SidewaysGameplay/Sounds/Bubble/Grabber_Bubble_Impact_Cue.Grabber_Bubble_Impact_Cue" _location))

(event OnComponentBeginOverlap(BuildBlockingVolume)
  (params (OverlappedComponent :type Primitive%20Component%20Object%20Reference) (OtherActor :type Actor%20Object%20Reference) (OtherComp :type Primitive%20Component%20Object%20Reference) (OtherBodyIndex :type Integer) (bFromSweep :type Boolean) (SweepResult :type Hit%20Result%20Structure%20%28by%20ref%29))
  (bind _returnvalue (Transformation|GetActorLocation OtherActor))
  ;[1 1 1] Detect if a pawn or vehicle enters/exits the bubble
  (bind _returnvalue_2 (GameplayTags|HasAnyMatchingGameplayTags(Message) OtherActor (Variables|Default|GetTC_Vehicle)))
  (if _returnvalue_2
    (CallFunction|VehicleEnterBubble _returnvalue)
    (else
      (bind _returnvalue_3 (GameplayTags|HasMatchingGameplayTag(Message) OtherActor (Variables|Default|GetT_Pawn)))
      (if _returnvalue_3
        (CallFunction|PawnEnterBubble _returnvalue)
        (Utilities|IsValid OtherActor
          (:Is%20Valid
            (bind _asfort_player_pawn (Utilities|Casting|CastToFortPlayerPawn OtherActor))
            (Utilities|Array|AddUnique (Variables|Default|GetPlayersInsideBubble) _asfort_player_pawn) ; Pawns we are tracking inside the bubble for quests
          ))))))

(event OnComponentEndOverlap(BuildBlockingVolume)
  (params (OverlappedComponent :type Primitive%20Component%20Object%20Reference) (OtherActor :type Actor%20Object%20Reference) (OtherComp :type Primitive%20Component%20Object%20Reference) (OtherBodyIndex :type Integer))
  (bind _returnvalue (Transformation|GetActorLocation OtherActor))
  ;[1 1 1] Detect if a pawn or vehicle enters/exits the bubble
  (bind _returnvalue_2 (GameplayTags|HasAnyMatchingGameplayTags(Message) OtherActor (Variables|Default|GetTC_Vehicle)))
  (if _returnvalue_2
    (CallFunction|VehicleExitBubble _returnvalue)
    (else
      (bind _returnvalue_3 (GameplayTags|HasMatchingGameplayTag(Message) OtherActor (Variables|Default|GetT_Pawn)))
      (if _returnvalue_3
        (CallFunction|PawnExitBubble _returnvalue)
        (Utilities|IsValid OtherActor
          (:Is%20Valid
            (bind _asfort_player_pawn (Utilities|Casting|CastToFortPlayerPawn OtherActor))
            (Utilities|Array|RemoveItem (Variables|Default|GetPlayersInsideBubble) _asfort_player_pawn) ; Pawns we are tracking inside the bubble for quests
          ))))))

(event Custom|EncounterStageCompleted
  ;[1 1 1] Triggers for Quests
  (if (Networking|IsServer)
    (for _array_element (Variables|Default|GetPlayersInsideBubble)
      (Utilities|IsValid _array_element
        (:Is%20Valid
          (Quest|SendComplexCustomStatEvent (Quests|GetQuestManager (Fortnite|GetFortPlayerControllerfromActor _array_element) "Invalid") 0 "(GameplayTags=)" (Variables|Default|GetTC_WaveCompleted) 1))))))

(event Custom|EncounterBeaten
  ;[1 1 1] Triggers for Quests
  (if (Networking|IsServer)
    (for _array_element (Variables|Default|GetPlayersInsideBubble)
      (Utilities|IsValid _array_element
        (:Is%20Valid
          (Quest|SendComplexCustomStatEvent (Quests|GetQuestManager (Fortnite|GetFortPlayerControllerfromActor _array_element) "Invalid") 0 "(GameplayTags=)" (Variables|Default|GetTC_EncounterCompleted) 1))))))

(fn DestroyBuildings
  (params (EndScale :type Float%20%28double-precision%29) (ScaleDuration :type Float%20%28double-precision%29))
  (Networking|SwitchHasAuthority
    (:Authority
      (Variables|Default|SetBuildingDestroyScale EndScale)
      (Utilities|Time|SetTimerbyFunctionName 0 "DestroyBuildingsFromTimer" (* ScaleDuration 0.25)))))

(fn DestroyBuildingsFromTimer
  (Networking|SwitchHasAuthority
    (:Authority
      (bind _outactors (Collision|SphereOverlapActors (Transformation|GetWorldLocation (Variables|Default|GetBuildBlockingVolume)) (Variables|Default|GetBuildingDestroyScale) (Utilities|Array|MakeArray "ObjectTypeQuery1" "ObjectTypeQuery2" "ObjectTypeQuery6") "/Script/FortniteGame.BuildingSMActor"))
      (for _array_element _outactors
        (bind _asbuilding_static_mesh (Utilities|Casting|CastToBuildingSMActor _array_element))
        (if (BuildingActor|IsPlayerBuilt _asbuilding_static_mesh)
          (GameplayEffects|ApplyGameplayEffectToTarget :self (Ability|GetAbilitySystemComponent self) :GameplayEffectClass (Variables|Default|GetGE_DamageBuilding) :Target (Ability|GetAbilitySystemComponent _asbuilding_static_mesh) :Level 0.0))))))

(fn GetSidewaysPlayspace
  (bind _assmall_sideways_playspace_volume (Utilities|Casting|CastToSmallSidewaysPlayspaceVolume (Variables|/Script/Engine.ChildActorComponent|GetChildActor (Variables|Default|GetPlayspaceGameplayVolume)))
    (:then
      (bind _assmall_sideways_playspace (Utilities|Casting|CastToSmallSidewaysPlayspace (Class|GameplayVolume|GetPlayspace _assmall_sideways_playspace_volume))
        (:then
          (return _assmall_sideways_playspace))
        (:CastFailed
          _)))
    (:CastFailed
      _))
  (return 0))

(fn OnRep_BubbleVisible
  (bind _bubblevisible (Variables|Default|GetBubbleVisible))
  (bind _bubblecurrentscale (Variables|Default|GetBubbleCurrentScale))
  (bind _core (Variables|Default|GetCore))
  ;[1 1 1] Set the collision and hidden if bubble should be shown
  (Collision|SetActorEnableCollision _bubblevisible)
  (Rendering|SetActorHiddenInGame (not _bubblevisible))
  ;[1 1 1] Enable other visuals on the core such as the world VFX
  (if _bubblevisible
    (Rendering|SetVisibility (Variables|/SidewaysGameplay/Environmental/SidewaysEncounter/BGA_Athena_SidewaysEncounter_EnvCoreSmall.BGA_Athena_SidewaysEncounter_EnvCoreSmall_C|GetSM_Sideways_SmokePillars _core) true false)
    (Components|Activation|Activate (Variables|/SidewaysGameplay/Environmental/SidewaysEncounter/BGA_Athena_SidewaysEncounter_EnvCoreSmall.BGA_Athena_SidewaysEncounter_EnvCoreSmall_C|GetNS_SideWays_SurfaceSmoke_SM _core) false)
    (if (not (Viewport|HasMultipleLocalPlayers))
      (Components|Activation|Activate (Variables|/SidewaysGameplay/Environmental/SidewaysEncounter/BGA_Athena_SidewaysEncounter_EnvCoreSmall.BGA_Athena_SidewaysEncounter_EnvCoreSmall_C|GetNS_SideWays_Volume_InWorld _core) false))
    ;[1 1 1] On the client if the bubble has ever been scaled, scale the bubble to match.
    (if (Networking|IsServer)
      _
      (elif (> _bubblecurrentscale 0.0)
        (Transformation|SetWorldScale3D (Variables|Default|GetEncounterEdge) (Math|Conversions|ToVector(Float) _bubblecurrentscale))))))

(fn ShouldDie
  (params (Damage :type Float%20%28single-precision%29) (EventInstigator :type Controller%20Object%20Reference) (DamageCauser :type Actor%20Object%20Reference) (bIsTest :type Boolean) (EffectContext :type Gameplay%20Effect%20Context%20Handle%20Structure%20%28by%20ref%29))
  (return false))

(fn ConstructionScript
  ;[1 1 1] Make sure collision is on or off in construction script baded off the visible bool
  (Collision|SetActorEnableCollision (Variables|Default|GetBubbleVisible)))
