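; Sideways encounter bubble actor tied to a Caretaker pawn, which shuts down when that pawn leaves the bubble, dies or despawns.
; On BeginPlay, the authority starts the encounter component after a short delay that lets things register.
; The first owner set becomes the instigator, binding its Sideways state, death and EndPlay delegates.
; Return to the normal world, death or EndPlay calls EncounterNoSpawnTimeout through DoOnce to close the encounter.
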(event EventBeginPlay
  (bind _encountercomponent (CallFunction|GetSidewaysEncounterComponent))
  ;[1 1 1] Start Encounter bubble
  (|CallParentFunction)
  (Networking|SwitchHasAuthority
    (:Authority
      (Utilities|IsValid _encountercomponent
        (:Is%20Valid
          (Utilities|FlowControl|Delay 0.2) ; NEED DELAY FOR THINGS TO REGISTER
          (Class|SmallSidewaysPlayspaceEncounterComponent|StartEncounter _encountercomponent))))))

(event Custom|OnSetOwner
  (params (NewOwner :type Fort%20Pawn%20Object%20Reference))
  ;[1 1 1] Shut down Encounter if Caretaker leaves bubble, or dies/despawns
  (if (Utilities|Equal(Object) (Game|GetInstigator) 0)
    (bind _output_get (Variables|Actor|SetInstigator NewOwner))
    (bind _returnvalue (Sideways|GetSidewaysStateComponent _output_get))
    (bind _output_get_2 (Variables|Default|SetSidewaysComponentFromOwner _returnvalue))
    (EventDispatchers|AssignOnSidewaysStateChangedDelegate _output_get_2 (delegate Custom|OnSidewaysStateChangedDelegate_Event))
    (bind _asfort_pawn (Utilities|Casting|CastToFortPawn (Game|GetInstigator)))
    (EventDispatchers|AssignOnDied _asfort_pawn (delegate Custom|OnDied_Event))
    (Game|AssignOnEndPlay _asfort_pawn (delegate Custom|OnEndPlay_Event))))

(event Custom|OnDied_Event
  (params (DamagedActor :type Actor%20Object%20Reference) (Damage :type Float%20%28single-precision%29) (InstigatedBy :type Controller%20Object%20Reference) (DamageCauser :type Actor%20Object%20Reference) (HitLocation :type Vector) (FHitComponent :type Primitive%20Component%20Object%20Reference) (BoneName :type Name) (Momentum :type Vector))
  ;[1 1 1] Shut down Encounter if Caretaker leaves bubble, or dies/despawns
  (Utilities|FlowControl|DoOnce false
    (:Completed
      (Class|SmallSidewaysPlayspaceEncounterComponent|EncounterNoSpawnTimeout (CallFunction|GetSidewaysEncounterComponent)) ; This guy works better...
    )) ; Shut down encounter bubble
)

(event Custom|OnEndPlay_Event
  (params (Actor :type Actor%20Object%20Reference) (EndPlayReason :type EEndPlayReason%20Enum :default "Destroyed"))
  ;[1 1 1] Shut down Encounter if Caretaker leaves bubble, or dies/despawns
  (Utilities|FlowControl|DoOnce false
    (:Completed
      (Class|SmallSidewaysPlayspaceEncounterComponent|EncounterNoSpawnTimeout (CallFunction|GetSidewaysEncounterComponent)) ; This guy works better...
    )) ; Shut down encounter bubble
)

(event Custom|OnSidewaysStateChangedDelegate_Event
  (params (SidewaysStateComponent :type Sideways%20State%20Component%20Object%20Reference) (SidewaysPlayspace :type Sideways%20Playspace%20Object%20Reference) (OldState :type ESidewaysState%20Enum :default "InNormalWorld") (NewState :type ESidewaysState%20Enum :default "InNormalWorld"))
  ;[1 1 1] Shut down Encounter if Caretaker leaves bubble, or dies/despawns
  (if (Utilities|Enum|NotEqual(Enum) OldState NewState)
    (if (Utilities|Enum|Equal(Enum) NewState "InNormalWorld")
      (Utilities|FlowControl|DoOnce false
        (:Completed
          (Class|SmallSidewaysPlayspaceEncounterComponent|EncounterNoSpawnTimeout (CallFunction|GetSidewaysEncounterComponent)) ; This guy works better...
        )) ; Shut down encounter bubble
    )))

(fn ConstructionScript
  (|CallParentFunction))
