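; Core environment actor for a large Sideways encounter, owning the bubble, skirt meshes, VFX and looping audio.
; Construction and BeginPlay scale the child bubble from hotfixable radius and height values and apply per-instance skirt meshes.
; Once playlist data is ready, the server starts the playspace's encounter component and handles its kill, cue and music events.
; A server timer re-adds the minimap radius circle until the game passes the aircraft phase.
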(event Custom|StartAddingMapCircle
  ;[1 0.7959 0] Logic for adding in the map radial marker, May need to adjust this to not need the timers and have a more global solution in code.
  (if (Networking|IsServer)
    (bind _returnvalue (Utilities|Time|SetTimerbyEvent (delegate Custom|AddMapCircle) (Variables|MapCircle|GetCircleAddTime) true))
    (Variables|MapCircle|SetTimer_MapCircleAdd _returnvalue)))

(event Custom|AddMapCircle
  ;[1 0.7959 0] Logic for adding in the map radial marker, May need to adjust this to not need the timers and have a more global solution in code.
  (bind _returnvalue (RadiusTracker|MakeRadiusTrackingInitializer 1.0 (- (ScalableFloat|GetValueatLevel (Variables|BubbleSettings|GetBubbleVisualRadius_HF)) 1e+02) (ScalableFloat|GetValueatLevel (Variables|BubbleSettings|GetBubbleVisualRadius_HF)) 1e+02 0.1 (Variables|MapCircle|GetMapCircleColor) "/FortniteCoreUI/HUD/MiniMap/M_MiniMapCircle_Sideways.M_MiniMapCircle_Sideways" self (Fortnite|GetAllFortPlayerControllers true) "Global"))
  (Fortnite|AddActorTrackingInfo (RadiusTracker|GetGameStateRadiusTrackerComponent) _returnvalue)
  (if (Editor|InEditororPIE)
    (bind _result (Math|Integer|IncrementInt (Variables|MapCircle|GetPIECircleCounter))
      (:%20%20
        (if (> _result 10)
          (Utilities|Time|ClearandInvalidateTimerbyHandle (Variables|MapCircle|GetTimer_MapCircleAdd)))))
    (elif (Math|Boolean|ORBoolean (Utilities|Enum|Equal(Enum) (BattleRoyaleGamePhaseLogic|GetGamePhase (BattleRoyaleGamePhaseLogic|GetBattleRoyaleGamePhaseLogicComponent)) "Setup") (Utilities|Enum|Equal(Enum) (BattleRoyaleGamePhaseLogic|GetGamePhase (BattleRoyaleGamePhaseLogic|GetBattleRoyaleGamePhaseLogicComponent)) "Warmup") (Utilities|Enum|Equal(Enum) (BattleRoyaleGamePhaseLogic|GetGamePhase (BattleRoyaleGamePhaseLogic|GetBattleRoyaleGamePhaseLogicComponent)) "Aircraft"))
      _
      (else
        (Utilities|Time|ClearandInvalidateTimerbyHandle (Variables|MapCircle|GetTimer_MapCircleAdd))))))

(event EventBeginPlay
  (bind _encountercomponent (CallFunction|GetSidewaysEncounterComponent))
  ;[1 1 1] BeginPlay - Register to events on the encounter component
  (CallFunction|SetSkirtMeshes)
  (CallFunction|SetupBubble)
  (if (Viewport|HasMultipleLocalPlayers)
    (Components|Activation|Deactivate (Variables|Default|GetNS_SideWays_Volume_InWorld))
    (else
      _))
  (if (Networking|IsServer)
    (Class|FortAsyncActionCurrentPlaylistReady|PlaylistDataReadyAsync
      (:OnReady
        (CallFunction|StartAddingMapCircle)
        (Default|BindEventtoKillEvent _encountercomponent (delegate Custom|KillEvent))
        (Default|BindEventtoSpawnCosmeticEvent _encountercomponent (delegate Custom|SpawnCosmeticEvent))
        (Default|BindEventtoUpdateLoopingAudioEvent _encountercomponent (delegate Custom|UpdateLoopingAudioEvent))
        (Class|LargeSidewaysPlayspaceEncounterComponent|SetupSidewaysEncounterLogic _encountercomponent :!pinned true) ; Starts Encounter
      )) ; Make sure we load any hotfix overrides
  ))

(event Custom|KillEvent
  ;[1 1 1] BeginPlay - Register to events on the encounter component
  (CallFunction|HideandKill))

(event Custom|SpawnCosmeticEvent
  (params (GameplayCueTag :type Gameplay%20Tag%20Structure))
  ;[1 1 1] BeginPlay - Register to events on the encounter component
  (Ability|ExecuteGameplayCueReplicated GameplayCueTag))

(event Custom|UpdateLoopingAudioEvent
  (params (Music :type Sound%20Base%20Object%20Reference))
  ;[1 1 1] BeginPlay - Register to events on the encounter component
  (CallFunction|UpdateLoopingAudio Music))

(event Custom|HideAndKill
  (bind _encounterbubble (Variables|Default|GetEncounterBubble))
  (bind _childactor (Variables|/Script/Engine.ChildActorComponent|GetChildActor _encounterbubble))
  ;[1 1 1] Hide And Kill - Hide and deactivate elements that should be removed then set the lifespan
  (Audio|Components|Audio|FadeOut (Variables|Default|GetLoopingAudio) 0.0 0.0 "Linear" :!class /Script/Engine.AudioComponent)
  (Development|SetHiddeninGame (Variables|Default|GetCoreMesh) true false)
  (Components|Activation|SetActive (Variables|Default|GetNS_Sideways_Core_Ambient) true false)
  (Rendering|SetActorHiddenInGame _childactor true)
  (Collision|SetActorEnableCollision _childactor false)
  (Actor|SetLifeSpan _childactor 0.2)
  (Actor|SetLifeSpan 5.0))

(event Custom|UpdateLoopingAudio
  (params (Music :type Sound%20Base%20Object%20Reference))
  (bind _loopingaudio (Variables|Default|GetLoopingAudio))
  ;[1 1 1] Handle looping audio updates
  (Audio|Components|Audio|SetSound _loopingaudio Music)
  (Audio|Components|Audio|FadeIn _loopingaudio 0.2 1.0 0.0 "Linear" :!class /Script/Engine.AudioComponent))

(fn EnsureVisualRadiusIsInSteppedIncrements
  (params (VisualRadius :type Float%20%28double-precision%29))
  (bind _returnvalue 1e+03)
  (bind _returnvalue_2 (/ VisualRadius _returnvalue))
  (bind _returnvalue_3 (Math|Float|Truncate _returnvalue_2))
  (bind _returnvalue_4 (Math|Integer|Clamp(Integer) _returnvalue_3 4 20))
  (bind _returnvalue_5 (Math|Conversions|ToFloat(Integer) _returnvalue_4))
  (bind _returnvalue_6 (* _returnvalue_5 _returnvalue))
  ;[1 1 1] Make sure the size is in reasonable increment chunks for data purposes.
  ;
  ; We need lots of corresponding EQS paramater data that matches the radius, so we don't want an infinite sliding scale. Just a set number of reasonable buckets.
  (Development|PrintString (Utilities|String|BuildString(Float) (Utilities|String|BuildString(Integer) (Utilities|String|BuildString(Float) (Utilities|String|BuildString(Float) (Utilities|String|BuildString(Float) "" "Incoming R = " VisualRadius " | ") "Increment Val = " _returnvalue " | ") "Num Increments = " _returnvalue_2 " | ") "Clamped Increments = " _returnvalue_4 " | ") "RESULTING DIST = " _returnvalue_6) false false)
  (return _returnvalue_6))

(fn GetSidewaysEncounterComponent
  (bind _playspace (CallFunction|GetSidewaysPlayspace))
  (Utilities|IsValid _playspace
    (:Is%20Valid
      (return (Actor|GetComponentbyClass _playspace "/SidewaysGameplay/Components/LargeSidewaysPlayspaceEncounterComponent.LargeSidewaysPlayspaceEncounterComponent_C")))
    (:Is%20Not%20Valid
      (return 0))))

(fn GetSidewaysPlayspace
  (bind _asbga_athena_sideways_encounter_bubble_large (Utilities|Casting|CastToBGA_Athena_SidewaysEncounter_BubbleLarge (Variables|/Script/Engine.ChildActorComponent|GetChildActor (Variables|Default|GetEncounterBubble)))
    (:then
      (return (Class|BGAAthenaSidewaysEncounterBubbleLarge|GetSidewaysPlayspace _asbga_athena_sideways_encounter_bubble_large)))
    (:CastFailed
      (return 0))))

(fn SetSkirtMeshes
  (bind _skirtouterpublic (Variables|Default|GetSkirtOuterPublic))
  (bind _skirtinnerpublic (Variables|Default|GetSkirtInnerPublic))
  ;[1 1 1] Set the meshes on the skirt using the exposed vars that are set on the actor in the map.
  (Utilities|IsValid _skirtinnerpublic
    (:Is%20Valid
      (Components|StaticMesh|SetStaticMesh (Variables|Default|GetSkirtInner) _skirtinnerpublic))
    (:Is%20Not%20Valid
      _))
  (Utilities|IsValid _skirtouterpublic
    (:Is%20Valid
      (Components|StaticMesh|SetStaticMesh (Variables|Default|GetSkirtOuter) _skirtouterpublic)
      (Transformation|SetWorldRotation (Variables|Default|GetSkirtInner) "0, 0, 0" false false))))

(fn SetupBubble
  (bind _bubblemeshassetdimensions (Variables|InternalLogicVariables|GetBubbleMeshAssetDimensions))
  (bind _returnvalue (/ _bubblemeshassetdimensions 2.0))
  (bind _y (.y _returnvalue))
  (bind _visualradiusinsteppedincrements (Variables|InternalLogicVariables|GetVisualRadiusInSteppedIncrements))
  (bind _returnvalue_2 (/ _visualradiusinsteppedincrements _y))
  (bind _returnvalue_3 (Math|Float|Max(Float) _returnvalue_2 0.1))
  (Variables|InternalLogicVariables|SetVisualRadiusInSteppedIncrements (ScalableFloat|GetValueatLevel (Variables|BubbleSettings|GetBubbleVisualRadius_HF)))
  (bind _encounterbubble (Variables|Default|GetEncounterBubble)
    (:then
      (bind _childactor (Variables|/Script/Engine.ChildActorComponent|GetChildActor _encounterbubble)
        (:then
          (bind _asbga_athena_sideways_encounter_bubble_large (Utilities|Casting|CastToBGA_Athena_SidewaysEncounter_BubbleLarge _childactor))
          (Transformation|SetActorScale3D _asbga_athena_sideways_encounter_bubble_large (Math|Vector|MakeVector (Math|Float|Max(Float) (/ _visualradiusinsteppedincrements (.x _returnvalue)) 0.1) _returnvalue_3 (Math|Float|Max(Float) (/ (ScalableFloat|GetValueatLevel (Variables|BubbleSettings|GetBubbleVisualHalfHeight_HF)) (.z _returnvalue)) 0.1)))
          (Transformation|SetWorldLocation (Variables|/Script/Engine.Actor|GetRootComponent _asbga_athena_sideways_encounter_bubble_large) (+ (Transformation|GetActorLocation) (Variables|BubbleSettings|GetBubbleVisualOffset)) false false)
          (Niagara|SetNiagaraVariable(Float) (Variables|Default|GetNS_SideWays_SurfaceSmoke) "SphereRadius" (/ (* _returnvalue_3 99.849998) 2.0)))
        (:else)))
    (:else)))

(fn ShouldDie
  (params (Damage :type Float%20%28single-precision%29) (EventInstigator :type Controller%20Object%20Reference) (DamageCauser :type Actor%20Object%20Reference) (bIsTest :type Boolean) (EffectContext :type Gameplay%20Effect%20Context%20Handle%20Structure%20%28by%20ref%29))
  (return false))

(fn ConstructionScript
  (CallFunction|SetupBubble)
  (CallFunction|SetSkirtMeshes))
