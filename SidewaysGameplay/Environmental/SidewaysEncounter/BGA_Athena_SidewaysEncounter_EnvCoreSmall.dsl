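; Small Sideways encounter core actor, which starts the encounter when a player pawn overlaps it, unless InstantStart is set.
; On the server, it binds the playspace encounter component's scale, kill, cosmetic cue and music events once playlist data is ready.
; A timeline scales the bubble mesh, audio volume and smoke, then notifies the component and playspace.
; It also provides staged core FX, kill cleanup and a replay fallback that finds the bubble by proximity.
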(event EventBeginPlay
  (bind _encountercomponent (CallFunction|GetSidewaysEncounterComponent))
  ;[1 1 1] BeginPlay - Register to events on the encounter component
  (if (Networking|IsServer)
    (Class|FortAsyncActionCurrentPlaylistReady|PlaylistDataReadyAsync
      (:OnReady
        (Default|AssignScaleBubbleEvent _encountercomponent (delegate Custom|ScaleBubbleEvent))
        (Default|AssignKillEvent _encountercomponent (delegate Custom|KillEvent))
        (Default|AssignSpawnCosmeticEvent _encountercomponent (delegate Custom|SpawnCosmeticEvent))
        (Default|AssignUpdateLoopingAudioEvent _encountercomponent (delegate Custom|UpdateLoopingAudioEvent))
        ;[1 1 1] Setup Niagara values
        (Effects|Components|ParticleSystem|SetFloatParameter (Variables|Default|GetNS_SideWays_Volume_InWorld) "SphereSize" 5e+02)
        (Effects|Components|ParticleSystem|SetFloatParameter (Variables|Default|GetNS_SideWays_Volume_InWorld) "SpawnRateScale" 0.5))) ; Make sure we load any hotfix overrides
    (else
      ;[1 0.3938 0.01266] Deal with Replay Weirdness - DISABLED FOR NOW
      (ReplayContext|AssignOnReplayScrubComplete (BlueprintContext|GetFortReplayContext) (delegate Custom|HandleReplayScrubComplete)))))

(event Custom|ScaleBubbleEvent
  (params (StartScale :type Float%20%28double-precision%29) (EndScale :type Float%20%28double-precision%29))
  ;[1 1 1] BeginPlay - Register to events on the encounter component
  (CallFunction|StartBubbleScaling StartScale EndScale))

(event Custom|SpawnCosmeticEvent
  (params (GameplayCueTag :type Gameplay%20Tag%20Structure))
  ;[1 1 1] BeginPlay - Register to events on the encounter component
  (Ability|ExecuteGameplayCueReplicated GameplayCueTag))

(event Custom|KillEvent
  ;[1 1 1] BeginPlay - Register to events on the encounter component
  (CallFunction|HideandKill))

(event Custom|UpdateLoopingAudioEvent
  (params (Music :type Sound%20Base%20Object%20Reference))
  ;[1 1 1] BeginPlay - Register to events on the encounter component
  (CallFunction|UpdateLoopingAudio Music))

(event Custom|HandleReplayScrubComplete)

(event Custom|StartBubbleScaling
  (params (StartScale :type Float%20%28double-precision%29) (EndScale :type Float%20%28double-precision%29))
  (bind _encountercomponent (CallFunction|GetSidewaysEncounterComponent))
  (bind _ns_sideways_surfacesmoke_sm (Variables|Default|GetNS_SideWays_SurfaceSmoke_SM))
  ;[1 1 1] Scale Bubble - Scale the bubble to the new size it needs to be
  (if (Replay|GetIsPlayingReplay)
    (Fortnite|FortShippingLog 0 "Small Sideways Encounter: Replay Start Bubble Scaling") ; Logging to try and catch a replay bug
    (bind (_direction _alpha) (Timeline|BubbleScale PlayFromStart)
      (:Update
        (bind _bubble_mesh (CallFunction|FindEncounterBubbleChildActorStaticMeshComponent))
        (Transformation|SetWorldScale3D _bubble_mesh (Math|Vector|MakeVector (Math|Float|Lerp StartScale EndScale _alpha) (Math|Float|Lerp StartScale EndScale _alpha) (Math|Float|Lerp StartScale EndScale _alpha)))
        (CallFunction|UpdateAGVScale (Math|Vector|MakeVector (Math|Float|Lerp StartScale EndScale _alpha) (Math|Float|Lerp StartScale EndScale _alpha) (Math|Float|Lerp StartScale EndScale _alpha)))
        ;[1 1 1] Check if closing
        (if (Math|Conversions|ToBoolean(Integer) (Math|Integer|Max(Integer) (Math|Float|Floor EndScale)))
          ;[1 1 1] Scale the smoke sheet in the correct direction
          (Effects|Components|ParticleSystem|SetFloatParameter _ns_sideways_surfacesmoke_sm "SphereScale" _alpha)
          (else
            ;[1 1 1] If in closing, also scale down the wobble
            (Rendering|Material|SetScalarParameterValueonMaterials _bubble_mesh "WPO_Scale" (- 1.0 _alpha))
            (Effects|Components|ParticleSystem|SetFloatParameter _ns_sideways_surfacesmoke_sm "SphereScale" (- 1.0 _alpha)))))
      (:Finished
        (Utilities|IsValid _encountercomponent
          (:Is%20Valid
            (Class|SmallSidewaysPlayspaceEncounterComponent|BubbleScalingFinished _encountercomponent)
            (Class|SidewaysPlayspace|NotifySizeChanged (CallFunction|GetSidewaysPlayspace)) ; This makes sure that the correct final radius gets set into the Material Parameter Collection Parameters
          ))))
    (else
      (Timeline|BubbleScale PlayFromStart))))

(event OnComponentBeginOverlap(CoreOverlap)
  (params (OverlappedComponent :type Primitive%20Component%20Object%20Reference) (OtherActor :type Actor%20Object%20Reference) (OtherComp :type Primitive%20Component%20Object%20Reference) (OtherBodyIndex :type Integer) (bFromSweep :type Boolean) (SweepResult :type Hit%20Result%20Structure%20%28by%20ref%29))
  (bind _encountercomponent (CallFunction|GetSidewaysEncounterComponent))
  ;[1 1 1] Core Begin Overlap - Start the Encounter
  (if (not (Variables|Default|GetInstantStart))
    (Utilities|Casting|CastToFortPlayerPawn OtherActor)
    (Utilities|IsValid _encountercomponent
      (:Is%20Valid
        (Class|SmallSidewaysPlayspaceEncounterComponent|StartEncounter _encountercomponent)))))

(event Custom|UpdateLoopingAudio
  (params (Music :type Sound%20Base%20Object%20Reference))
  (bind _loopingaudio (Variables|Default|GetLoopingAudio))
  ;[1 1 1] Handle looping audio updates
  (Audio|Components|Audio|SetSound _loopingaudio Music)
  (Audio|Components|Audio|FadeIn _loopingaudio 0.2 1.0 0.0 "Linear" :!class /Script/Engine.AudioComponent))

(event Custom|HideAndKill
  (bind _encounterbubble (Variables|Default|GetEncounterBubble))
  (bind _childactor (Variables|/Script/Engine.ChildActorComponent|GetChildActor _encounterbubble))
  ;[1 1 1] Hide And Kill - Hide and deactivate elements that should be removed then set the lifespan
  (Audio|Components|Audio|FadeOut (Variables|Default|GetLoopingAudio) 0.0 0.0 "Linear" :!class /Script/Engine.AudioComponent)
  (Components|Activation|Deactivate (Variables|Default|GetCoreSystem))
  (Collision|SetCollisionEnabled (Variables|Default|GetCoreOverlap) "NoCollision")
  (Collision|SetActorEnableCollision _childactor false)
  (Actor|SetLifeSpan _childactor 0.2)
  (Actor|SetLifeSpan 3.0)
  (Actor|DestroyActor (Variables|/Script/Engine.ChildActorComponent|GetChildActor (Variables|Default|GetAudioGameplayVolume)))
  (bind _asbga_athena_sideways_encounter_bubble_small (Utilities|Casting|CastToBGA_Athena_SidewaysEncounter_BubbleSmall (Variables|/Script/Engine.ChildActorComponent|GetChildActor (Variables|Default|GetEncounterBubble))))
  (Collision|SetActorEnableCollision (Variables|/Script/Engine.ChildActorComponent|GetChildActor (Variables|/SidewaysGameplay/Environmental/SidewaysEncounter/BGA_Athena_SidewaysEncounter_BubbleSmall.BGA_Athena_SidewaysEncounter_BubbleSmall_C|GetPlayspaceGameplayVolume _asbga_athena_sideways_encounter_bubble_small)) false)
  (Development|SetHiddeninGame (Variables|/SidewaysGameplay/Environmental/SidewaysEncounter/BGA_Athena_SidewaysEncounter_BubbleSmall.BGA_Athena_SidewaysEncounter_BubbleSmall_C|GetEncounterEdge _asbga_athena_sideways_encounter_bubble_small) true false)
  (Components|Activation|Deactivate (Variables|Default|GetNS_SideWays_Volume_InWorld))
  (Rendering|SetVisibility (Variables|Default|GetSM_Sideways_SmokePillars) false false))

(event Custom|ChangeCoreFX
  (params (Stage :type Integer))
  ;[1 1 1] Change Core FX - Based on the stage adjust the core FX to be more intense.
  (switch int Stage
    (:1
      (CallFunction|CEUpdateCoreFXStage1))
    (:2
      (CallFunction|CEUpdateCoreFXStage2))
    (:3
      (CallFunction|CEUpdateCoreFXStage3))))

(event Custom|CE_UpdateCoreFX_Stage1
  (bind _coresystem (Variables|Default|GetCoreSystem))
  ;[1 1 1] Change Core FX - Based on the stage adjust the core FX to be more intense.
  (Networking|CanExecuteCosmeticEvents
    (:True
      (Utilities|IsValid _coresystem
        (:Is%20Valid
          (Components|Activation|Activate _coresystem true)
          (Effects|Components|ParticleSystem|SetIntParameter _coresystem "Stage" 1)
          (Effects|Components|ParticleSystem|SetBoolParameter _coresystem "TurnOnBurstEmitters" true)
          (Development|PrintString "Stage1" true true "(R=1.000000,G=0.000000,B=0.000000,A=1.000000)" 2e+01))))))

(event Custom|CE_UpdateCoreFX_Stage2
  (bind _coresystem (Variables|Default|GetCoreSystem))
  ;[1 1 1] Change Core FX - Based on the stage adjust the core FX to be more intense.
  (Networking|CanExecuteCosmeticEvents
    (:True
      (Utilities|IsValid _coresystem
        (:Is%20Valid
          (Components|Activation|Activate _coresystem true)
          (Effects|Components|ParticleSystem|SetIntParameter _coresystem "Stage" 2)
          (Effects|Components|ParticleSystem|SetBoolParameter _coresystem "TurnOnBurstEmitters" true)
          (Development|PrintString "Stage2" true true "(R=1.000000,G=0.000000,B=0.000000,A=1.000000)" 2e+01))))))

(event Custom|CE_UpdateCoreFX_Stage3
  (bind _coresystem (Variables|Default|GetCoreSystem))
  ;[1 1 1] Change Core FX - Based on the stage adjust the core FX to be more intense.
  (Networking|CanExecuteCosmeticEvents
    (:True
      (Utilities|IsValid _coresystem
        (:Is%20Valid
          (Components|Activation|Activate _coresystem true)
          (Effects|Components|ParticleSystem|SetIntParameter _coresystem "Stage" 3)
          (Effects|Components|ParticleSystem|SetBoolParameter _coresystem "TurnOnBurstEmitters" true)
          (Development|PrintString "Stage3" true true "(R=1.000000,G=0.000000,B=0.000000,A=1.000000)" 2e+01))))))

(event Custom|EncounterStateChangedBGA
  (params (EncounterState :type Enum_Athena_SidewaysEncounter_State%20Enum :default "NewEnumerator6")))

(fn FindEncounterBubbleChildActorStaticMeshComponent
  (bind _encounterbubble (Variables|Default|GetEncounterBubble))
  (bind _childactor (Variables|/Script/Engine.ChildActorComponent|GetChildActor _encounterbubble))
  (Utilities|IsValid _childactor
    (:Is%20Valid
      (bind _asbga_athena_sideways_encounter_bubble_small (Utilities|Casting|CastToBGA_Athena_SidewaysEncounter_BubbleSmall _childactor)
        (:then
          (return (Variables|/SidewaysGameplay/Environmental/SidewaysEncounter/BGA_Athena_SidewaysEncounter_BubbleSmall.BGA_Athena_SidewaysEncounter_BubbleSmall_C|GetEncounterEdge _asbga_athena_sideways_encounter_bubble_small)))
        (:CastFailed
          (return 0))))
    (:Is%20Not%20Valid
      (if (Replay|GetIsPlayingReplay)
        (bind _outactors (Actor|GetAllActorsOfClass (Variables|/Script/Engine.ChildActorComponent|GetChildActorClass (Variables|Default|GetEncounterBubble))))
        (for _array_element _outactors
          (if (< (Math|Vector|VectorLengthSquared (- (Transformation|GetActorLocation _array_element) (Transformation|GetActorLocation))) 1e+04)
            (bind _asbga_athena_sideways_encounter_bubble_small_2 (Utilities|Casting|CastToBGA_Athena_SidewaysEncounter_BubbleSmall _array_element))
            (return (Variables|/SidewaysGameplay/Environmental/SidewaysEncounter/BGA_Athena_SidewaysEncounter_BubbleSmall.BGA_Athena_SidewaysEncounter_BubbleSmall_C|GetEncounterEdge _asbga_athena_sideways_encounter_bubble_small_2))))
        (return 0)
        (else
          (return 0))))))

(fn GetSidewaysEncounterComponent
  (bind _playspace (CallFunction|GetSidewaysPlayspace))
  (Utilities|IsValid _playspace
    (:Is%20Valid
      (return (Actor|GetComponentbyClass _playspace "/SidewaysGameplay/Components/SmallSidewaysPlayspaceEncounterComponent.SmallSidewaysPlayspaceEncounterComponent_C")))
    (:Is%20Not%20Valid
      (return 0))))

(fn GetSidewaysPlayspace
  (bind _asbga_athena_sideways_encounter_bubble_small (Utilities|Casting|CastToBGA_Athena_SidewaysEncounter_BubbleSmall (Variables|/Script/Engine.ChildActorComponent|GetChildActor (Variables|Default|GetEncounterBubble)))
    (:then
      (return (Class|BGAAthenaSidewaysEncounterBubbleSmall|GetSidewaysPlayspace _asbga_athena_sideways_encounter_bubble_small)))
    (:CastFailed
      (return 0))))

(fn OnRep_DisableWorldFX
  ;[1 1 1] Disable World FX
  (if (Variables|Default|GetDisableWorldFX)
    (Rendering|SetVisibility (Variables|Default|GetSM_Sideways_SmokePillars) false false)
    (Components|Activation|Deactivate (Variables|Default|GetNS_SideWays_Volume_InWorld))))

(fn OnRep_EncounterState
  (CallFunction|EncounterStateChangedBGA (Variables|Default|GetEncounterState))
  (Networking|ForceNetUpdate))

(fn OnRep_Music
  ;[1 1 1] Rep the music down to set the audio we should be hearing
  (CallFunction|UpdateLoopingAudio (Variables|Default|GetMusic))
  (Networking|ForceNetUpdate))

(fn ShouldDie
  (params (Damage :type Float%20%28single-precision%29) (EventInstigator :type Controller%20Object%20Reference) (DamageCauser :type Actor%20Object%20Reference) (bIsTest :type Boolean) (EffectContext :type Gameplay%20Effect%20Context%20Handle%20Structure%20%28by%20ref%29))
  (return false))

(fn UpdateAGVScale
  (params (NewScale :type Vector))
  (bind _asagv_sideways (Utilities|Casting|CastToAGV_Sideways (Variables|/Script/Engine.ChildActorComponent|GetChildActor (Variables|Default|GetAudioGameplayVolume))))
  (Class|AGVSideways|SetRadius _asagv_sideways (* (.x NewScale) 5e+01)))

(fn ConstructionScript)
