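; Grabber child actor of the small Sideways encounter core that targets players entering its range with an ability.
; On the server it registers with the core as owner, or expires if disabled; clients disable its collision.
; Range overlaps maintain a pawn list, trigger the grab ability and start or stop a cosmetic timer pulsing the core's particles at the fire interval.
; StartEncounter starts the core's encounter component and destroys the grabber unless shut down.
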(event EventBeginPlay
  ;[1 1 1] Initialize Kismit if enabled, if not kill the actor
  (if (Networking|IsServer)
    (Utilities|IsEnabled
      (:true
        (bind _asbga_athena_sideways_encounter_env_core_small (Utilities|Casting|CastToBGA_Athena_SidewaysEncounter_EnvCoreSmall (Actor|GetParentActor)))
        (bind _output_get (Variables|Default|SetOwnerEncounter _asbga_athena_sideways_encounter_env_core_small))
        (Building|InitializeKismetSpawnedBuildingActor _output_get 0 false))
      (:false
        (Actor|SetLifeSpan 0.2)))))

(event Custom|StartEncounter
  ;[1 1 1] When the encounter starts destory the grabber
  (if (not (Variables|Default|GetEncounterShutdown))
    (bind _asbga_athena_sideways_encounter_env_core_small (Utilities|Casting|CastToBGA_Athena_SidewaysEncounter_EnvCoreSmall (Actor|GetParentActor)))
    (Class|SmallSidewaysPlayspaceEncounterComponent|StartEncounter (Class|BGAAthenaSidewaysEncounterEnvCoreSmall|GetSidewaysEncounterComponent _asbga_athena_sideways_encounter_env_core_small))
    (Actor|DestroyActor)))

(event OnComponentBeginOverlap(GrabberRange)
  (params (OverlappedComponent :type Primitive%20Component%20Object%20Reference) (OtherActor :type Actor%20Object%20Reference) (OtherComp :type Primitive%20Component%20Object%20Reference) (OtherBodyIndex :type Integer) (bFromSweep :type Boolean) (SweepResult :type Hit%20Result%20Structure%20%28by%20ref%29))
  ;[1 1 1] Add and remove actors to target and being the ability.
  (bind _asfort_player_pawn (Utilities|Casting|CastToFortPlayerPawn OtherActor))
  (Utilities|Array|AddUnique (Variables|Default|GetPawnsInRange) _asfort_player_pawn)
  (Utilities|IsEnabled
    (:true
      (Abilities|TryActivateAbilitybyClass (Ability|GetFortAbilitySystemComponent self) (Variables|Default|GetGA_GrabberAbility) false)
      (CallFunction|CEStartSpawningPrj))))

(event OnComponentEndOverlap(GrabberRange)
  (params (OverlappedComponent :type Primitive%20Component%20Object%20Reference) (OtherActor :type Actor%20Object%20Reference) (OtherComp :type Primitive%20Component%20Object%20Reference) (OtherBodyIndex :type Integer))
  ;[1 1 1] Add and remove actors to target and being the ability.
  (bind _asfort_player_pawn (Utilities|Casting|CastToFortPlayerPawn OtherActor))
  (Utilities|Array|RemoveItem (Variables|Default|GetPawnsInRange) _asfort_player_pawn)
  (CallFunction|CEStopSpawningPrj))

(event Custom|CE_StartSpawningPrj
  ;[1 1 1] Pulsing the core when fring the grabber prj
  (Networking|CanExecuteCosmeticEvents
    (:True
      (bind _asbga_athena_sideways_encounter_env_core_small (Utilities|Casting|CastToBGA_Athena_SidewaysEncounter_EnvCoreSmall (Actor|GetParentActor)))
      (Variables|Default|SetCore%20System (Variables|/SidewaysGameplay/Environmental/SidewaysEncounter/BGA_Athena_SidewaysEncounter_EnvCoreSmall.BGA_Athena_SidewaysEncounter_EnvCoreSmall_C|GetCoreSystem _asbga_athena_sideways_encounter_env_core_small))
      (bind _returnvalue (Utilities|Time|SetTimerbyEvent (delegate Custom|CE_CorePulsing) (ScalableFloat|GetValueatLevel (Variables|Default|GetRow_FireInterval)) true))
      (Variables|Default|SetTimer_LaunchProjectile _returnvalue))))

(event Custom|CE_CorePulsing
  ;[1 1 1] Pulsing the core when fring the grabber prj
  (bind (_direction _corepuling) (Timeline|TL_CorePulsing PlayFromStart)
    (:Update
      (Effects|Components|ParticleSystem|SetFloatParameter (Variables|Default|GetCore%20System) "CorePlusingScalar" (Math|Float|Lerp 1.0 2.0 _corepuling)))))

(event Custom|CE_StopSpawningPrj
  ;[1 1 1] Pulsing the core when fring the grabber prj
  (Networking|CanExecuteCosmeticEvents
    (:True
      (Utilities|Time|ClearandInvalidateTimerbyHandle (Variables|Default|GetTimer_LaunchProjectile))
      (Variables|Default|GetCore%20System
        (:then
          (Effects|Components|ParticleSystem|SetFloatParameter (Variables|Default|GetCore%20System) "CorePlusingScalar" 1.0))
        (:else)))))

(fn ConstructionScript
  ;[1 1 1] Turn off collision on the client. It is not needed.
  (if (not (Networking|IsServer))
    (Collision|SetActorEnableCollision false)))
