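# EQS query in the Sideways encounter environmentals set that picks a randomized location in a donut around the querier.
# A Donut generator places a single ring of points roughly 25 to 29 meters out.
# A Project test filters out points lacking a clear upward box sweep against geometry, ensuring open space overhead.
# A weighted Random test then scores the remaining points to vary the chosen spot.
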EnvQuery /SidewaysGameplay/Environmental/SidewaysEncounter/EQS_Donut_Environmentals_3
options [{
  generator {
   name=EnvQueryGenerator_Donut_1
   class=EnvQueryGenerator_Donut
   properties {
    bUseSpiralPattern=false
    center=&EnvQueryContext_Querier
    innerRadius=2500
    numberOfRings=1
    outerRadius=2900
    pointsPerRing=25
    projectionData {extentX=0, navigationFilter=&RecastFilter_UseDefaultArea, postProjectionVerticalOffset=0}
   }
  }
  tests [
   EnvQueryTest_Project_0 {
    class=EnvQueryTest_Project
    properties {
     boolValue=false
     scoringFactor=-1
     testComment=Require clear sweep up
     testPurpose=Filter
     projectionData {
      extentY=100
      extentZ=100
      navigationFilter=&RecastFilter_UseDefaultArea
      postProjectionVerticalOffset=0
      projectDown=0
      projectUp=500
      traceMode=GeometryByChannel
      traceShape=Box
     }
    }
   }
   FortQueryTest_Random_0 {class=FortQueryTest_Random, properties {floatValueMax=0, scoringFactor=3}}
  ]
 }]
