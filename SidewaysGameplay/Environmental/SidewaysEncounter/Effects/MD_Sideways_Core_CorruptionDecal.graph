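# Translucent deferred decal material for the Sideways encounter core's corruption, with animated cellular noise fading radially.
# Actor-relative world XY, offset by noise, pans the cellular texture.
# A sphere mask drives a curve-atlas lookup that blends two colors and masks the noise.
# The UseCubePrint static switch adds a cube imprint to the distortion, a tinted emissive term and an opacity term faded by decal lifetime.
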Material /SidewaysGameplay/Environmental/SidewaysEncounter/Effects/MD_Sideways_Core_CorruptionDecal
graph {
 properties {blendMode=BLEND_Translucent, materialDomain=MD_DeferredDecal, pixelDepthOffsetMode=PDOM_Legacy, usedShadingModels=Default Lit}
 nodes [
  Constant()
  Constant() {r=0.5}
  n1 = VectorParameter() {parameterName=ColorC, defaultValue=<rgba 5 0.5112 0.362 0>}
  n2 = Constant() {r=0.5}
  n3 = CurveAtlasRowParameter(CurveTime=SphereMask(A=TextureCoordinate(), B=@n2, Radius=@n2) {attenuationRadius=0.5, hardnessPercent=0}) {
   atlas=/SidewaysGameplay/Environmental/SidewaysEncounter/Effects/Curves/CurveAtlas_Sideways
   curve=/SidewaysGameplay/Environmental/SidewaysEncounter/Effects/Curves/Curve_Sideways_Core01
   parameterName=ShadowCurve
  }
  n4 = WorldPosition()
  n5 = ActorPositionWS()
  n6 = ComponentMask(None=Divide(A=Subtract(A=@n4.XYZ, B=@n5), B=ScalarParameter() {defaultValue=600, parameterName=Tile})) {g=true, r=true}
  n7 = Divide(A=TextureSample() {samplerType=SAMPLERTYPE_Masks, texture=/Game/Global/Textures/Default/Noise/T_CellularNoise}.R) {constB=3}
  n8 = TextureSample() {texture=/CorruptionGameplay/Gameplay/Environmental/Cube/FX/T_Decal_Cube_Imprint}
  n9 = Add(A=@n7, B=Multiply(A=@n8.G) {constB=0.25})
  n10 = Panner(Coordinate=Add(A=@n6, B=StaticSwitchParameter(True=@n9, False=@n7) {parameterName=UseCubePrint})) {speedX=0.05, speedY=0.05}
  n11 = TextureSample(UVs=@n10) {texture=/CorruptionGameplay/Textures/T_Cellular_Noise}
  n12 = Multiply(
   A=LinearInterpolate(
    A=VectorParameter() {parameterName=ColorA, defaultValue=<rgba 0.0469 0 0.0938 1>}.RGB
    B=VectorParameter() {parameterName=ColorB, defaultValue=<rgba 1.225 2 0.0513 1>}.RGB
    Alpha=@n3.G
   )
   B=Add(B=@n11.RGB)
  )
  n13 = Multiply(A=Power(Base=NamedRerouteUsage(), Exp=ScalarParameter() {defaultValue=5, parameterName=CubePrintEmissivePower}), B=Add(A=@n12) {constB=0.25})
  MaterialFunctionCall(Emissive=StaticSwitchParameter(True=Multiply(A=@n1.RGB, B=@n13), False=Constant()) {parameterName=UseCubePrint}) {materialFunction=&MF_FX_EmissiveAdjustments}
  Multiply(A=@n3.G, B=TextureSample(UVs=Multiply(A=@n10) {constB=2}) {texture=/Game/Global/Textures/Default/Caustics/T_Turbulence_Seamless}.R)
  n14 = NamedRerouteUsage()
  Multiply(A=OneMinus(None=@n14), B=Desaturation(None=@n12, Fraction=Constant() {r=0.4}))
  n15 = MaterialFunctionCall() {materialFunction=&VectorToRadialValue}
  n16 = ScalarParameter() {defaultValue=600, parameterName=Tile}
  n17 = Add(
   A=Multiply(A=Time(), B=ScalarParameter() {defaultValue=1.5, parameterName=PanSpeed})
   B=MaterialFunctionCall(Float2=Multiply(A=@n15.Radial Coordinates, B=@n16) {constB=2}) {materialFunction=&BreakOutFloat2Components}.G
  )
  n18 = SphereMask(A=TextureCoordinate(), B=Constant() {r=0.5}, Radius=Constant() {r=0.2}) {attenuationRadius=0.5, hardnessPercent=0}
  Multiply(A=Multiply(A=Sine(None=@n17), B=Subtract(B=@n18))) {constB=1.3}
  Multiply(A=@n6, B=Constant2Vector() {g=2, r=2})
  Multiply(A=TextureSample() {texture=/CorruptionGameplay/Textures/T_Caustics_Pattern1}.RGB, B=Constant3Vector() {constant=<rgba 3 1.3654 0.6677 0>})
  NamedRerouteDeclaration(None=@n8.G) {name=CubeImprint, nodeColor=<rgba 0.4297 0 1 1>}
  n19 = Constant() {r=0.1}
  StaticSwitchParameter(True=Add(A=@n14, B=@n19), False=@n19) {parameterName=UseCubePrint}
  n20 = Multiply(A=@n11.B, B=@n3.G)
  n21 = Multiply(A=NamedRerouteUsage()) {constB=1.5}
  StaticSwitchParameter(True=Multiply(A=Add(A=@n20, B=@n21), B=DecalLifetimeOpacity().Opacity), False=@n20) {parameterName=UseCubePrint}
  TextureCoordinate()
 ]
}
