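# Translucent unlit Niagara sprite material for a soft, warped impact effect in the Sideways encounter, colored by particle color through an emissive adjustment function.
# Panning noise distorts center-scaled mask UVs, with dynamic parameters driving distortion, scale, erosion and a radial center cutout.
# Opacity combines that mask with particle alpha and a squared fade on distance between the particle and the surface behind it.
# A static switch adds optional box-gradient edge control.
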Material /SidewaysGameplay/Environmental/SidewaysEncounter/Effects/M_Decal_Warp_Sideways_Impact
graph {
 properties {bUsedWithNiagaraSprites=true, blendMode=BLEND_Translucent, pixelDepthOffsetMode=PDOM_Legacy, shadingModel=MSM_Unlit, usedShadingModels=Unlit}
 nodes [
  Constant() {r=3.5}
  n1 = ParticleColor()
  MaterialFunctionCall(Emissive=@n1.RGB) {materialFunction=&MF_FX_EmissiveAdjustments}
  n2 = DynamicParameter() {parameterIndex=1, defaultValue=<rgba 0 0 0 0>, paramNames [CenterMaskStrength, none1, none2, none3]}
  n3 = LinearInterpolate(B=Constant() {r=0.7}, Alpha=@n2.CenterMaskStrength)
  n4 = OneMinus(None=ComponentMask(None=MaterialFunctionCall(Radius=@n3) {materialFunction=&RadialGradientExponential}.RadialGradientExponential) {r=true})
  n5 = DynamicParameter() {defaultValue=<rgba 0 0 1 0>, paramNames [DistortionStrength, DistortionPanSpeed, MaskUVScale, ErosionStrength]}
  n6 = TextureSampleParameter2D(UVs=Panner(Coordinate=TextureCoordinate(), Speed=@n5.DistortionPanSpeed)) {parameterName=DistortionNoise, samplerSource=SSM_Wrap_WorldGroupSettings, texture=&T_NoisePack_01}
  n7 = Multiply(A=@n6.RGB) {constB=2}
  n8 = DynamicParameter() {defaultValue=<rgba 0 0 1 0>, paramNames [DistortionStrength, DistortionPanSpeed, MaskUVScale, ErosionStrength]}
  n9 = Multiply(A=ComponentMask(None=Subtract(A=@n7)) {g=true, r=true}, B=@n8.DistortionStrength)
  n10 = MaterialFunctionCall(UVs=TextureCoordinate(), Texture Scale=@n8.MaskUVScale) {materialFunction=&ScaleUVsByCenter}
  n11 = DynamicParameter() {defaultValue=<rgba 0 0 1 0>, paramNames [DistortionStrength, DistortionPanSpeed, MaskUVScale, ErosionStrength]}
  n12 = Subtract(
   A=TextureSampleParameter2D(UVs=Add(A=@n9, B=@n10.UVs)) {parameterName=Mask Tex, samplerSource=SSM_Clamp_WorldGroupSettings, texture=/Game/Effects/Fort_Effects/Textures/T_Smoke_Soft}.RGB
   B=@n11.ErosionStrength
  )
  n13 = Divide(
   A=Distance(
    A=MaterialFunctionCall() {materialFunction=&WorldPositionBehindTranslucency}.Result
    B=ParticlePositionWS()
   )
   B=ScalarParameter() {defaultValue=10, group=Distance Fade, parameterName=Fade Distance}
  )
  n14 = ScalarParameter() {defaultValue=300, group=Distance Fade, parameterName=Fade Start Distance, sortPriority=1}
  n15 = OneMinus(None=Clamp(None=Subtract(A=@n13, B=@n14)))
  n16 = TextureCoordinate()
  n17 = ComponentMask(None=TextureCoordinate()) {r=true}
  n18 = Multiply(
   A=Power(
    Base=ComponentMask(None=@n16) {r=true}
    Exp=ScalarParameter() {defaultValue=1, group=Opacity, parameterName=Left-Right Contrast, sortPriority=2}
   )
   B=Power(Base=OneMinus(None=@n17), Exp=ScalarParameter() {defaultValue=1, group=Opacity, parameterName=Right-Left Contrast, sortPriority=3})
  )
  n19 = TextureCoordinate()
  n20 = ComponentMask(None=TextureCoordinate()) {g=true}
  n21 = Multiply(
   A=Power(
    Base=ComponentMask(None=@n19) {g=true}
    Exp=ScalarParameter() {defaultValue=1, group=Opacity, parameterName=Top-Bottom Contrast, sortPriority=4}
   )
   B=Power(Base=OneMinus(None=@n20), Exp=ScalarParameter() {defaultValue=1, group=Opacity, parameterName=Bottom-Top Contrast, sortPriority=5})
  )
  n22 = Clamp(None=Multiply(A=Multiply(A=@n18, B=@n21), B=ScalarParameter() {defaultValue=1, group=Opacity, parameterName=Opacity_Intensity_Compensator, sortPriority=1}))
  n23 = Constant() {r=1}
  Multiply(
   A=Multiply(A=@n1.A, B=Multiply(A=@n4, B=@n12))
   B=Multiply(A=Multiply(A=@n15, B=@n15), B=StaticSwitchParameter(True=@n22, False=@n23) {group=Opacity, parameterName=Use Box Gradient Opacity Control})
  )
 ]
}
