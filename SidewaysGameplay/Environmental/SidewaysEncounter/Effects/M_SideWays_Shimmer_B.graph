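# Translucent unlit Niagara sprite material for a smoky, tinted shimmer effect in the Sideways encounter.
# Panning chromatic noise distorts a sphere-masked smoke flipbook and detail layer, which are tinted by particle color for emissive, with depth-faded opacity on the default shading path.
# World position offset math uses runtime virtual texture landscape height and a particle-supplied sphere radius to fit a skirt where sphere meets terrain, spherized along the V coordinate.
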Material /SidewaysGameplay/Environmental/SidewaysEncounter/Effects/M_SideWays_Shimmer_B
graph {
 properties {bUsedWithNiagaraSprites=true, blendMode=BLEND_Translucent, pixelDepthOffsetMode=PDOM_Legacy, shadingModel=MSM_Unlit, usedShadingModels=Unlit}
 nodes [
  Constant2Vector() {g=-0.1, r=-0.02}
  Constant() {r=2}
  Constant() {r=5000}
  n1 = VectorParameter() {parameterName=Color, defaultValue=<rgba 0.0625 0.0086 0.0339 0>}
  n2 = ParticleColor()
  n3 = DynamicParameter() {defaultValue=<rgba 0 1 1 1>, paramNames [Random,Random2,Flicker,Param4]}
  n4 = ScalarParameter() {defaultValue=2, parameterName=SubUVs, sliderMax=8}
  n5 = MaterialFunctionCall("Animation  Phase (0-1)"=@n3.Random, Number of Rows=@n4, Number of Columns=@n4) {materialFunction=&FlipBook}
  n6 = Add(A=TextureCoordinate() {uTiling=1.2, vTiling=1.2}, B=@n3.Random2)
  n7 = Multiply(A=Time(), B=ScalarParameter() {defaultValue=1, parameterName=PanSpeed})
  n8 = ComponentMask(None=TextureSampleParameter2D(UVs=Panner(Coordinate=@n6, Time=@n7) {speedX=-0.02, speedY=0.05}) {parameterName=Distortion, texture=&T_Inky_Smoke_Tile_Chromatic}.RGB) {g=true, r=true}
  n9 = DynamicParameter() {parameterIndex=1, defaultValue=<rgba 0.05 0.025 1 1>, paramNames [DistortX, DistortY, DistortX_Detail, DistortY_Detail]}
  n10 = Multiply(A=@n8, B=AppendVector(A=@n9.DistortX, B=@n9.DistortX))
  n11 = TextureCoordinate()
  n12 = Constant() {r=0.5}
  n13 = Multiply(
   A=TextureSampleParameter2D(UVs=Add(A=@n5.UVs, B=@n10)) {parameterName=Opacity, texture=/Game/Effects/Fort_Effects/Textures/T_InkySmoke_Flipbook}.RGB
   B=SphereMask(A=Add(A=@n10, B=@n11), B=@n12, Radius=@n12) {hardnessPercent=0}
  )
  n14 = TextureCoordinate() {uTiling=1.5, vTiling=1.5}
  n15 = DynamicParameter() {parameterIndex=1, defaultValue=<rgba 0.05 0.025 1 1>, paramNames [DistortX, DistortY, DistortX_Detail, DistortY_Detail]}
  n16 = AppendVector(A=@n15.DistortX_Detail, B=@n15.DistortX_Detail)
  n17 = Panner(Coordinate=Add(A=Add(A=@n3.Random, B=@n14), B=Multiply(A=@n8, B=@n16)), Time=@n7) {speedX=0.05, speedY=0.12}
  n18 = Multiply(
   A=@n2.RGB
   B=LinearInterpolate(
    A=Multiply(A=@n13, B=@n13)
    B=TextureSampleParameter2D(UVs=@n17) {parameterName=Color, texture=&T_Inky_Smoke_Tile_Chromatic}.G
   ) {constAlpha=0.7}
  )
  n19 = Multiply(A=Desaturation(None=@n18, Fraction=Constant() {r=0.3})) {constB=0.35}
  n20 = ScalarParameter() {defaultValue=0.7, parameterName=Flat Color Mix Weight, sliderMax=1}
  MaterialFunctionCall(Emissive=Max(
    A=LinearInterpolate(A=@n1.RGB, B=@n19, Alpha=@n20) {constAlpha=0.7}
    B=VectorParameter() {parameterName=Color, defaultValue=<rgba 0.151 0.0083 0.0417 0>}.RGB
   )) {materialFunction=&MF_FX_EmissiveAdjustments}
  Multiply(A=@n1.RGB)
  Multiply(A=NamedRerouteUsage(), B=ScalarParameter() {defaultValue=1, parameterName=WPO Multiplier, sliderMax=1})
  n21 = DepthFade(Opacity=@n13, FadeDistance=ScalarParameter() {defaultValue=50, parameterName=DepthFadeOutDist}) {fadeDistanceDefault=50}
  Multiply(
   A=@n2.A
   B=Power(Base=ShadingPathSwitch(Default=@n21, Forward=@n13, Mobile=@n13), Exp=ScalarParameter() {defaultValue=0.5, parameterName=Opacity Exponent}) {constExponent=0.5}
  )
  n22 = ActorPositionWS()
  n23 = MaterialFunctionCall(Float3=@n22) {materialFunction=&BreakOutFloat3Components}
  n24 = NamedRerouteUsage()
  n25 = Subtract(A=@n23.B, B=@n24)
  n26 = Normalize(VectorInput=Subtract(
    A=WorldPosition().XYZ
    B=MaterialFunctionCall(X=@n23.R, Y=@n23.R, Z=@n25) {materialFunction=&MakeFloat3}.Result
   ))
  n27 = MaterialFunctionCall(Float3=@n26) {materialFunction=&BreakOutFloat3Components}
  n28 = ActorPositionWS()
  n29 = Add(
   A=Subtract(A=NamedRerouteUsage(), B=ComponentMask(None=@n28) {b=true})
   B=ScalarParameter() {defaultValue=100, parameterName=WorldHeightBias}
  )
  n30 = ScalarParameter() {defaultValue=256, parameterName=Skirt_Scale}
  n31 = Add(A=DynamicParameter() {parameterIndex=3, paramNames [Param1,Param2,RandomOffset,SphereRadius]}.SphereRadius)
  n32 = Multiply(A=@n31, B=@n31)
  n33 = Distance(A=@n24, B=ComponentMask(None=@n22) {b=true})
  n34 = Multiply(A=@n33, B=@n33)
  n35 = Subtract(A=@n31, B=SquareRoot(None=Subtract(A=@n32, B=@n34)))
  n36 = Add(
   A=Add(
    A=MaterialFunctionCall(X=@n27.R, Y=@n27.R, Z=@n29) {materialFunction=&MakeFloat3}.Result
    B=Multiply(A=@n26, B=@n30)
   )
   B=Multiply(A=@n26, B=Multiply(A=@n35) {constB=-1})
  ) {desc=adjust radius}
  n37 = Subtract(A=Add(A=@n36, B=WorldPosition().XYZ), B=ActorPositionWS())
  n38 = Multiply(
   A=Normalize(VectorInput=@n37)
   B=Subtract(
    A=MaterialFunctionCall(Vector 3=@n37) {materialFunction=/Engine/Functions/Engine_MaterialFunctions02/Utility/VectorLength}.V3 Length
    B=DynamicParameter() {parameterIndex=3, paramNames [Param1,Param2,RandomOffset,SphereRadius]}.SphereRadius
   ) {constB=5000}
  )
  n39 = OneMinus(None=MaterialFunctionCall(Float2=TextureCoordinate()) {materialFunction=&BreakOutFloat2Components}.G)
  NamedRerouteDeclaration(None=Subtract(A=@n36, B=Multiply(A=@n38, B=@n39)) {desc=spherize it}) {name=WPO, nodeColor=<rgba 0.5234 1 0 1>}
  NamedRerouteDeclaration(None=RuntimeVirtualTextureSample!(UVs, World Position) {materialType=WorldHeight, virtualTexture=/Game/Athena/Apollo/Maps/Landscape/VT/RVT_Apollo_Landscape_Height}.WorldHeight) {name=WorldHeight, nodeColor=<rgba 0 1 0.2734 1>}
  ObjectPositionWS()
  ObjectRadius()
  ObjectRadius()
  OneMinus(None=@n18)
 ]
}
