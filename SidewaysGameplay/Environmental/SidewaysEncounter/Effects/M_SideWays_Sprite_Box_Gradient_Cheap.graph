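# Unlit additive Niagara sprite material for SidewaysEncounter effects, drawing a cheap soft box glow tinted by particle color.
# Multiplied, inverted linear UV gradients form the box falloff, scaled by particle alpha and a near-camera fade.
# A switchable Sprite_Capsule world position offset reshapes sprites, subtracting per-frame particle velocity on the previous frame for motion vectors.
# BoxMask nodes are unwired leftovers, and a capsule-UV SphereMask feeds no other listed node.
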Material /SidewaysGameplay/Environmental/SidewaysEncounter/Effects/M_SideWays_Sprite_Box_Gradient_Cheap
graph {
 properties {bUsedWithNiagaraSprites=true, blendMode=BLEND_Additive, pixelDepthOffsetMode=PDOM_Legacy, shadingModel=MSM_Unlit, usedShadingModels=Unlit}
 nodes [
  MaterialFunctionCall!(A, B, Bounds, Edge Falloff) {materialFunction=/Engine/Functions/Engine_MaterialFunctions01/Gradient/BoxMask-2D}
  MaterialFunctionCall!(A, B, Bounds, Edge Falloff) {materialFunction=/Engine/Functions/Engine_MaterialFunctions01/Gradient/BoxMask-3D}
  n1 = ParticleColor()
  MaterialFunctionCall(Emissive=@n1.RGB) {materialFunction=&MF_FX_EmissiveAdjustments}
  n2 = MaterialFunctionCall(UV Channel=TextureCoordinate()) {materialFunction=&LinearGradient}
  n3 = Multiply(A=Multiply(A=OneMinus(None=@n2.UGradient), B=@n2.UGradient), B=Multiply(A=OneMinus(None=@n2.VGradient), B=@n2.VGradient))
  n4 = Constant() {r=16}
  n5 = ScalarParameter() {defaultValue=100, parameterName=NearCameraFadeDist}
  Multiply(
   A=Power(Base=Multiply(A=@n3, B=@n4)) {constExponent=1}
   B=Multiply(A=@n1.A, B=MaterialFunctionCall(Distance=@n5) {materialFunction=&Near_Camera_Fade}.Result)
  )
  n6 = StaticBool()
  n7 = MaterialFunctionCall!(
   Particle Position
   Normalized Particle Direction
   Capsule Diameter
   Capsule Pivot Alignment
   Scale Along Velocity Vector
   ---------Scale Over Velocity Parameters---------
   Particle Speed
   Minimum Speed
   Max Speed
   Minimum Scale Factor=MaterialFunctionCall(Float2=ParticleSize()) {materialFunction=&BreakOutFloat2Components}.G
   Maximum Scale Factor
   ---------- Advanced -----------
   Cap Size Multiplier
   Use Camera Offset=@n6
   Use Advanced Features=@n6
   World Position to Negate
  ) {materialFunction=/Engine/Functions/Engine_MaterialFunctions02/WorldPositionOffset/Sprite_Capsule}
  n8 = Constant() {r=0.5}
  SphereMask(A=@n7.Texture Coordinates, B=@n8, Radius=@n8) {hardnessPercent=0}
  n9 = ParticleDirection()
  n10 = ParticleSpeed()
  n11 = Multiply(A=Multiply(A=Multiply(A=@n9, B=@n10), B=DeltaTime()))
  StaticSwitchParameter(
   True=PreviousFrameSwitch(Current Frame=@n7.World Position Offset, Previous Frame=Subtract(A=@n7.World Position Offset, B=@n11))
   False=Constant()
  ) {defaultValue=true, parameterName=UseWPO?}
 ]
}
