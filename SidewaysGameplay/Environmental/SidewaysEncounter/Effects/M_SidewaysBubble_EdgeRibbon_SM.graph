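# Translucent two-sided Niagara mesh-particle material for the edge ribbon of a Sideways encounter bubble where it meets the landscape.
# World position offset uses RVT landscape height and particle sphere radius to fit the ribbon, plus a camera-masked push relative to the local player's bubble.
# Panning noise distorts curve-atlas lookups for opacity and orange color.
# Opacity fades near the camera, with depth fade and extra turbulence limited by quality level.
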Material /SidewaysGameplay/Environmental/SidewaysEncounter/Effects/M_SidewaysBubble_EdgeRibbon_SM
graph {
 properties {
  bUsedWithNiagaraMeshParticles=true
  blendMode=BLEND_Translucent
  pixelDepthOffsetMode=PDOM_Legacy
  twoSided=true
  usedShadingModels=Default Lit
 }
 nodes [
  ActorPositionWS()
  n1 = ActorPositionWS()
  n2 = MaterialFunctionCall(Float3=@n1) {materialFunction=&BreakOutFloat3Components}
  n3 = RuntimeVirtualTextureSample!(UVs, World Position) {materialType=WorldHeight, virtualTexture=/Game/Athena/Apollo/Maps/Landscape/VT/RVT_Apollo_Landscape_Height}
  n4 = Subtract(A=@n2.B, B=@n3.WorldHeight)
  n5 = Normalize(VectorInput=Subtract(
    A=WorldPosition().XYZ
    B=MaterialFunctionCall(X=@n2.R, Y=@n2.R, Z=@n4) {materialFunction=&MakeFloat3}.Result
   ))
  n6 = MaterialFunctionCall(Float3=@n5) {materialFunction=&BreakOutFloat3Components}
  n7 = ActorPositionWS()
  n8 = Add(A=Subtract(A=@n3.WorldHeight, B=ComponentMask(None=@n7) {b=true}), B=ScalarParameter() {defaultValue=100, parameterName=WorldHeightBias})
  n9 = ScalarParameter() {defaultValue=256, parameterName=Skirt_Scale}
  n10 = DynamicParameter() {parameterIndex=2, paramNames [EngineOwnerScaleX, EngineOwnerScaleY, EngineOwnerScaleZ, Param4]}
  n11 = Divide(A=Add(A=Add(A=@n10.EngineOwnerScaleX, B=@n10.EngineOwnerScaleX), B=@n10.EngineOwnerScaleZ)) {constB=3}
  n12 = DynamicParameter() {parameterIndex=3, paramNames [Param1,Param2,RandomOffset,SphereRadius]}
  n13 = Distance(A=@n3.WorldHeight, B=ComponentMask(None=@n1) {b=true})
  n14 = Subtract(A=Multiply(A=@n12.SphereRadius, B=@n12.SphereRadius), B=Multiply(A=@n13, B=@n13))
  n15 = Multiply(A=Subtract(A=@n12.SphereRadius, B=SquareRoot(None=@n14))) {constB=-1}
  n16 = Add(
   A=Add(
    A=MaterialFunctionCall(X=@n6.R, Y=@n6.R, Z=@n8) {materialFunction=&MakeFloat3}.Result
    B=Multiply(A=@n5, B=@n9)
   )
   B=Multiply(A=@n5, B=Multiply(A=@n15, B=@n11))
  ) {desc=adjust radius}
  n17 = Subtract(A=Add(A=@n16, B=WorldPosition().XYZ), B=ActorPositionWS())
  n18 = MaterialFunctionCall(Vector 3=@n17) {materialFunction=/Engine/Functions/Engine_MaterialFunctions02/Utility/VectorLength}
  n19 = DynamicParameter() {parameterIndex=3, paramNames [Param1,Param2,RandomOffset,SphereRadius]}
  n20 = OneMinus(None=MaterialFunctionCall(Float2=TextureCoordinate()) {materialFunction=&BreakOutFloat2Components}.G) {desc=only upper half}
  n21 = Multiply(
   A=Multiply(A=Normalize(VectorInput=@n17), B=Subtract(A=@n18.V3 Length, B=@n19.SphereRadius) {constB=5000})
   B=Clamp(None=Multiply(A=@n20) {constB=1.3})
  )
  Add(A=@n16, B=@n21)
  n22 = Subtract(A=@n16, B=@n21)
  n23 = MaterialFunctionCall(Z=ScalarParameter() {parameterName=Z_offset}) {materialFunction=&MakeFloat3}
  n24 = CollectionParameter() {
   collection=&FortniteMaterialParameters
   parameterName=SidewaysBubble_Location_LocalP1
  }
  n25 = ComponentMask(None=@n24) {b=true, g=true, r=true}
  n26 = Normalize(VectorInput=Subtract(A=WorldPosition().XYZ, B=@n25))
  n27 = LinearInterpolate(Alpha=DynamicParameter() {parameterIndex=3, paramNames [BubbleMask,Param2,Param3,Param4]}.BubbleMask) {constA=-1}
  n28 = Multiply(A=@n26, B=Multiply(A=ScalarParameter() {defaultValue=350, parameterName=BubblePushDistance}, B=@n27))
  n29 = CameraPositionWS()
  n30 = Distance(A=@n25, B=@n29)
  n31 = CollectionParameter() {
   collection=&FortniteMaterialParameters
   parameterName=SidewaysBubble_Scale_LocalP1
  }
  n32 = Subtract(A=@n31, B=ScalarParameter() {defaultValue=768, parameterName=Bias})
  n33 = ScalarParameter() {defaultValue=400, parameterName=DarkenRadius}
  n34 = ScalarParameter() {defaultValue=0.5, parameterName=DarkenHardness}
  n35 = WorldPosition()
  n36 = ScalarParameter() {defaultValue=512, parameterName=CameraMaskRadius}
  n37 = Constant() {r=0.5}
  n38 = DotProduct(A=MaterialFunctionCall() {materialFunction=&CameraDirectionVector}.Result, B=@n26)
  n39 = Multiply(A=@n38) {constB=-1}
  n40 = Clamp(None=@n27)
  n41 = Multiply(
   A=Multiply(A=SphereMask(A=@n30, B=@n32, Radius=@n33, Hardness=@n34), B=SphereMask(A=@n35.XYZ, B=@n29, Radius=@n36, Hardness=@n37))
   B=Clamp(None=LinearInterpolate(A=@n38, B=@n39, Alpha=@n40))
  )
  NamedRerouteDeclaration(None=Add(A=Add(A=@n22, B=@n23.Result), B=LinearInterpolate(B=@n28, Alpha=@n41))) {name=WPO, nodeColor=<rgba 0.5234 1 0 1>}
  NamedRerouteUsage()
  n42 = Add(
   A=Time()
   B=Multiply(A=DynamicParameter() {parameterIndex=3, paramNames [Param1,Param2,RandomOffset,SphereRadius]}.RandomOffset) {constA=1, constB=35, desc=time offset}
  )
  n43 = TextureCoordinate() {coordinateIndex=1}
  Noise!(World Position=AppendVector(A=@n43, B=Multiply(A=@n42) {constB=0.1, desc=speed}), FilterWidth) {desc=1d noise with 1d time, levelScale=1.5, noiseFunction=NOISEFUNCTION_GradientTex3D, scale=2}
  Noise(World Position=Multiply(A=WorldPosition().XYZ) {constB=0.001}) {bTiling=true, levelScale=1.5, levels=4, noiseFunction=NOISEFUNCTION_GradientTex3D, outputMax=0.9, outputMin=0, scale=0.6}
  ObjectPositionWS()
  n44 = MaterialFunctionCall(Float2=@n43) {materialFunction=&BreakOutFloat2Components}
  n45 = CurveAtlasRowParameter(CurveTime=@n44.R) {
   atlas=/SidewaysGameplay/Environmental/SidewaysEncounter/Effects/Curves/CA_Sideways_Bubble
   curve=/SidewaysGameplay/Environmental/SidewaysEncounter/Effects/Curves/C_Sideways_OpacityU
   defaultValue=2
   parameterName=CurveAtlas_Opacity
  }
  n46 = MaterialFunctionCall(Float2=TextureCoordinate() {coordinateIndex=1}) {materialFunction=&BreakOutFloat2Components}
  n47 = MaterialFunctionCall(X=Multiply(A=@n46.R) {constB=2}, Y=@n46.G) {materialFunction=&MakeFloat2}
  n48 = Multiply(A=@n47.Result) {constB=0.1}
  n49 = ComponentMask(None=TextureSample(UVs=Panner(Coordinate=@n48) {speedY=0.005}) {texture=/Game/TimeOfDay/LightFunction/T_Cloud2dNoise}.RGB) {g=true, r=true}
  n50 = TextureSample(UVs=Multiply(A=@n47.Result) {constB=0.4}) {texture=&T_Cloud2dNoise}
  n51 = Add(A=Add(A=Multiply(A=@n49) {constB=0.4}, B=@n48), B=Multiply(A=ComponentMask(None=@n50.RGB) {g=true, r=true}) {constB=0.2})
  n52 = TextureSample(UVs=Panner(Coordinate=@n51) {speedY=0.03}) {texture=/Game/Packages/Fortress_Transylvania/Haunted_Forest/Swamps/Textures/clouds_dust_detail_01}
  n53 = TextureSample(UVs=Panner(Coordinate=@n51) {speedY=0.05}) {texture=/Game/Global/Textures/Default/Noise/T_FireNoise_02}
  n54 = Add(A=Multiply(A=Time()) {constB=0.1}, B=@n53.RGB)
  n55 = MaterialFunctionCall(
   Input=Cosine(None=Multiply(A=@n54) {constB=1.5})
   Input Low=Constant()
   Input High=Constant() {r=1}
   Target Low=Constant()
   Target High=Constant() {r=0.3}
  ) {materialFunction=&RemapValueRange}
  n56 = Add(
   A=Multiply(A=Multiply(A=@n52.RGB) {constB=0.3}, B=@n53.RGB)
   B=MaterialFunctionCall(X=@n55.Result, Y=@n55.Result, Z=@n55.Result) {materialFunction=&MakeFloat3}.Result
  )
  n57 = Add(
   A=@n44.G
   B=Multiply(A=MaterialFunctionCall(Float3=@n56) {materialFunction=&BreakOutFloat3Components}.R) {constB=0.125}
  )
  n58 = CurveAtlasRowParameter(CurveTime=@n57) {
   atlas=&CA_Sideways_Bubble
   curve=/SidewaysGameplay/Environmental/SidewaysEncounter/Effects/Curves/C_Sideways_OpacityV_B
   defaultValue=4
   parameterName=CurveAtlas_Opacity_B
  }
  n59 = Subtract(A=Distance(A=CameraPositionWS(), B=WorldPosition().XYZ), B=ScalarParameter() {defaultValue=25, parameterName=Depth Bias})
  n60 = ScalarParameter() {defaultValue=55, parameterName=Fade Distance}
  n61 = Multiply(
   A=Multiply(A=Multiply(A=@n45, B=@n58), B=ScalarParameter() {defaultValue=1.5, parameterName=OopacityScale}) {constB=2}
   B=Clamp(None=Divide(A=@n59, B=@n60))
  )
  n62 = ShadingPathSwitch(
   Default=DepthFade(Opacity=@n61, FadeDistance=ScalarParameter() {defaultValue=50, parameterName=DepthFadeOutDist}) {fadeDistanceDefault=50}
   Forward=@n61
   Mobile=@n61
  )
  Reroute(None=QualitySwitch(Default=@n62, Low=@n61, High=@n61, Medium=@n61, Epic=@n62))
  n63 = TextureCoordinate() {coordinateIndex=1, uTiling=4, vTiling=4}
  n64 = Time()
  n65 = Panner(Coordinate=TextureCoordinate() {coordinateIndex=1, uTiling=2, vTiling=2}, Time=@n64) {speedY=0.3}
  n66 = Multiply(
   A=TextureSample(UVs=Panner(Coordinate=@n63, Time=@n64) {speedY=0.2}) {texture=/FortWater/Textures/T_Turbulence_Seamless}.RGB
   B=Power(Base=TextureSample(UVs=@n65) {texture=&T_Turbulence_Seamless}.RGB) {constExponent=0.5}
  )
  n67 = TextureSample(UVs=Multiply(A=@n43) {constB=0.45}) {texture=&T_Inky_Smoke_Tile}
  n68 = Multiply(
   A=@n57
   B=MaterialFunctionCall(
    Input=Power(Base=@n67.R) {constExponent=0.2}
    Input Low=Constant() {r=0.3}
    Input High=Constant() {r=1}
    Target Low=Constant()
    Target High=Constant() {r=1}
   ) {materialFunction=&RemapValueRange}.Result
  ) {constB=0.6}
  n69 = Add(
   A=Multiply(A=MaterialFunctionCall(Float3=@n66) {materialFunction=&BreakOutFloat3Components}.R) {constB=0.5}
   B=@n68
  )
  n70 = CurveAtlasRowParameter(CurveTime=QualitySwitch(Default=@n69, Low=@n68, High=@n68, Medium=@n68, Epic=@n69)) {
   atlas=&CA_Sideways_Bubble
   curve=/SidewaysGameplay/Environmental/SidewaysEncounter/Effects/Curves/C_Sidways_Bubble_Orange
   defaultValue=6
   parameterName=CurveAtlas_Color
  }
  Reroute(None=@n70)
  ScalarParameter() {parameterName=BubbleMask}
  VertexNormalWS()
 ]
}
