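# Niagara system for the Sideways encounter that draws a ring of meshes and purple surface smoke around its owner, sized by user sphere parameters.
# A GPU burst of meshes is repositioned and aligned on a ring each update, following SphereRadius times SphereScale, with BubbleMask fed to materials.
# Rising GPU smoke sprites spawn at SphereRadius with SphereScale as probability, scaling alpha and size by camera distance.
# Landscape-sampling modules are present but disabled.
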NiagaraSystem /SidewaysGameplay/Environmental/SidewaysEncounter/Effects/NS_SideWays_SurfaceSmoke_SM
dependencies {
 usedDataInterfaces [
  &NiagaraDataInterfaceGBuffer
  &NiagaraDataInterfaceCurve
  &NiagaraDataInterfaceColorCurve
  &NiagaraDataInterfaceCamera
 ]
 usedDynamicInputs [
  &MakeFloatFromVector
  /Game/Effects/Niagara/DynamicInputs/Math/FN_Fit
  &ReturnCameraVector
 ]
 usedModules [
  &self:SystemSpawnScript.SystemScriptSource.SystemScriptGraph.NiagaraNodeAssignment_0.SetVariables_#1
  &SystemState
  &self:SystemSpawnScript.SystemScriptSource.SystemScriptGraph.NiagaraNodeAssignment_1.SetVariables_#2
 ]
}
emitters {
 Sample_Landscape {
  simTarget=GPUComputeSim
  emitterUpdateScript [
   &EmitterState(Inactive Response=::NewEnumerator2, Life Cycle Mode=::NewEnumerator1, Loop Behavior=::NewEnumerator1, Loop Duration=4.5, UseLoopDelay=false)
   &SpawnBurst_Instantaneous(Spawn Count=32, Use Spawn Probability=true)
  ]
  particleSpawnScript [
   &InitializeParticle(
    Color=<rgba 0 1 0.5673 1>
    Position Mode=::NewEnumerator1
    Sprite Size Mode=::NewEnumerator3
    Uniform Sprite Size=25
    UsePositionOffset=false
    Write Material Random=true
   )
   &TorusLocation(
    Align Torus With Axis=false
    Handle Radius=0
    Large Radius:NiagaraFloat=&Multiply_Float(A=$User.SphereRadius, B=$User.SphereScale)
    MaskSpawnGroup=false
    Offset=<0 0 -200>
    Torus Distribution Mode:ENiagaraTorusDistributionMode=::NewEnumerator1
    Torus Origin:NiagaraPosition=$Engine.Owner.Position
    U Position:NiagaraFloat=&Multiply_Float(A=&ReturnNormalizedExecIndex(), B=1)
   )
   /SetVariables() {enabled=false, isSetParameters=true, module=SetVariables_#3}
   &self:Sample_Landscape() {enabled=false, module=Sample_Landscape}
   &self:Align_Mesh(
    Align Direction:Vector3f=&Multiply_VectorByFloat(Vector=$Output.TorusLocation.TorusNormal, Float=-1)
    Up Vector:Vector3f=&Lerp_Vector(A=<0 0 1>, B=$Particles.Landscape Normal, Alpha=1)
   ) {enabled=false, module=Align_Mesh001}
   &self:Align_Mesh(
    Align Direction:Vector3f=&Multiply_VectorByFloat(Vector=$Output.TorusLocation.TorusNormal, Float=1)
    Up Vector:Vector3f=&Lerp_Vector(A=<0 0 1>, B=<0 0 1>, Alpha=1)
   ) {enabled=false, module=Align_Mesh}
   &self:PosValues() {enabled=false, module=PosValues}
   &self:ScratchModule_03() {enabled=false, module=ScratchModule_03}
  ]
  particleUpdateScript [
   &ParticleState(Kill Particles When Lifetime Has Elapsed=false)
   &TorusLocation(
    Align Torus With Axis=false
    Handle Radius=0
    Large Radius:NiagaraFloat=&Multiply_Float(A=$User.SphereRadius, B=$User.SphereScale)
    MaskSpawnGroup=false
    Offset=<0 0 -200>
    Torus Distribution Mode:ENiagaraTorusDistributionMode=::NewEnumerator1
    Torus Origin:NiagaraPosition=$Engine.Owner.Position
    U Position:NiagaraFloat=&Multiply_Float(A=&ReturnNormalizedExecIndex(), B=1)
   ) {module=TorusLocation002}
   &self:Align_Mesh(
    Align Direction:Vector3f=&Multiply_VectorByFloat(Vector=$Output.TorusLocation002.TorusNormal, Float=1)
    Up Vector:Vector3f=&Lerp_Vector(A=<0 0 1>, B=<0 0 1>, Alpha=1)
   ) {module=Align_Mesh002}
   &DynamicMaterialParameters(
    Index 3 Param 1=$User.BubbleMask
    Index 3 Param 3=&RandomRangeFloat(Minimum=0, Maximum=1)
    Index 3 Param 4=&Multiply_Float(A=$User.SphereRadius, B=$User.SphereScale)
    Module.Param0WriteEnabled3=true
    Module.Param1WriteEnabled3=false
    Module.Param2WriteEnabled3=true
    Module.Param3WriteEnabled3=true
    Write Parameter Index 0=false
    Write Parameter Index 3=true
   )
   /SetVariables(Particles.MeshScale:Vector3f=&Multiply_VectorByFloat(Vector=<0.3 0.75 0.75>, Float=$User.SphereScale)) {isSetParameters=true}
  ]
  rendererClasses [&NiagaraMeshRendererProperties]
  renderers [
   {rendererClass=&NiagaraMeshRendererProperties, rendererIndex=0}
   {rendererClass=&NiagaraMeshRendererProperties, rendererIndex=1}
  ]
 }
 Smoke_dark_surface001 {
  simTarget=GPUComputeSim
  emitterUpdateScript [
   &EmitterState()
   &SpawnRate(Spawn Probability=$User.SphereScale, SpawnRate=60, Use Spawn Probability=true)
  ]
  particleSpawnScript [
   /SetVariables(Particles.SpriteRotationRate=&RandomRangeFloat(Minimum=-0.35, Maximum=0.35, Evaluation Type=::NewEnumerator1)) {isSetParameters=true}
   /SetVariables() {isSetParameters=true, module=SetVariables_#3}
   &InitializeParticle(
    Color=&RandomRangeLinearColor(Minimum=<rgba 0.1778 0 0.1912 1>, Maximum=<rgba 1 0 1 1>)
    Lifetime Max=6
    Lifetime Min=3
    Lifetime Mode=::NewEnumerator1
    Sprite Rotation Mode=::NewEnumerator2
    Sprite Size Mode=::NewEnumerator1
    Uniform Sprite Size Max=1200
    Uniform Sprite Size Min=400
    UsePositionOffset=false
    Write Material Random=true
   ) {module=InitializeParticle001}
   /Game/Effects/Niagara/Modules/Spawn/FN_Spawn_GBuffer_Color() {enabled=false}
   &SphereLocation(
    Hemisphere Z=true
    MaskSpawnGroup=false
    Non Uniform Scale=<1 1 0>
    Override Seed=false
    Sphere Radius=5000
    Surface Only=true
    UseNonUniformScale=true
    UseOffset=false
    UseSphereOrientation=false
   ) {enabled=false}
   &TorusLocation(
    Align Torus With Axis=false
    Handle Radius=0
    Large Radius:NiagaraFloat=$User.SphereRadius
    MaskSpawnGroup=false
    Module.Override Randomness=false
   )
   &self:Sample_Landscape() {enabled=false, module=Sample_Landscape}
   &DynamicMaterialParameters(
    Index 0 Param 1=&RandomRangeFloat(Minimum=0, Maximum=1, Evaluation Type=::NewEnumerator1)
    Index 0 Param 2=&RandomRangeFloat(Minimum=0, Maximum=1, Evaluation Type=::NewEnumerator1)
    Module.Param0WriteEnabled=true
    Module.Param1WriteEnabled=true
    Module.Param2WriteEnabled=true
    Module.Param3WriteEnabled=true
   )
   &CameraOffset(Camera Offset Amount=&RandomRangeFloat(Minimum=100, Maximum=500, Evaluation Type=::NewEnumerator1))
   &AddVelocity(Velocity=&RandomRangeVector(Minimum=<0 0 25>, Maximum=<0 0 50>))
   &self:ScratchModule_01(OffsetZ=75) {enabled=false, module=ScratchModule_01}
   &self:Angular_Coordinates() {enabled=false, module=Angular_Coordinates}
   &self:Lifetime_by_Angular(Angular:NiagaraFloat=&FloatFromCurve(
     FloatCurve=DataInterface() {propertyValues {Curve {keys [<at 0 0>,<at 0.5 1>]}}}
     CurveIndex=&ModuloFloat(A=&Multiply_Float(A=$Particles.Angular, B=8), B=1)
    )) {enabled=false, module=Lifetime_by_Angular}
  ]
  particleUpdateScript [
   &self:Sample_Landscape_UPdate(Bias:NiagaraFloat=$Particles.Terrain Offset, New Landscape Sample:NiagaraDataInterfaceLandscape=$User.Landscape) {enabled=false, module=Sample_Landscape_UPdate}
   &GravityForce(Gravity=&Multiply_VectorByFloat(
     Vector=<0 0 125>
     Float=&FloatFromCurve(FloatCurve=DataInterface() {propertyValues {Curve {keys [<at 0 0>,<at 1 1>]}}})
    )) {module=GravityForce001}
   &ParticleState()
   &Drag(Drag=0, Use Linear Drag=true, Use Rotational Drag=false)
   &self:AlignToCenter() {module=AlignToCenter}
   &SolveForcesAndVelocity(Clamp Velocity=false, Limit Acceleration=false)
   &SpriteRotationRate(Delta Time:NiagaraFloat=$Particles.NormalizedAge, Rotation Rate=$Particles.SpriteRotationRate) {enabled=false}
   &ScaleColor(
    Scale Alpha=&Multiply_Float(
     A=&FloatFromCurve(
      FloatCurve=DataInterface() {propertyValues {Curve {keys [
          {interpMode=RCIM_Cubic, tangentMode=RCTM_User, time=0, value=0}
          {interpMode=RCIM_Cubic, tangentMode=RCTM_User, time=0.5, value=1}
          {interpMode=RCIM_Cubic, tangentMode=RCTM_User, time=1, value=0}
         ]}}}
      Scale Curve=25
     )
     B=&FloatFromCurve(
      FloatCurve=DataInterface() {propertyValues {Curve {keys [
          {arriveTangent=2.6303, interpMode=RCIM_Cubic, leaveTangent=2.6303, tangentMode=RCTM_User, time=0, value=0}
          {interpMode=RCIM_Cubic, tangentMode=RCTM_User, time=0.25, value=1}
          <at 0.9 1>
          <at 1 0 RCIM_Cubic>
         ]}}}
      Scale Curve=&Multiply_Float(
       A=&NormalizeDistanceRange(
        Start Position:NiagaraPosition=$Particles.Position
        End Position:NiagaraPosition=Linked() {linkedVariable {name=System.CameraPostion, type=Vector3f}}
        Distance:NiagaraFloat=25000
        Invert Normalized Range:NiagaraBool=false
       )
       B=0.25
      )
     )
    )
    Scale RGB=&Multiply_VectorByFloat(
     Vector=&MakeVectorFromLinearColorRGB(Color=&MakeLinearColorFromVectorAndFloat(
       "Vector (RGB)"=&MakeVectorFromLinearColorRGB(Color=&ColorFromCurve(ColorCurve=DataInterface() {propertyValues {
           AlphaCurve {keys [<at 0 1 RCIM_Cubic>,<at 0.9977 1 RCIM_Cubic>]}
           BlueCurve {keys [<at 0.027 0.3372 RCIM_Cubic>,<at 0.2263 0>,<at 0.4153 2>,<at 0.6291 0>,<at 0.8083 1.9736>]}
           GreenCurve {keys [<at 0.027 0 RCIM_Cubic>,<at 0.2263 0.3189>,<at 0.4153 0.312>,<at 0.6291 0.0903>,<at 0.8083 0>]}
           RedCurve {keys [<at 0.027 0.0686 RCIM_Cubic>,<at 0.2263 1>,<at 0.4153 0>,<at 0.6291 0.5>,<at 0.8083 3>]}
          }}))
       "Float (Alpha)"=&FloatFromCurve(FloatCurve=DataInterface() {propertyValues {Curve {keys [
            {interpMode=RCIM_Cubic, tangentMode=RCTM_User, time=0, value=1}
            {interpMode=RCIM_Cubic, tangentMode=RCTM_User, time=0.5, value=1}
            {interpMode=RCIM_Cubic, tangentMode=RCTM_User, time=1, value=1}
           ]}}})
      ))
     Float=1
    )
    ScaleA=true
    ScaleRGB=true
   ) {enabled=false}
   &ScaleColor(
    Scale Alpha=&Multiply_Float(
     A=&FloatFromCurve(FloatCurve=DataInterface() {propertyValues {Curve {keys [
          {interpMode=RCIM_Cubic, tangentMode=RCTM_User, time=0, value=0}
          {interpMode=RCIM_Cubic, tangentMode=RCTM_User, time=0.5, value=1}
          {interpMode=RCIM_Cubic, tangentMode=RCTM_User, time=1, value=0}
         ]}}})
     B=&FloatFromCurve(
      FloatCurve=DataInterface() {propertyValues {Curve {keys [
          {arriveTangent=2.6303, interpMode=RCIM_Cubic, leaveTangent=2.6303, tangentMode=RCTM_User, time=0.1, value=0}
          {interpMode=RCIM_Cubic, tangentMode=RCTM_User, time=0.15, value=1}
          <at 0.9 1>
          <at 1 0 RCIM_Cubic>
         ]}}}
      Scale Curve=&Multiply_Float(
       A=&NormalizeDistanceRange(
        Start Position:NiagaraPosition=$Particles.Position
        End Position:NiagaraPosition=Linked() {linkedVariable {name=System.CameraPostion, type=Vector3f}}
        Distance:NiagaraFloat=1500
        Invert Normalized Range:NiagaraBool=false
       )
       B=1
      )
     )
    )
    Scale RGB=&Multiply_VectorByFloat(
     Vector=&MakeVectorFromLinearColorRGB(Color=&MakeLinearColorFromVectorAndFloat(
       "Vector (RGB)"=&MakeVectorFromLinearColorRGB(Color=&ColorFromCurve(ColorCurve=DataInterface() {propertyValues {
           AlphaCurve {keys [<at 0 1 RCIM_Cubic>,<at 0.9977 1 RCIM_Cubic>]}
           BlueCurve {keys [<at 0.027 0.1542 RCIM_Cubic>,<at 0.362 2>,<at 0.7496 1>]}
           GreenCurve {keys [<at 0.027 0 RCIM_Cubic>,<at 0.362 0.0347>,<at 0.7496 0>]}
           RedCurve {keys [<at 0.027 0.3372 RCIM_Cubic>,<at 0.362 0.5>,<at 0.7496 0.6272>]}
          }}))
       "Float (Alpha)"=&FloatFromCurve(
        FloatCurve=DataInterface() {propertyValues {Curve {keys [
            {interpMode=RCIM_Cubic, tangentMode=RCTM_User, time=0, value=1}
            {interpMode=RCIM_Cubic, tangentMode=RCTM_User, time=0.5, value=1}
            {interpMode=RCIM_Cubic, tangentMode=RCTM_User, time=1, value=1}
           ]}}}
        Scale Curve=3
       )
      ))
     Float=1
    )
    ScaleA=true
    ScaleRGB=true
   ) {module=ScaleColor001}
   &ScaleSpriteSize(Scale Factor=&Vector2DFromFloat(Value=&FloatFromCurve(
      FloatCurve=DataInterface() {propertyValues {Curve {keys [
          {interpMode=RCIM_Cubic, tangentMode=RCTM_User, time=0, value=0}
          {arriveTangent=2.6303, interpMode=RCIM_Cubic, leaveTangent=2.6303, tangentMode=RCTM_User, time=0.1, value=0}
          <at 0.9 1>
          <at 1 0 RCIM_Cubic>
         ]}}}
      Scale Curve=&Multiply_Float(
       A=&NormalizeDistanceRange(
        Start Position:NiagaraPosition=$Particles.Position
        End Position:NiagaraPosition=Linked() {linkedVariable {name=System.CameraPostion, type=Vector3f}}
        Distance:NiagaraFloat=15000
        Invert Normalized Range:NiagaraBool=false
       )
       B=5
      )
     ))) {module=ScaleSpriteSize001}
   &DynamicMaterialParameters(
    Index 0 Param 1=&RandomRangeFloat(Minimum=0, Maximum=1)
    Index 0 Param 2=&RandomRangeFloat(Minimum=0, Maximum=1)
    Index 1 Param 1=&RandomRangeFloat(Minimum=0.05, Maximum=0.15)
    Index 1 Param 2=&RandomRangeFloat(Minimum=0.05, Maximum=0.1)
    Index 1 Param 3=&RandomRangeFloat(Minimum=-0.15, Maximum=0.1)
    Index 1 Param 4=&RandomRangeFloat(Minimum=-0.15, Maximum=0.25)
    Module.Param0WriteEnabled=true
    Module.Param0WriteEnabled1=true
    Module.Param1WriteEnabled=true
    Module.Param1WriteEnabled1=true
    Module.Param2WriteEnabled=true
    Module.Param2WriteEnabled1=true
    Module.Param3WriteEnabled=true
    Module.Param3WriteEnabled1=true
    Write Parameter Index 1=true
   ) {module=DynamicMaterialParameters001}
   &DynamicMaterialParameters(
    Index 3 Param 3=&RandomRangeFloat(Minimum=0, Maximum=1)
    Index 3 Param 4=$User.SphereRadius
    Module.Param0WriteEnabled3=false
    Module.Param1WriteEnabled3=false
    Module.Param2WriteEnabled3=true
    Module.Param3WriteEnabled3=true
    Write Parameter Index 0=false
    Write Parameter Index 3=true
   ) {module=DynamicMaterialParameters002}
  ]
  rendererClasses [&NiagaraSpriteRendererProperties]
  renderers [{rendererClass=&NiagaraSpriteRendererProperties}]
 }
}
summary {
 systemName=NS_SideWays_SurfaceSmoke_SM
 emitters [
  {emitterName=Smoke_dark_surface001, simTarget=GPUComputeSim, rendererClasses [&NiagaraSpriteRendererProperties]}
  {emitterName=Sample_Landscape, simTarget=GPUComputeSim, rendererClasses [&NiagaraMeshRendererProperties]}
 ]
 userVariables [
  User.BoxSize {type=Vector3f, description="", defaultValue {struct=Vector3f, value=<1 1 1>}}
  User.BubbleMask {type=NiagaraFloat, defaultValue=0, description=""}
  User.Landscape {
   type=NiagaraDataInterfaceLandscape
   description=""
   defaultValue {
    struct=NiagaraExt_VariableValue_DataInterface
    value {
     dataInterface=&self:NiagaraDataInterfaceLandscape_0
     dataInterfaceClass=&NiagaraDataInterfaceLandscape
    }
   }
  }
  User.SphereRadius {type=NiagaraFloat, defaultValue=2500, description=""}
  User.SphereScale {type=NiagaraFloat, defaultValue=0.01, description=""}
 ]
}
