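# Niagara GPU effect for a small in-world Sideways encounter volume, layering color-cycling mesh flow, sparks and smoke.
# Flow, Pop_Sparks and Smoke fill a hemisphere driven by User.SphereSize and User.SpawnRateScale.
# Pop_Sparks001 instead spawns at a fixed rate in a cylinder, samples the landscape for height, and is pushed from a camera-relative point by negative attraction.
# Camera distance scales spark size plus smoke size and opacity.
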NiagaraSystem /SidewaysGameplay/Environmental/SidewaysEncounter/Effects/NS_SideWays_Volume_InWorld_Small
dependencies {
 usedDataInterfaces [&NiagaraDataInterfaceCurve, &NiagaraDataInterfaceColorCurve, &NiagaraDataInterfaceCamera]
 usedDynamicInputs [/Game/Effects/Niagara/DynamicInputs/Math/FN_Fit]
 usedModules [
  &self:SystemSpawnScript.SystemScriptSource.SystemScriptGraph.NiagaraNodeAssignment_0.SetVariables_#1
  &SystemState
  &self:SystemSpawnScript.SystemScriptSource.SystemScriptGraph.NiagaraNodeAssignment_1.SetVariables_#2
 ]
}
emitters {
 Flow {
  simTarget=GPUComputeSim
  emitterUpdateScript [
   &EmitterState(Life Cycle Mode=::NewEnumerator1, Loop Behavior=::NewEnumerator1, Loop Duration Mode=::NewEnumerator1, UseLoopDelay=false)
   &SpawnRate(SpawnRate=&Multiply_Float(A=5, B=$User.SpawnRateScale), Use Spawn Probability=false)
  ]
  particleSpawnScript [
   &InitializeParticle(
    Color=&RandomRangeLinearColor(Minimum=<rgba 0 0 0 0>, Maximum=<rgba 0 0 0 1>)
    Lifetime Max=4
    Lifetime Min=3
    Lifetime Mode=::NewEnumerator1
    Mesh Scale Mode=::NewEnumerator1
    Mesh Uniform Scale Max=8
    Mesh Uniform Scale Min=4
    Sprite Rotation Mode=::NewEnumerator2
    Sprite Size Mode=::NewEnumerator1
    Uniform Sprite Size Max=2500
    Uniform Sprite Size Min=2500
    UsePositionOffset=false
    Write Material Random=true
   ) {module=InitializeParticle001}
   &SphereLocation(
    Hemisphere Z=true
    MaskSpawnGroup=false
    Non Uniform Scale=<1 1 0.8>
    Offset=<0 0 500>
    Override Seed=false
    Sphere Radius=$User.SphereSize
    Surface Only=false
    UseNonUniformScale=true
    UseOffset=true
    UseSphereOrientation=false
   )
   &InitialMeshOrientation(Orientation Axis=<0 0 1>, Orientation Vector=<0 0 1>, Rotation=<0 0.25 0>, Use Orientation Vector=true, Use Rotation Vector=true) {module=InitialMeshOrientation001}
   &InitialMeshOrientation(
    Rotation=&MakeCustomVectorFromBool(Bool=true, True Vector=<0 -0.05 -0.05>, False Vector=<0.5 0.05 0.05>)
    Use Orientation Vector=true
    Use Rotation Vector=true
   )
   &DynamicMaterialParameters(
    Index 0 Param 1=&RandomRangeFloat(Minimum=0, Maximum=1)
    Index 0 Param 2=&RandomRangeFloat(Minimum=0, Maximum=0.1)
    Index 0 Param 3=&RandomRangeFloat(Minimum=0.1, Maximum=0.2)
    Index 1 Param 1=&RandomRangeFloat(Minimum=0, Maximum=0.1)
    Index 1 Param 2=&RandomRangeFloat(Minimum=0.05, Maximum=0.1)
    Module.Param0WriteEnabled=true
    Module.Param0WriteEnabled1=true
    Module.Param1WriteEnabled=true
    Module.Param1WriteEnabled1=true
    Module.Param2WriteEnabled=true
    Module.Param2WriteEnabled1=true
    Module.Param3WriteEnabled=true
    Module.Param3WriteEnabled1=true
    Write Parameter Index 1=true
   )
  ]
  particleUpdateScript [
   &GravityForce(Gravity=&Multiply_VectorByFloat(
     Vector=<0 100 100>
     Float=&FloatFromCurve(FloatCurve=DataInterface() {propertyValues {Curve {keys [<at 0 0>,<at 1 1>]}}})
    )) {module=GravityForce001}
   &VortexForce(UseOriginOffset=false, Vortex Force Amount=0)
   &SolveForcesAndVelocity(Clamp Velocity=false, Limit Acceleration=false)
   &ParticleState()
   &ScaleColor(
    Scale Alpha=&FNTimeOfDayScalar(Night Scale=0.2, Morning Scale=0.2, Day Scale=0.2, Evening Scale=0.2)
    Scale RGB=<100 100 100>
    ScaleA=true
    ScaleRGB=true
   ) {enabled=false}
   &ScaleColor(
    Scale Alpha=&FloatFromCurve(
     FloatCurve=DataInterface() {propertyValues {Curve {keys [
         {interpMode=RCIM_Cubic, tangentMode=RCTM_User, time=0, value=0}
         {interpMode=RCIM_Cubic, tangentMode=RCTM_User, time=0.5, value=1}
         {interpMode=RCIM_Cubic, tangentMode=RCTM_User, time=1, value=0}
        ]}}}
     Scale Curve=5
    )
    Scale RGB=&Multiply_VectorByFloat(
     Vector=&MakeVectorFromLinearColorRGB(Color=&MakeLinearColorFromVectorAndFloat(
       "Vector (RGB)"=&MakeVectorFromLinearColorRGB(Color=&ColorFromCurve(ColorCurve=DataInterface() {propertyValues {
           AlphaCurve {keys [<at 0 1 RCIM_Cubic>,<at 0.9977 1 RCIM_Cubic>]}
           BlueCurve {keys [<at 0.027 0.3372 RCIM_Cubic>,<at 0.2263 0>,<at 0.4153 1>,<at 0.6291 0>,<at 0.8083 0.6588>]}
           GreenCurve {keys [<at 0.027 0 RCIM_Cubic>,<at 0.2263 0.3189>,<at 0.4153 0.1529>,<at 0.6291 0.1804>,<at 0.8083 0>]}
           RedCurve {keys [<at 0.027 0.0686 RCIM_Cubic>,<at 0.2263 1>,<at 0.4153 0>,<at 0.6291 1>,<at 0.8083 1>]}
          }}))
       "Float (Alpha)"=&FloatFromCurve(FloatCurve=DataInterface() {propertyValues {Curve {keys [
            {interpMode=RCIM_Cubic, tangentMode=RCTM_User, time=0, value=1}
            {interpMode=RCIM_Cubic, tangentMode=RCTM_User, time=0.5, value=1}
            {interpMode=RCIM_Cubic, tangentMode=RCTM_User, time=1, value=1}
           ]}}})
      ))
     Float=100
    )
    ScaleA=true
    ScaleRGB=true
   ) {module=ScaleColor001}
   &DynamicMaterialParameters(
    Index 0 Param 1=&RandomRangeFloat(Minimum=0, Maximum=1)
    Index 0 Param 3=0.02
    Index 1 Param 2=0.1
    Module.Param0WriteEnabled=true
    Module.Param0WriteEnabled1=true
    Module.Param1WriteEnabled=true
    Module.Param1WriteEnabled1=true
    Module.Param2WriteEnabled=true
    Module.Param2WriteEnabled1=true
    Module.Param3WriteEnabled=true
    Module.Param3WriteEnabled1=true
    Write Parameter Index 1=true
   ) {module=DynamicMaterialParameters001}
  ]
  rendererClasses [&NiagaraMeshRendererProperties]
  renderers [{rendererClass=&NiagaraMeshRendererProperties, rendererIndex=0}]
 }
 Pop_Sparks {
  simTarget=GPUComputeSim
  emitterUpdateScript [
   &EmitterState()
   &SpawnRate(SpawnRate=&Multiply_Float(A=500, B=$User.SpawnRateScale), Use Spawn Probability=true)
  ]
  particleSpawnScript [
   &InitializeParticle(
    AdjustAlpha=true
    AdjustHue=true
    AdjustSaturation=false
    AdjustValue=true
    Alpha Scale Range:Vector2f=<0.1 0.5>
    Color=<rgba 1 0 1 1>
    Color Mode=::NewEnumerator3
    Hue Shift Range:Vector2f=<0.15 0.25>
    Lifetime Max=5
    Lifetime Min=3
    Lifetime Mode=::NewEnumerator1
    Sprite Rotation Mode=::NewEnumerator2
    Sprite Size Mode=::NewEnumerator1
    Uniform Sprite Size Max=50
    Uniform Sprite Size Min=15
    UsePositionOffset=false
    Value Range:Vector2f=<0.02 0.05>
    Write Material Random=true
   ) {module=InitializeParticle001}
   /SetVariables(Particles.Terrain Offset=5) {isSetParameters=true, module=SetVariables_#3}
   &SphereLocation(
    Hemisphere Z=true
    MaskSpawnGroup=false
    Override Seed=false
    Sphere Radius=$User.SphereSize
    Surface Only=false
    UseNonUniformScale=false
    UseOffset=false
    UseSphereOrientation=false
   )
   &self:Sample_Landscape() {enabled=false, module=Sample_Landscape}
   /SetVariables(Particles.Position=&Add_Vector(
     A=Linked() {linkedVariable {name=Particles.Position, type=NiagaraPosition}}
     B=&MakeVector(X=0, Y=0, Z=$Particles.Terrain Offset)
    )) {isSetParameters=true, module=SetVariables_#4}
   /Niagara/Modules/Update/Forces/CurlNoiseForce(Curl Noise Cone Mask Angle=50, Mask Curl Noise=true, Noise Frequency=150, Noise Strength=25, Pan Noise=true, Pan Noise Field=<0 0 -5>) {module=CurlNoiseForce001}
   &ApplyInitialForces()
   &CameraOffset(Camera Offset Amount=25)
  ]
  particleUpdateScript [
   &self:Sample_Landscape_UPdate(Bias:NiagaraFloat=$Particles.Terrain Offset, New Landscape Sample:NiagaraDataInterfaceLandscape=$User.Landscape) {enabled=false, module=Sample_Landscape_UPdate}
   &ParticleState()
   &Drag(
    Drag=&Multiply_Float(
     A=&UniformRangedFloat(Minimum=0, Maximum=0)
     B=&FloatFromCurve(FloatCurve=DataInterface() {propertyValues {Curve {keys [<at 0.1 1>,<at 0.25 0.1>]}}})
    )
    Use Linear Drag=true
    Use Rotational Drag=false
   )
   &GravityForce(Gravity=&Multiply_VectorByFloat(
     Vector=<0 150 150>
     Float=&FloatFromCurve(FloatCurve=DataInterface() {propertyValues {Curve {keys [<at 0 0>,<at 1 1>]}}})
    ))
   &CurlNoiseForce(
    Curl Noise Cone Mask Angle=35
    Mask Curl Noise=true
    Noise Frequency=250
    Noise Quality / Cost=::NewEnumerator0
    Noise Strength=&FloatFromCurve(FloatCurve=DataInterface() {propertyValues {Curve {keys [<at 0 1>,<at 1 0>]}}}, Scale Curve=1000)
    Pan Noise=true
    Pan Noise Field=<0 10 0>
   )
   &SolveForcesAndVelocity(Clamp Velocity=false, Limit Acceleration=false)
   &ScaleColor(
    Scale Alpha=&FloatFromCurve(FloatCurve=DataInterface() {propertyValues {Curve {keys [
         {interpMode=RCIM_Cubic, tangentMode=RCTM_User, time=0, value=0}
         {interpMode=RCIM_Cubic, tangentMode=RCTM_User, time=0.15, value=1}
         {interpMode=RCIM_Cubic, tangentMode=RCTM_User, time=0.3, value=2.875}
         {interpMode=RCIM_Cubic, tangentMode=RCTM_User, time=0.4, value=1.25}
         {interpMode=RCIM_Cubic, tangentMode=RCTM_User, time=0.5, value=1.2}
         {interpMode=RCIM_Cubic, tangentMode=RCTM_User, time=0.6, value=2.125}
         {interpMode=RCIM_Cubic, tangentMode=RCTM_User, time=0.7, value=1.2}
         {interpMode=RCIM_Cubic, tangentMode=RCTM_User, time=0.8, value=1}
         {interpMode=RCIM_Cubic, tangentMode=RCTM_User, time=1, value=0}
        ]}}})
    Scale RGB=&Multiply_VectorByFloat(
     Vector=&Lerp_Vector(
      A=<1 1 1>
      B=&MakeVectorFromLinearColorRGB(Color=&MakeLinearColorFromVectorAndFloat(
        "Vector (RGB)"=&MakeVectorFromLinearColorRGB(Color=&ColorFromCurve(ColorCurve=DataInterface() {propertyValues {
            AlphaCurve {keys [<at 0 1 RCIM_Cubic>,<at 0.9977 1 RCIM_Cubic>]}
            BlueCurve {keys [<at 0.027 0.3372 RCIM_Cubic>,<at 0.4507 0>,<at 0.5087 0.0214>,<at 0.5565 1>,<at 0.7691 1.5>]}
            GreenCurve {keys [<at 0.027 0 RCIM_Cubic>,<at 0.4507 1.5945>,<at 0.5087 0.4769>,<at 0.5565 0.0471>,<at 0.7691 0.234>]}
            RedCurve {keys [<at 0.027 0.0686 RCIM_Cubic>,<at 0.4507 5>,<at 0.5087 1>,<at 0.5565 0.7255>,<at 0.7691 0>]}
           }}))
        "Float (Alpha)"=&FloatFromCurve(FloatCurve=DataInterface() {propertyValues {Curve {keys [
             {interpMode=RCIM_Cubic, tangentMode=RCTM_User, time=0, value=1}
             {interpMode=RCIM_Cubic, tangentMode=RCTM_User, time=0.5, value=1}
             {interpMode=RCIM_Cubic, tangentMode=RCTM_User, time=1, value=1}
            ]}}})
       ))
      Alpha=&FloatFromCurve(FloatCurve=DataInterface() {propertyValues {Curve {keys [
           {interpMode=RCIM_Cubic, tangentMode=RCTM_User, time=0, value=1}
           {interpMode=RCIM_Cubic, tangentMode=RCTM_User, time=0.1, value=1}
           {interpMode=RCIM_Cubic, tangentMode=RCTM_User, time=1, value=1}
          ]}}})
     )
     Float=25
    )
    ScaleA=true
    ScaleRGB=true
   ) {enabled=false}
   &ScaleColor(
    Scale Alpha=&FloatFromCurve(
     FloatCurve=DataInterface() {propertyValues {Curve {keys [
         {interpMode=RCIM_Cubic, tangentMode=RCTM_User, time=0, value=0}
         {interpMode=RCIM_Cubic, tangentMode=RCTM_User, time=0.15, value=1}
         {interpMode=RCIM_Cubic, tangentMode=RCTM_User, time=0.3, value=2.875}
         {interpMode=RCIM_Cubic, tangentMode=RCTM_User, time=0.4, value=1.25}
         {interpMode=RCIM_Cubic, tangentMode=RCTM_User, time=0.5, value=1.2}
         {interpMode=RCIM_Cubic, tangentMode=RCTM_User, time=0.6, value=2.125}
         {interpMode=RCIM_Cubic, tangentMode=RCTM_User, time=0.7, value=1.2}
         {interpMode=RCIM_Cubic, tangentMode=RCTM_User, time=0.8, value=1}
         {interpMode=RCIM_Cubic, tangentMode=RCTM_User, time=1, value=0}
        ]}}}
     Scale Curve=0.25
    )
    Scale RGB=&Multiply_VectorByFloat(
     Vector=&Lerp_Vector(
      A=<1 1 1>
      B=&MakeVectorFromLinearColorRGB(Color=&MakeLinearColorFromVectorAndFloat(
        "Vector (RGB)"=&MakeVectorFromLinearColorRGB(Color=&ColorFromCurve(ColorCurve=DataInterface() {propertyValues {
            AlphaCurve {keys [<at 0 1 RCIM_Cubic>,<at 0.9977 1 RCIM_Cubic>]}
            BlueCurve {keys [<at 0.027 0.3372 RCIM_Cubic>,<at 0.4507 0>,<at 0.5087 0.0214>,<at 0.5565 1>,<at 0.7691 1.5>]}
            GreenCurve {keys [<at 0.027 0 RCIM_Cubic>,<at 0.4507 1.5945>,<at 0.5087 0.4769>,<at 0.5565 0.0471>,<at 0.7691 0.234>]}
            RedCurve {keys [<at 0.027 0.0686 RCIM_Cubic>,<at 0.4507 5>,<at 0.5087 1>,<at 0.5565 0.7255>,<at 0.7691 0>]}
           }}))
        "Float (Alpha)"=&FloatFromCurve(FloatCurve=DataInterface() {propertyValues {Curve {keys [
             {interpMode=RCIM_Cubic, tangentMode=RCTM_User, time=0, value=1}
             {interpMode=RCIM_Cubic, tangentMode=RCTM_User, time=0.5, value=1}
             {interpMode=RCIM_Cubic, tangentMode=RCTM_User, time=1, value=1}
            ]}}})
       ))
      Alpha=&FloatFromCurve(FloatCurve=DataInterface() {propertyValues {Curve {keys [
           {interpMode=RCIM_Cubic, tangentMode=RCTM_User, time=0, value=1}
           {interpMode=RCIM_Cubic, tangentMode=RCTM_User, time=0.1, value=1}
           {interpMode=RCIM_Cubic, tangentMode=RCTM_User, time=1, value=1}
          ]}}})
     )
     Float=&Waveform(
      Global Phase=$Particles.Age
      Global Amplitude Scale=250
      "Waveform [1]"=::NewEnumerator0
      "[1] Frequency"=&RandomRangeFloat(Minimum=0.25, Maximum=1)
      "[1] Amplitude Min/Max"=<0 1>
      "[2] Phase"=$Emitter.Age
      "[3] Phase"=$Emitter.Age
     )
    )
    ScaleA=true
    ScaleRGB=true
   ) {module=ScaleColor001}
   &ScaleSpriteSize(Scale Factor=&Vector2DFromFloat(Value=&FloatFromCurve(
      FloatCurve=DataInterface() {propertyValues {Curve {keys [
          {arriveTangent=7.1136, interpMode=RCIM_Cubic, leaveTangent=7.1136, tangentMode=RCTM_User, time=0, value=0}
          {interpMode=RCIM_Cubic, tangentMode=RCTM_User, time=0.1, value=1.375}
          {interpMode=RCIM_Cubic, tangentMode=RCTM_User, time=0.7, value=1}
          {interpMode=RCIM_Cubic, tangentMode=RCTM_User, time=0.8, value=0}
          <at 1 0 RCIM_Cubic>
         ]}}}
      Scale Curve=&Multiply_Float(
       A=&NormalizeDistanceRange(
        Start Position:NiagaraPosition=$Particles.Position
        End Position:NiagaraPosition=Linked() {linkedVariable {name=System.CameraPostion, type=Vector3f}}
        Distance:NiagaraFloat=25000
        Invert Normalized Range:NiagaraBool=false
       )
       B=&Waveform(
        Global Phase=$Particles.Age
        Global Amplitude Scale=4
        "Waveform [1]"=::NewEnumerator0
        "[1] Frequency"=&RandomRangeFloat(Minimum=0.25, Maximum=1)
        "[1] Amplitude Min/Max"=<0 1>
        "[2] Phase"=$Emitter.Age
        "[3] Phase"=$Emitter.Age
       )
      )
     )))
   &ScaleSpriteSizeBySpeed(
    Max Scale Factor=<1 10>
    Sample Scale Factor By Curve=true
    Scale Factor Curve=DataInterface() {propertyValues {Curve {keys [<at 0.5 0>,<at 1 1>]}}}
   ) {enabled=false}
   &DynamicMaterialParameters(Module.Param0WriteEnabled=true, Module.Param1WriteEnabled=true, Module.Param2WriteEnabled=true, Module.Param3WriteEnabled=true) {enabled=false}
   /SetVariables() {isSetParameters=true, module=SetVariables_#5}
   /Niagara/Modules/Update/Color/ScaleColorBySpeed(RGB Scale Max=0.7, RGB Scale Min=1)
  ]
  rendererClasses [&NiagaraSpriteRendererProperties]
  renderers [{rendererClass=&NiagaraSpriteRendererProperties}]
 }
 Pop_Sparks001 {
  simTarget=GPUComputeSim
  emitterUpdateScript [
   /SetVariables(
    Emitter.CameraForward:Vector3f=&ReturnCameraVector(Camera Property=::NewEnumerator1, Coordinate Space=::Simulation, CameraQuery=$Emitter.CameraRead)
    Emitter.CameraPosition:Vector3f=&ReturnCameraVector(Camera Property=::NewEnumerator0, Coordinate Space=::World, CameraQuery=$Emitter.CameraRead)
    Emitter.CameraUp:Vector3f=&ReturnCameraVector(Camera Property=::NewEnumerator2, Coordinate Space=::Simulation, CameraQuery=$Emitter.CameraRead)
   ) {isSetParameters=true}
   &EmitterState()
   &SpawnRate(SpawnRate=&Multiply_Float(A=500, B=1), Use Spawn Probability=true)
  ]
  particleSpawnScript [
   &InitializeParticle(
    AdjustAlpha=true
    AdjustHue=true
    AdjustSaturation=false
    AdjustValue=true
    Alpha Scale Range:Vector2f=<0.1 0.5>
    Color=<rgba 1 0 1 1>
    Color Mode=::NewEnumerator3
    Hue Shift Range:Vector2f=<0.15 0.25>
    Lifetime Max=5
    Lifetime Min=3
    Lifetime Mode=::NewEnumerator1
    Sprite Rotation Mode=::NewEnumerator2
    Sprite Size Mode=::NewEnumerator1
    Uniform Sprite Size Max=40
    Uniform Sprite Size Min=25
    UsePositionOffset=false
    Value Range:Vector2f=<0.02 0.04>
    Write Material Random=true
   ) {module=InitializeParticle001}
   /SetVariables(Particles.Terrain Offset=5) {isSetParameters=true, module=SetVariables_#3}
   &SphereLocation(
    MaskSpawnGroup=false
    Non Uniform Scale=<1 1 0.25>
    Override Seed=false
    Sphere Origin:Vector3f=$Emitter.CameraPosition
    Sphere Radius=250
    Surface Only=true
    Surface Only Band Thickness=50
    UseNonUniformScale=true
    UseOffset=false
    UseSphereOrientation=false
   ) {enabled=false}
   &CylinderLocation(Cylinder Radius=4000, MaskSpawnGroup=false, Override Seed=false, UseOwnerScale=true)
   &self:Sample_Landscape() {module=Sample_Landscape}
   /SetVariables(Particles.Position=&Add_Vector(
     A=Linked() {linkedVariable {name=Particles.Position, type=NiagaraPosition}}
     B=&MakeVector(X=0, Y=0, Z=$Particles.Terrain Offset)
    )) {isSetParameters=true, module=SetVariables_#4}
   &CurlNoiseForce(Curl Noise Cone Mask Angle=50, Mask Curl Noise=true, Noise Frequency=150, Noise Strength=25, Pan Noise=true, Pan Noise Field=<0 0 -5>) {module=CurlNoiseForce001}
   &ApplyInitialForces()
   &CameraOffset(Camera Offset Amount=25)
   &KillParticlesInVolume(Invert Volume=true, Sphere Radius=5000)
  ]
  particleUpdateScript [
   &self:Sample_Landscape_UPdate(Bias:NiagaraFloat=$Particles.Terrain Offset, New Landscape Sample:NiagaraDataInterfaceLandscape=$User.Landscape) {module=Sample_Landscape_UPdate}
   &ParticleState()
   /Niagara/Modules/Update/Forces/PointAttractionForce(
    Attraction Radius=750
    AttractionStrength:NiagaraFloat=&Lerp_Float(A=-0.35, B=-0.35, Alpha=0.5)
    AttractorPosition:Vector3f=&Add_Vector(
     A=$Emitter.CameraPosition
     B=&Add_Vector(
      A=&Multiply_VectorByFloat(Vector=$Emitter.CameraUp, Float=-200)
      B=&Multiply_VectorByFloat(Vector=$Emitter.CameraForward, Float=-250)
     )
    )
    Falloff Exponent=0
    Module.Kill Within Radius=false
    Module.Use Falloff=false
   )
   &GravityForce(Gravity=&Multiply_VectorByFloat(
     Vector=<0 30 30>
     Float=&FloatFromCurve(FloatCurve=DataInterface() {propertyValues {Curve {keys [<at 0 0>,<at 1 1>]}}})
    ))
   &CurlNoiseForce(
    Curl Noise Cone Mask Angle=35
    Mask Curl Noise=true
    Noise Frequency=250
    Noise Quality / Cost=::NewEnumerator0
    Noise Strength=&FloatFromCurve(FloatCurve=DataInterface() {propertyValues {Curve {keys [<at 0 1>,<at 1 0>]}}}, Scale Curve=500)
    Pan Noise=true
    Pan Noise Field=<0 10 0>
   )
   &Drag(
    Drag=&Multiply_Float(
     A=&UniformRangedFloat(Minimum=0, Maximum=0)
     B=&FloatFromCurve(
      FloatCurve=DataInterface() {propertyValues {Curve {keys [<at 0.1 1>,<at 0.25 1>]}}}
      CurveIndex=&MakeFloatFromVector(Vector=$Particles.Velocity, Channel=::NewEnumerator2)
      Scale Curve=15
     )
    )
    Use Linear Drag=true
    Use Rotational Drag=false
   )
   &SolveForcesAndVelocity(Clamp Velocity=false, Limit Acceleration=false)
   &ScaleColor(
    Scale Alpha=&FloatFromCurve(FloatCurve=DataInterface() {propertyValues {Curve {keys [
         {interpMode=RCIM_Cubic, tangentMode=RCTM_User, time=0, value=0}
         {interpMode=RCIM_Cubic, tangentMode=RCTM_User, time=0.15, value=1}
         {interpMode=RCIM_Cubic, tangentMode=RCTM_User, time=0.3, value=2.875}
         {interpMode=RCIM_Cubic, tangentMode=RCTM_User, time=0.4, value=1.25}
         {interpMode=RCIM_Cubic, tangentMode=RCTM_User, time=0.5, value=1.2}
         {interpMode=RCIM_Cubic, tangentMode=RCTM_User, time=0.6, value=2.125}
         {interpMode=RCIM_Cubic, tangentMode=RCTM_User, time=0.7, value=1.2}
         {interpMode=RCIM_Cubic, tangentMode=RCTM_User, time=0.8, value=1}
         {interpMode=RCIM_Cubic, tangentMode=RCTM_User, time=1, value=0}
        ]}}})
    Scale RGB=&Multiply_VectorByFloat(
     Vector=&Lerp_Vector(
      A=<1 1 1>
      B=&MakeVectorFromLinearColorRGB(Color=&MakeLinearColorFromVectorAndFloat(
        "Vector (RGB)"=&MakeVectorFromLinearColorRGB(Color=&ColorFromCurve(ColorCurve=DataInterface() {propertyValues {
            AlphaCurve {keys [<at 0 1 RCIM_Cubic>,<at 0.9977 1 RCIM_Cubic>]}
            BlueCurve {keys [<at 0.027 0.3372 RCIM_Cubic>,<at 0.4507 0>,<at 0.7691 1.5>]}
            GreenCurve {keys [<at 0.027 0 RCIM_Cubic>,<at 0.4507 1.5945>,<at 0.7691 0.234>]}
            RedCurve {keys [<at 0.027 0.0686 RCIM_Cubic>,<at 0.4507 5>,<at 0.7691 0>]}
           }}))
        "Float (Alpha)"=&FloatFromCurve(FloatCurve=DataInterface() {propertyValues {Curve {keys [
             {interpMode=RCIM_Cubic, tangentMode=RCTM_User, time=0, value=1}
             {interpMode=RCIM_Cubic, tangentMode=RCTM_User, time=0.5, value=1}
             {interpMode=RCIM_Cubic, tangentMode=RCTM_User, time=1, value=1}
            ]}}})
       ))
      Alpha=&FloatFromCurve(FloatCurve=DataInterface() {propertyValues {Curve {keys [
           {interpMode=RCIM_Cubic, tangentMode=RCTM_User, time=0, value=1}
           {interpMode=RCIM_Cubic, tangentMode=RCTM_User, time=0.1, value=1}
           {interpMode=RCIM_Cubic, tangentMode=RCTM_User, time=1, value=1}
          ]}}})
     )
     Float=10
    )
    ScaleA=true
    ScaleRGB=true
   )
   &ScaleSpriteSize(Scale Factor=&Vector2DFromFloat(Value=&FloatFromCurve(
      FloatCurve=DataInterface() {propertyValues {Curve {keys [
          {arriveTangent=7.1136, interpMode=RCIM_Cubic, leaveTangent=7.1136, tangentMode=RCTM_User, time=0, value=0}
          {interpMode=RCIM_Cubic, tangentMode=RCTM_User, time=0.1, value=1.375}
          {interpMode=RCIM_Cubic, tangentMode=RCTM_User, time=0.7, value=1}
          {interpMode=RCIM_Cubic, tangentMode=RCTM_User, time=0.8, value=0}
          <at 1 0 RCIM_Cubic>
         ]}}}
      Scale Curve=&Multiply_Float(
       A=&NormalizeDistanceRange(
        Start Position:NiagaraPosition=$Particles.Position
        End Position:NiagaraPosition=Linked() {linkedVariable {name=System.CameraPostion, type=Vector3f}}
        Distance:NiagaraFloat=25000
        Invert Normalized Range:NiagaraBool=false
       )
       B=3
      )
     )))
   &ScaleSpriteSizeBySpeed(
    Max Scale Factor=<1 10>
    Sample Scale Factor By Curve=true
    Scale Factor Curve=DataInterface() {propertyValues {Curve {keys [
        {interpMode=RCIM_Cubic, tangentMode=RCTM_User, time=0.5, value=0}
        {interpMode=RCIM_Cubic, tangentMode=RCTM_User, time=0.7, value=0.85}
        {interpMode=RCIM_Cubic, tangentMode=RCTM_User, time=0.8, value=0.7}
        {interpMode=RCIM_Cubic, tangentMode=RCTM_User, time=0.9, value=0}
       ]}}}
   )
   &DynamicMaterialParameters(Module.Param0WriteEnabled=true, Module.Param1WriteEnabled=true, Module.Param2WriteEnabled=true, Module.Param3WriteEnabled=true) {enabled=false}
   /SetVariables() {isSetParameters=true, module=SetVariables_#5}
   &ScaleColorBySpeed(RGB Scale Min=1)
  ]
  rendererClasses [&NiagaraSpriteRendererProperties]
  renderers [{rendererClass=&NiagaraSpriteRendererProperties}]
 }
 Smoke {
  simTarget=GPUComputeSim
  emitterUpdateScript [
   &EmitterState()
   &SpawnRate(SpawnRate=&Multiply_Float(A=10, B=$User.SpawnRateScale), Use Spawn Probability=false)
  ]
  particleSpawnScript [
   /SetVariables(Particles.SpriteRotationRate=&RandomRangeFloat(Minimum=-0.35, Maximum=0.35, Evaluation Type=::NewEnumerator1)) {isSetParameters=true}
   &InitializeParticle(
    Color=&RandomRangeLinearColor(Minimum=<rgba 0.1912 0.0908 0.1144 1>, Maximum=<rgba 1 0 1 1>)
    Lifetime Max=4
    Lifetime Min=3
    Lifetime Mode=::NewEnumerator1
    Sprite Rotation Mode=::NewEnumerator2
    Sprite Size Mode=::NewEnumerator1
    Uniform Sprite Size Max=500
    Uniform Sprite Size Min=300
    UsePositionOffset=false
    Write Material Random=true
   ) {module=InitializeParticle001}
   &SphereLocation(
    Hemisphere Z=true
    MaskSpawnGroup=false
    Non Uniform Scale=<1 1 0>
    Override Seed=false
    Sphere Radius=$User.SphereSize
    Surface Only=false
    UseNonUniformScale=true
    UseOffset=false
    UseSphereOrientation=false
   )
   &DynamicMaterialParameters(
    Index 0 Param 1=&RandomRangeFloat(Minimum=0, Maximum=1, Evaluation Type=::NewEnumerator1)
    Index 0 Param 2=&RandomRangeFloat(Minimum=0, Maximum=1, Evaluation Type=::NewEnumerator1)
    Module.Param0WriteEnabled=true
    Module.Param1WriteEnabled=true
    Module.Param2WriteEnabled=true
    Module.Param3WriteEnabled=true
   )
   &CameraOffset(Camera Offset Amount=&RandomRangeFloat(Minimum=0, Maximum=50, Evaluation Type=::NewEnumerator1))
   &AddVelocity(Velocity=&RandomRangeVector(Minimum=<0 0 50>, Maximum=<0 0 100>))
  ]
  particleUpdateScript [
   &GravityForce(Gravity=&Multiply_VectorByFloat(
     Vector=<0 -100 100>
     Float=&FloatFromCurve(FloatCurve=DataInterface() {propertyValues {Curve {keys [<at 0 0>,<at 1 1>]}}})
    )) {module=GravityForce001}
   &ParticleState()
   &Drag(Drag=1.5, Use Linear Drag=true, Use Rotational Drag=false)
   &SolveForcesAndVelocity(Clamp Velocity=false, Limit Acceleration=false)
   &SpriteRotationRate(Delta Time:NiagaraFloat=$Particles.NormalizedAge, Rotation Rate=$Particles.SpriteRotationRate)
   &ScaleColor(
    Scale Alpha=&Multiply_Float(
     A=&FloatFromCurve(
      FloatCurve=DataInterface() {propertyValues {Curve {keys [
          {interpMode=RCIM_Cubic, tangentMode=RCTM_User, time=0, value=0}
          {interpMode=RCIM_Cubic, tangentMode=RCTM_User, time=0.5, value=1}
          {interpMode=RCIM_Cubic, tangentMode=RCTM_User, time=1, value=0}
         ]}}}
      Scale Curve=25
     )
     B=&FloatFromCurve(
      FloatCurve=DataInterface() {propertyValues {Curve {keys [
          {arriveTangent=2.6303, interpMode=RCIM_Cubic, leaveTangent=2.6303, tangentMode=RCTM_User, time=0, value=0}
          {interpMode=RCIM_Cubic, tangentMode=RCTM_User, time=0.25, value=1}
          <at 0.9 1>
          <at 1 0 RCIM_Cubic>
         ]}}}
      Scale Curve=&Multiply_Float(
       A=&NormalizeDistanceRange(
        Start Position:NiagaraPosition=$Particles.Position
        End Position:NiagaraPosition=Linked() {linkedVariable {name=System.CameraPostion, type=Vector3f}}
        Distance:NiagaraFloat=25000
        Invert Normalized Range:NiagaraBool=false
       )
       B=0.25
      )
     )
    )
    Scale RGB=&Multiply_VectorByFloat(
     Vector=&MakeVectorFromLinearColorRGB(Color=&MakeLinearColorFromVectorAndFloat(
       "Vector (RGB)"=&MakeVectorFromLinearColorRGB(Color=&ColorFromCurve(ColorCurve=DataInterface() {propertyValues {
           AlphaCurve {keys [<at 0 1 RCIM_Cubic>,<at 0.9977 1 RCIM_Cubic>]}
           BlueCurve {keys [<at 0.027 0.3372 RCIM_Cubic>,<at 0.2263 0>,<at 0.4153 2>,<at 0.6291 0>,<at 0.8083 1.9736>]}
           GreenCurve {keys [<at 0.027 0 RCIM_Cubic>,<at 0.2263 0.3189>,<at 0.4153 0.312>,<at 0.6291 0.0903>,<at 0.8083 0>]}
           RedCurve {keys [<at 0.027 0.0686 RCIM_Cubic>,<at 0.2263 1>,<at 0.4153 0>,<at 0.6291 0.5>,<at 0.8083 3>]}
          }}))
       "Float (Alpha)"=&FloatFromCurve(FloatCurve=DataInterface() {propertyValues {Curve {keys [
            {interpMode=RCIM_Cubic, tangentMode=RCTM_User, time=0, value=1}
            {interpMode=RCIM_Cubic, tangentMode=RCTM_User, time=0.5, value=1}
            {interpMode=RCIM_Cubic, tangentMode=RCTM_User, time=1, value=1}
           ]}}})
      ))
     Float=25
    )
    ScaleA=true
    ScaleRGB=true
   ) {enabled=false}
   &ScaleColor(
    Scale Alpha=&Multiply_Float(
     A=&FloatFromCurve(
      FloatCurve=DataInterface() {propertyValues {Curve {keys [
          {interpMode=RCIM_Cubic, tangentMode=RCTM_User, time=0, value=0}
          {interpMode=RCIM_Cubic, tangentMode=RCTM_User, time=0.5, value=1}
          {interpMode=RCIM_Cubic, tangentMode=RCTM_User, time=1, value=0}
         ]}}}
      Scale Curve=25
     )
     B=&FloatFromCurve(
      FloatCurve=DataInterface() {propertyValues {Curve {keys [
          {arriveTangent=2.6303, interpMode=RCIM_Cubic, leaveTangent=2.6303, tangentMode=RCTM_User, time=0, value=0}
          {interpMode=RCIM_Cubic, tangentMode=RCTM_User, time=0.25, value=1}
          <at 0.9 1>
          <at 1 0 RCIM_Cubic>
         ]}}}
      Scale Curve=&Multiply_Float(
       A=&NormalizeDistanceRange(
        Start Position:NiagaraPosition=$Particles.Position
        End Position:NiagaraPosition=Linked() {linkedVariable {name=System.CameraPostion, type=Vector3f}}
        Distance:NiagaraFloat=25000
        Invert Normalized Range:NiagaraBool=false
       )
       B=0.5
      )
     )
    )
    Scale RGB=&Multiply_VectorByFloat(
     Vector=&MakeVectorFromLinearColorRGB(Color=&MakeLinearColorFromVectorAndFloat(
       "Vector (RGB)"=&MakeVectorFromLinearColorRGB(Color=&ColorFromCurve(ColorCurve=DataInterface() {propertyValues {
           AlphaCurve {keys [<at 0 1 RCIM_Cubic>,<at 0.9977 1 RCIM_Cubic>]}
           BlueCurve {keys [<at 0.027 0.1542 RCIM_Cubic>,<at 0.2263 0>,<at 0.4153 0>,<at 0.6291 0>,<at 0.8083 1.0369>]}
           GreenCurve {keys [<at 0.027 0 RCIM_Cubic>,<at 0.2263 0.3189>,<at 0.4153 0.4827>,<at 0.6291 0.0903>,<at 0.8083 0>]}
           RedCurve {keys [<at 0.027 0.3372 RCIM_Cubic>,<at 0.2263 1>,<at 0.4153 2>,<at 0.6291 0.5>,<at 0.8083 3>]}
          }}))
       "Float (Alpha)"=&FloatFromCurve(FloatCurve=DataInterface() {propertyValues {Curve {keys [
            {interpMode=RCIM_Cubic, tangentMode=RCTM_User, time=0, value=1}
            {interpMode=RCIM_Cubic, tangentMode=RCTM_User, time=0.5, value=1}
            {interpMode=RCIM_Cubic, tangentMode=RCTM_User, time=1, value=1}
           ]}}})
      ))
     Float=1
    )
    ScaleA=true
    ScaleRGB=true
   ) {module=ScaleColor001}
   &ScaleSpriteSize(Scale Factor=&Vector2DFromFloat(Value=&FloatFromCurve(
      FloatCurve=DataInterface() {propertyValues {Curve {keys [
          {arriveTangent=2.6303, interpMode=RCIM_Cubic, leaveTangent=2.6303, tangentMode=RCTM_User, time=0.1, value=0}
          {interpMode=RCIM_Cubic, tangentMode=RCTM_User, time=0.25, value=1}
          <at 0.9 1>
          <at 1 0 RCIM_Cubic>
         ]}}}
      Scale Curve=&Multiply_Float(
       A=&NormalizeDistanceRange(
        Start Position:NiagaraPosition=$Particles.Position
        End Position:NiagaraPosition=Linked() {linkedVariable {name=System.CameraPostion, type=Vector3f}}
        Distance:NiagaraFloat=15000
        Invert Normalized Range:NiagaraBool=false
       )
       B=5
      )
     ))) {module=ScaleSpriteSize001}
   &DynamicMaterialParameters(
    Index 0 Param 1=&RandomRangeFloat(Minimum=0, Maximum=1)
    Index 0 Param 2=&RandomRangeFloat(Minimum=0, Maximum=1)
    Index 1 Param 1=&RandomRangeFloat(Minimum=0.05, Maximum=0.15)
    Index 1 Param 2=&RandomRangeFloat(Minimum=0.15, Maximum=0.25)
    Index 1 Param 3=&RandomRangeFloat(Minimum=-0.15, Maximum=0.1)
    Index 1 Param 4=&RandomRangeFloat(Minimum=-0.15, Maximum=1)
    Module.Param0WriteEnabled=true
    Module.Param0WriteEnabled1=true
    Module.Param1WriteEnabled=true
    Module.Param1WriteEnabled1=true
    Module.Param2WriteEnabled=true
    Module.Param2WriteEnabled1=true
    Module.Param3WriteEnabled=true
    Module.Param3WriteEnabled1=true
    Write Parameter Index 1=true
   ) {module=DynamicMaterialParameters001}
  ]
  rendererClasses [&NiagaraSpriteRendererProperties]
  renderers [{rendererClass=&NiagaraSpriteRendererProperties}]
 }
}
summary {
 systemName=NS_SideWays_Volume_InWorld_Small
 emitters [
  {emitterName=Pop_Sparks001, simTarget=GPUComputeSim, rendererClasses [&NiagaraSpriteRendererProperties]}
  {emitterName=Pop_Sparks, simTarget=GPUComputeSim, rendererClasses [&NiagaraSpriteRendererProperties]}
  {emitterName=Smoke, simTarget=GPUComputeSim, rendererClasses [&NiagaraSpriteRendererProperties]}
  {emitterName=Flow, simTarget=GPUComputeSim, rendererClasses [&NiagaraMeshRendererProperties]}
 ]
 userVariables [
  User.Landscape {
   type=NiagaraDataInterfaceLandscape
   description=""
   defaultValue {
    struct=NiagaraExt_VariableValue_DataInterface
    value {
     dataInterface=&self:NiagaraDataInterfaceLandscape_0
     dataInterfaceClass=&NiagaraDataInterfaceLandscape
    }
   }
  }
  User.SpawnRateScale {type=NiagaraFloat, defaultValue=1, description=""}
  User.SphereSize {type=NiagaraFloat, defaultValue=5000, description=""}
 ]
}
