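# Sound cue for the Sideways Grabber burst, layering two randomized spawn sounds with slight pitch modulation.
# A mixer blends six Tendril spawn variants under distant projectile attenuation and the distant ranged-fire class with six Grabber "Dist" variants under close missile attenuation and the AGV class.
# Note the apparent crossover: the "Dist"-named waves use the close attenuation.
# The cue lowers overall pitch, sets world-impact concurrency and sends to distant reverb on a custom distance curve.
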SoundCue /SidewaysGameplay/Sounds/Grabber/Grabber_Burst_Cue
root=Mixer_0
nodes [
 {
  id=Attenuation_0
  children [Modulator_0]
  properties {attenuationSettings=&Fort_Weapon_Projectile_Distant_Mono_Athena_Att}
 }
 {
  id=Attenuation_3
  children [Modulator_6]
  properties {attenuationSettings=/Game/Sounds/Attenuation/Fort_Weapon_Projectile_Close_Missile_Athena_Att}
 }
 {id=Mixer_0, children [SoundClass_0,SoundClass_3], properties {inputVolume [1,1]}}
 {id=Modulator_0, children [Random_0], properties {pitchMax=1.1}}
 {id=Modulator_6, children [Random_4], properties {pitchMax=1.1}}
 {id=Random_0, children [WavePlayer_1,WavePlayer_2,WavePlayer_3,WavePlayer_4,WavePlayer_5,WavePlayer_6], properties {weights [1,1,1,1,1,1]}}
 {
  id=Random_4
  children [WavePlayer_0, WavePlayer_7, WavePlayer_8, WavePlayer_9, WavePlayer_10, WavePlayer_11]
  properties {weights [1,1,1,1,1,1]}
 }
 {
  id=SoundClass_0
  children [Attenuation_0]
  properties {soundClassOverride=&Fort_WeaponRanged_Fire_Distant}
 }
 {id=SoundClass_3, children [Attenuation_3], properties {soundClassOverride=/Game/Sounds/FortSoundClassesAndModes/Fort_AGV}}
 {id=WavePlayer_0, properties {soundWaveAssetPtr=&/Sideways_Grabber_Spawn_Dist_05}}
 {id=WavePlayer_1, properties {soundWaveAssetPtr=&/Sideways_Tendril_Spawn_01_v2}}
 {id=WavePlayer_10, properties {soundWaveAssetPtr=&/Sideways_Grabber_Spawn_Dist_01}}
 {id=WavePlayer_11, properties {soundWaveAssetPtr=&/Sideways_Grabber_Spawn_Dist_06}}
 {id=WavePlayer_2, properties {soundWaveAssetPtr=&/Sideways_Tendril_Spawn_02_v2}}
 {id=WavePlayer_3, properties {soundWaveAssetPtr=&/Sideways_Tendril_Spawn_03_v2}}
 {id=WavePlayer_4, properties {soundWaveAssetPtr=&/Sideways_Tendril_Spawn_04_v2}}
 {id=WavePlayer_5, properties {soundWaveAssetPtr=&/Sideways_Tendril_Spawn_05_v2}}
 {id=WavePlayer_6, properties {soundWaveAssetPtr=&/Sideways_Tendril_Spawn_06_v2}}
 {id=WavePlayer_7, properties {soundWaveAssetPtr=&/Sideways_Grabber_Spawn_Dist_04}}
 {id=WavePlayer_8, properties {soundWaveAssetPtr=&/Sideways_Grabber_Spawn_Dist_03}}
 {id=WavePlayer_9, properties {soundWaveAssetPtr=&/Sideways_Grabber_Spawn_Dist_02}}
]
properties {
 bOverrideAttenuation=true
 maxDistance=27500
 pitchMultiplier=0.9
 priority=0.5
 soundClassObject=&Fort_AGV
 attenuationOverrides {falloffDistance=2500}
 concurrencySet [&/Sideways_Player_World_Impacts_Conc]
 soundSubmixSends [{
   maxSendDistance=2500
   minSendDistance=0
   minSendLevel=0
   sendLevel=0
   sendLevelControlMethod=CustomCurve
   soundSubmix=&Fortnite_Reverb_Distant
   customSendLevelCurve {editorCurveData {keys [<at 0 0>,<at 1 1>]}}
  }]
}
