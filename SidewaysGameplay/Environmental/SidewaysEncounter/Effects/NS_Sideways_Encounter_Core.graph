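# Niagara system for the glowing core of a Sideways encounter, with looping idle effects and a burst that grows with the encounter stage.
# User.Stage indexes arrays setting counts, sizes and speeds, while GlobalScale and a pulsing scalar modulate the core's flare, meshes and light.
# Embers, shimmers and tendrils surround the core; TurnOnBurstEmitters gates shockwaves, GPU sparks and burst audio.
# Distance culling and remaps adjust embers, sparks and flare by range.
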NiagaraSystem /SidewaysGameplay/Environmental/SidewaysEncounter/Effects/NS_Sideways_Encounter_Core
dependencies {
 usedDataInterfaces [
  &NiagaraDataInterfaceArrayFloat
  &NiagaraDataInterfaceCurve
  &NiagaraDataInterfaceMeshRendererInfo
  &NiagaraDataInterfaceArrayInt32
  &NiagaraDataInterfaceColorCurve
  &NiagaraDataInterfaceAudioPlayer
  &NiagaraDataInterfaceCamera
 ]
 usedDynamicInputs [&MakeVector4_FromLinearColor, &ReturnCameraVector]
 usedModules [
  &self:SystemSpawnScript.SystemScriptSource.SystemScriptGraph.NiagaraNodeAssignment_1.SetVariables_#1
  &self:SystemSpawnScript.SystemScriptSource.SystemScriptGraph.NiagaraNodeAssignment_0.SetVariables_#2
  &SystemState
  &self:CameraToSimDistance
 ]
}
emitters {
 Embers_Dark {
  simTarget=GPUComputeSim
  emitterUpdateScript [
   &EmitterState(
    Enable Distance Culling=true
    Life Cycle Mode=::NewEnumerator1
    Loop Duration=4
    MaxDistance=4500
    Module.bUseMaxDistance=true
    Module.bUseMinDistance=false
    Scalability Mode=::NewEnumerator1
    UseLoopDelay=true
   )
   &SpawnRate(
    SpawnRate=&Multiply_Float(
     A=60
     B=&FloatFromCurve(
      FloatCurve=DataInterface() {propertyValues {Curve {keys [
          <at 0 1 RCIM_Cubic>
          {arriveTangent=-0.2543, interpMode=RCIM_Cubic, leaveTangent=-0.2543, time=1.2828, value=1.6362}
          {arriveTangent=-0.3703, interpMode=RCIM_Cubic, leaveTangent=-0.3703, time=2.4472, value=0.3776}
          {arriveTangent=-0.11, interpMode=RCIM_Cubic, leaveTangent=-0.11, time=3.6302, value=0.7671}
          {arriveTangent=0.2907, interpMode=RCIM_Cubic, leaveTangent=0.2907, time=5.034, value=0.0931}
          {arriveTangent=0.3058, interpMode=RCIM_Cubic, leaveTangent=0.3058, time=6.8459, value=1.702}
          <at 8 1>
         ]}}}
      CurveIndex=$Emitter.NormalizedLoopAge
      Scale Curve=&SelectFloatFromArray(Direct Array Index=$User.Stage, Float Selection Array=DataInterface() {propertyValues {FloatData [0,1,1.2,1.4]}})
     )
    )
    Use Spawn Probability=false
   )
   /SetVariables(Emitter.ConeDir:Vector3f=&RandomRangeVector(Minimum=<-1 -1 0>, Maximum=<1 1 1>)) {isSetParameters=true, module=SetVariables_#3}
  ]
  particleSpawnScript [
   /SetVariables(Particles.SpriteRotationRate=&RandomRangeFloat(Minimum=-1, Maximum=1, Evaluation Type=::NewEnumerator1)) {isSetParameters=true, module=SetVariables_#4}
   &InitializeParticle(
    Color Maximum=<rgba 0.0443 0.0117 0.6875 1>
    Color Minimum=<rgba 0.7865 0.0934 0 1>
    Color Mode=::NewEnumerator2
    Lifetime Max=3
    Lifetime Min=0.6
    Lifetime Mode=::NewEnumerator1
    Sprite Rotation Mode=::NewEnumerator2
    Sprite Size Mode=::NewEnumerator3
    Uniform Sprite Size=&FloatFromCurve(
     FloatCurve=DataInterface() {propertyValues {Curve {keys [<at 0 6 RCIM_Cubic>, {arriveTangent=66.1397, interpMode=RCIM_Cubic, tangentMode=RCTM_Break, time=1, value=30}]}}}
     CurveIndex=&RandomRangeFloat(Minimum=0, Maximum=0.5)
    )
    UsePositionOffset=false
    Write Material Random=true
   )
   /Niagara/Modules/Spawn/Location/ConeLocation(
    Cone Angle:NiagaraFloat=&RandomRangeFloat(Minimum=15, Maximum=50)
    Cone Axis:Vector3f=$Emitter.ConeDir
    Cone Length:NiagaraFloat=&Multiply_Float(A=100, B=$User.GlobalScale)
    MaskSpawnGroup=false
    Module.Override Randomness=false
   )
   &AddVelocityInCone(
    Cone Axis=$Emitter.ConeDir
    Module.Use Velocity Falloff On Cone Axis=false
    Velocity Strength=&FloatFromCurve(
     FloatCurve=DataInterface() {propertyValues {Curve {keys [<at 0 10 RCIM_Cubic>, {arriveTangent=189.7314, interpMode=RCIM_Cubic, tangentMode=RCTM_Break, time=1, value=75}]}}}
     CurveIndex=&RandomRangeFloat(Minimum=0, Maximum=1)
     Scale Curve=&SelectFloatFromArray(Direct Array Index=$User.Stage, Float Selection Array=DataInterface() {propertyValues {FloatData [0,1,2,10]}})
    )
   )
   &AddVelocity(Velocity=&Multiply_VectorByFloat(
     Vector=$Output.SphereLocation.SphereVector
     Float=&RandomRangeFloat(Minimum=-0.04, Maximum=0.02, Evaluation Type=::NewEnumerator1)
    ))
  ]
  particleUpdateScript [
   &ParticleState()
   &CurlNoiseForce(Noise Frequency=10, Noise Strength=300, Pan Noise=false)
   &AccelerationForce(Acceleration=<0 0 50>)
   &Drag(Drag=3, Use Linear Drag=true, Use Rotational Drag=false)
   &SolveForcesAndVelocity(Clamp Velocity=false, Limit Acceleration=false)
   &ScaleSpriteSize(Scale Factor=&Vector2DFromFloat(Value=&FloatFromCurve(FloatCurve=DataInterface() {propertyValues {Curve {keys [
          <at -0.0044 0.2654>
          {arriveTangent=0.7558, interpMode=RCIM_Cubic, leaveTangent=0.7558, time=0.157, value=0.5261}
          {arriveTangent=-0.6241, interpMode=RCIM_Cubic, leaveTangent=-0.6241, time=0.7034, value=0.8003}
          <at 1 0>
         ]}}})))
   &SpriteRotationRate(Rotation Rate=$Particles.SpriteRotationRate)
   &ScaleColor(
    Scale RGB=&VectorFromFloat(Value=&RandomRangeFloat(Minimum=40, Maximum=150))
    ScaleA=false
    ScaleRGB=true
   )
  ]
  rendererClasses [&NiagaraSpriteRendererProperties]
  renderers [{rendererClass=&NiagaraSpriteRendererProperties}]
 }
 Embers_Hot001 {
  simTarget=GPUComputeSim
  emitterUpdateScript [
   &EmitterState(Enable Distance Culling=true, MaxDistance=4500, Module.bUseMaxDistance=true, Module.bUseMinDistance=false, Scalability Mode=::NewEnumerator1)
   &SpawnRate(
    Spawn Probability=&ClampFloat(
     Float=&FN_LOD_Distance_Remap(Start Distance=0, End Distance=1000, Minimum Output=1, Maximum Output=0.2)
     Min=0.2
    )
    SpawnRate=&Multiply_Float(
     A=200
     B=&FloatFromCurve(
      FloatCurve=DataInterface() {propertyValues {Curve {keys [
          <at 0 1 RCIM_Cubic>
          {arriveTangent=-0.2543, interpMode=RCIM_Cubic, leaveTangent=-0.2543, time=1.2828, value=1.6362}
          {arriveTangent=-0.3703, interpMode=RCIM_Cubic, leaveTangent=-0.3703, time=2.4472, value=0.3776}
          {arriveTangent=-0.11, interpMode=RCIM_Cubic, leaveTangent=-0.11, time=3.6302, value=0.7671}
          {arriveTangent=0.2907, interpMode=RCIM_Cubic, leaveTangent=0.2907, time=5.034, value=0.0931}
          {arriveTangent=0.3058, interpMode=RCIM_Cubic, leaveTangent=0.3058, time=6.8459, value=1.702}
          <at 8 1>
         ]}}}
      CurveIndex=$Emitter.NormalizedLoopAge
      Scale Curve=&SelectFloatFromArray(Direct Array Index=$User.Stage, Float Selection Array=DataInterface() {propertyValues {FloatData [0,1,1.2,1.4]}})
     )
    )
    Use Spawn Probability=true
   )
   /SetVariables(Emitter.ConeDir:Vector3f=&MakeVector(
     X=&Waveform(
      "Waveform [1]"=::NewEnumerator0
      "[1] Frequency"=0.2
      "[1] Amplitude Min/Max"=<0 1>
      "[1] Scale"=3
      "[1] Bias"=-0.5
      "[2] Phase"=$Emitter.Age
      "[3] Phase"=$Emitter.Age
     )
     Y=&Waveform(
      "Waveform [1]"=::NewEnumerator1
      "[1] Frequency"=0.2
      "[1] Amplitude Min/Max"=<0 1>
      "[1] Scale"=3
      "[1] Bias"=-0.5
      "[2] Phase"=$Emitter.Age
      "[3] Phase"=$Emitter.Age
     )
     Z=&Waveform(
      "Waveform [1]"=::NewEnumerator2
      "[1] Frequency"=0.2
      "[1] Compound Frequency"=0.2
      "[1] Amplitude Min/Max"=<0 1>
      "[1] Scale"=3
      "[1] Bias"=-0.5
      "[2] Phase"=$Emitter.Age
      "[3] Phase"=$Emitter.Age
     )
    )) {isSetParameters=true, module=SetVariables_#3}
  ]
  particleSpawnScript [
   /SetVariables(Particles.SpriteRotationRate=&RandomRangeFloat(Minimum=-1, Maximum=1, Evaluation Type=::NewEnumerator1)) {isSetParameters=true, module=SetVariables_#4}
   &InitializeParticle(
    Color Maximum=<rgba 0.25 0 0.138 1>
    Color Minimum=<rgba 0.7865 0.0934 0 1>
    Color Mode=::NewEnumerator2
    Lifetime Max=5
    Lifetime Min=0.6
    Lifetime Mode=::NewEnumerator1
    Sprite Rotation Mode=::NewEnumerator2
    Sprite Size Mode=::NewEnumerator1
    Uniform Sprite Size Max=6
    Uniform Sprite Size Min=2
    UsePositionOffset=false
    Write Material Random=true
   )
   &SphereLocation(
    MaskSpawnGroup=false
    Override Seed=false
    Sphere Radius=&RandomRangeFloat(Minimum=1, Maximum=20)
    Surface Only=false
    UseNonUniformScale=false
    UseOffset=false
    UseSphereOrientation=false
   )
   &AddVelocityInCone(
    Cone Angle=10
    Cone Axis=$Emitter.ConeDir
    Module.Use Velocity Falloff On Cone Axis=false
    Velocity Strength=&SelectFloatFromArray(Direct Array Index=$User.Stage, Float Selection Array=DataInterface() {propertyValues {FloatData [100,100,100,200]}})
   )
  ]
  particleUpdateScript [
   &ParticleState()
   &CurlNoiseForce(Noise Frequency=20, Noise Strength=150, Pan Noise=false)
   &AccelerationForce(Acceleration=<0 0 50>)
   &Drag(Use Linear Drag=true, Use Rotational Drag=false)
   &SolveForcesAndVelocity(Clamp Velocity=false, Limit Acceleration=false)
   &ScaleSpriteSize(Scale Factor=&Vector2DFromFloat(Value=&FloatFromCurve(
      FloatCurve=DataInterface() {propertyValues {Curve {keys [
          {arriveTangent=4.1769, interpMode=RCIM_Cubic, leaveTangent=4.1769, tangentMode=RCTM_User, time=-0.0044, value=0.2654}
          {arriveTangent=0.7558, interpMode=RCIM_Cubic, leaveTangent=0.7558, time=0.2, value=0.8}
          {arriveTangent=-1, interpMode=RCIM_Cubic, leaveTangent=-1, time=0.7034, value=0.8003}
          <at 1 0>
         ]}}}
      Scale Curve=&ClampFloat(
       Float=&FN_LOD_Distance_Remap(Start Distance=0, End Distance=2000, Minimum Output=0.1, Maximum Output=3)
       Min=0.1
       Max=3
      )
     )))
   &SpriteRotationRate(Rotation Rate=$Particles.SpriteRotationRate)
   &ScaleColor(
    Scale RGB=&VectorFromFloat(Value=&RandomRangeFloat(Minimum=15, Maximum=75))
    ScaleA=false
    ScaleRGB=true
   )
  ]
  rendererClasses [&NiagaraSpriteRendererProperties]
  renderers [{rendererClass=&NiagaraSpriteRendererProperties}]
 }
 Idle_FlareAndShard {
  emitterSpawnScript [/SetVariables(Emitter.OpenDistanceInput:NiagaraFloat=&SelectFloatFromArray(Direct Array Index=$User.Stage, Float Selection Array=DataInterface() {propertyValues {FloatData [0,12,38,150]}})) {isSetParameters=true}]
  emitterUpdateScript [
   &EmitterState(Life Cycle Mode=::NewEnumerator1, Loop Behavior=::NewEnumerator1, Loop Duration Mode=::NewEnumerator1, UseLoopDelay=false)
   &SpawnBurst_Instantaneous(Use Spawn Probability=false, UseLoopCountLimit=false)
  ]
  particleSpawnScript [
   &InitializeParticle(
    Color=<rgba 0.25 0.0229 0.0378 1>
    Lifetime Max=1.5
    Lifetime Min=0.12
    Lifetime Mode=::NewEnumerator1
    Mesh Scale Mode=::NewEnumerator3
    Mesh Uniform Scale=1.5
    Sprite Size=&Multiply_Vector2D_ByFloat(Float=1, Vector2D=<800 400>)
    Sprite Size Mode=::NewEnumerator4
    UsePositionOffset=false
    Write Material Random=true
   )
   /SetVariables(Particles.RoundFlareColor:LinearColor=<rgba 0.0434 0.127 1 1>) {isSetParameters=true}
   &DynamicMaterialParameters(
    Index 0 Param 1=&RandomRangeFloat(Minimum=0.16, Maximum=0.35, Evaluation Type=::NewEnumerator1)
    Index 0 Param 2=&RandomRangeFloat(Minimum=1, Maximum=2, Evaluation Type=::NewEnumerator1)
    Index 0 Param 3=1
    Index 0 Param 4=1
    Module.Param0WriteEnabled=true
    Module.Param1WriteEnabled=true
    Module.Param2WriteEnabled=true
    Module.Param3WriteEnabled=true
   )
   /SetVariables() {isSetParameters=true}
   /SetVariables(
    Particles.DarkGradColor:LinearColor=<rgba 0.0833 0.0057 0.0382 0.8>
    Particles.DarkGradSize:Vector2f=&Multiply_Vector2D_ByFloat(Float=&Divide_Float(A=1, B=1.5), Vector2D=<700 700>)
   ) {isSetParameters=true}
   /SetVariables(
    Particles.CoronaColor:LinearColor=<rgba 0.9393 0.3177 2 1>
    Particles.CoronaDP:Vector4f=&MakeVector4(
     X=&RandomRangeFloat(Minimum=0.8, Maximum=1.2)
     Y=&RandomRangeFloat(Minimum=0.2, Maximum=0.4, Evaluation Type=::NewEnumerator1)
     Z=&RandomRangeFloat(Minimum=0.4, Maximum=0.6, Evaluation Type=::NewEnumerator1)
     W=0
    )
    Particles.CoronaSize:Vector2f=&Multiply_Vector2D_ByFloat(Float=&Divide_Float(A=1, B=1.3), Vector2D=<500 500>)
   ) {isSetParameters=true}
   &InitialMeshOrientation(Use Orientation Vector=true, Use Rotation Vector=false)
  ]
  particleUpdateScript [
   &ParticleState(Kill Particles When Lifetime Has Elapsed=false, Let Infinitely Lived Particles Die When Emitter Deactivates=true)
   /SetVariables(Particles.DarkGradSize:Vector2f=&Multiply_Vector2D_ByFloat(
     Float=&Multiply_Float(
      A=$User.GlobalScale
      B=&SelectFloatFromArray(Direct Array Index=$User.Stage, Float Selection Array=DataInterface() {propertyValues {FloatData [1,1,2,2.5]}})
     )
     Vector2D=$Particles.Initial.DarkGradSize
    )) {isSetParameters=true}
   /SetVariables(Particles.CoronaSize:Vector2f=&Multiply_Vector2D_ByFloat(Float=$User.GlobalScale, Vector2D=$Particles.Initial.CoronaSize)) {isSetParameters=true}
   /SetVariables(Particles.PointLightColor:LinearColor=&MultiplyLinearColor(Linear Color=$Particles.Color, Scale Factor=&MakeVector4_FromFloat(Float=15))) {isSetParameters=true}
   /SetVariables(Particles.LightRadius=&Multiply_Float(
     A=100
     B=&Multiply_Float(
      A=&Multiply_Float(
       A=&Multiply_Float(
        A=$User.GlobalScale
        B=&SelectFloatFromArray(Direct Array Index=$User.Stage, Float Selection Array=DataInterface() {propertyValues {FloatData [0.6,1,1.2,1.4]}})
       )
       B=&Multiply_Float(
        A=$User.GlobalScale
        B=&SelectFloatFromArray(Direct Array Index=$User.Stage, Float Selection Array=DataInterface() {propertyValues {FloatData [1,1,1.2,1.4]}})
       )
      )
      B=&Lerp_Float(A=1, B=1.5, Alpha=$User.CorePlusingScalar)
     )
    )) {isSetParameters=true}
   /SetVariables(
    Particles.CoreMeshOri:Quat4f=<0 0 -1 1>
    Particles.CoreSize:Vector3f=&VectorFromFloat(Value=&Multiply_Float(
      A=&Multiply_Float(A=$User.GlobalScale, B=&Lerp_Float(A=1, B=1.05, Alpha=$User.CorePlusingScalar))
      B=&Multiply_Float(A=$System.OpenTheCore, B=2.5)
     ))
   ) {isSetParameters=true}
   /SetVariables(Particles.DecalSize:Vector3f=&Multiply_VectorByFloat(
     Vector=<1000 256 256>
     Float=&SelectFloatFromArray(Direct Array Index=$User.Stage, Float Selection Array=DataInterface() {propertyValues {FloatData [0.1,1,1.6,2.2]}})
    )) {isSetParameters=true}
   /SetVariables(Particles.CoronaSize:Vector2f=&Multiply_Vector2D_ByFloat(
     Float=&SelectFloatFromArray(Direct Array Index=$User.Stage, Float Selection Array=DataInterface() {propertyValues {FloatData [0.7,1,2,2.5]}})
     Vector2D=$Particles.Initial.CoronaSize
    )) {isSetParameters=true}
   &ScaleSpriteSize(
    Scale Sprite Size Mode=::NewEnumerator0
    Uniform Scale Factor=&Multiply_Float(
     A=&Multiply_Float(
      A=&Multiply_Float(
       A=$User.GlobalScale
       B=&MakeCustomFloatFromBool(
        Bool=$User.TurnOnBurstEmitters
        True Float=&FloatFromCurve(
         FloatCurve=DataInterface() {propertyValues {Curve {keys [<at 0 1 RCIM_Cubic>,<at 0.15 2.5 RCIM_Cubic>,<at 0.6 1>,<at 1 1>]}}}
         CurveIndex=$Particles.Age
        )
        False Float=1
       )
      )
      B=&Lerp_Float(A=1, B=6, Alpha=$User.CorePlusingScalar)
     )
     B=&Multiply_Float(
      A=&ClampFloat(
       Float=&FN_LOD_Distance_Remap(Start Distance=0, End Distance=1500, Minimum Output=0, Maximum Output=1)
       Min=0.03
       Max=0.3
      )
      B=&SelectFloatFromArray(Direct Array Index=$User.Stage, Float Selection Array=DataInterface() {propertyValues {FloatData [0.3,0.8,1.2,1.4]}})
     )
    )
   )
   &SolveForcesAndVelocity(Clamp Velocity=false, Limit Acceleration=false)
   &ScaleColor(
    Scale RGB=&VectorFromFloat(Value=&Multiply_Float(A=$User.GlobalScale, B=&Lerp_Float(A=2, B=7, Alpha=$User.CorePlusingScalar)))
    ScaleA=true
    ScaleRGB=true
   )
   &ScaleMeshSize(Scale Factor=&VectorFromFloat(Value=&Multiply_Float(
      A=&Multiply_Float(
       A=$User.GlobalScale
       B=&SelectFloatFromArray(Direct Array Index=$User.Stage, Float Selection Array=DataInterface() {propertyValues {FloatData [0.7,1,1.2,1.4]}})
      )
      B=&Lerp_Float(A=1, B=1.05, Alpha=$User.CorePlusingScalar)
     )))
   &DynamicMaterialParameters(
    Index 0 Param 1=1
    Index 1 Param 1=&Lerp_Float(
     A=-0.5
     B=&Lerp_Float(A=1, B=1.65, Alpha=$User.CorePlusingScalar)
     Alpha=&Multiply_Float(
      A=&Multiply_Float(
       A=$System.OpenTheCore
       B=&FloatFromCurve(
        FloatCurve=DataInterface() {propertyValues {Curve {keys [<at 0 1 RCIM_Cubic>,<at 0.15 2.8 RCIM_Cubic>,<at 0.6 1>,<at 1 1>]}}}
        CurveIndex=$Particles.Age
       )
      )
      B=1
     )
    )
    Module.Param0WriteEnabled=false
    Module.Param0WriteEnabled1=true
    Module.Param1WriteEnabled=false
    Module.Param1WriteEnabled1=true
    Module.Param2WriteEnabled=false
    Module.Param2WriteEnabled1=true
    Module.Param3WriteEnabled=false
    Module.Param3WriteEnabled1=true
    Write Parameter Index 1=true
   ) {module=DynamicMaterialParameters001}
  ]
  rendererClasses [
   &NiagaraSpriteRendererProperties
   &NiagaraMeshRendererProperties
   &NiagaraLightRendererProperties
   &NiagaraComponentRendererProperties
  ]
  renderers [
   {rendererClass=&NiagaraSpriteRendererProperties}
   {rendererClass=&NiagaraSpriteRendererProperties, rendererIndex=1}
   {rendererClass=&NiagaraSpriteRendererProperties, rendererIndex=2}
   {rendererClass=&NiagaraMeshRendererProperties, rendererIndex=3}
   {rendererClass=&NiagaraSpriteRendererProperties, rendererIndex=4}
   {rendererClass=&NiagaraLightRendererProperties, rendererIndex=5}
   {rendererClass=&NiagaraMeshRendererProperties, rendererIndex=6}
   {rendererClass=&NiagaraMeshRendererProperties, rendererIndex=7}
   {rendererClass=&NiagaraComponentRendererProperties, rendererIndex=8}
  ]
 }
 Shimmers {
  emitterUpdateScript [
   &EmitterState(
    Enable Distance Culling=true
    Inactive Response=::NewEnumerator1
    Life Cycle Mode=::NewEnumerator1
    Loop Duration=1
    MaxDistance=6000
    Module.bUseMaxDistance=true
    Module.bUseMinDistance=false
    Scalability Mode=::NewEnumerator1
    UseLoopDelay=true
   )
   &SpawnRate(SpawnRate=4, Use Spawn Probability=false)
   /SetVariables(Emitter.PanSpeed=1) {isSetParameters=true}
  ]
  particleSpawnScript [
   /SetVariables(Particles.SpriteRotationRate=&RandomRangeFloat(Minimum=-0.03, Maximum=0.03, Evaluation Type=::NewEnumerator1)) {isSetParameters=true}
   /SetVariables(
    Particles.ColorEnd:LinearColor=&RandomRangeLinearColor(Minimum=<rgba 0.25 0 0.138 1>, Maximum=<rgba 0.0466 0 0.25 1>)
    Particles.ColorStart:LinearColor=&RandomRangeLinearColor(Minimum=<rgba 0.36 0.3 0.45 1>, Maximum=<rgba 0.6 0.68 0.74 1>)
   ) {isSetParameters=true}
   &InitializeParticle(
    Color=$Particles.ColorStart
    Lifetime Max=4.3
    Lifetime Min=3.7
    Lifetime Mode=::NewEnumerator1
    Sprite Rotation Mode=::NewEnumerator2
    Sprite Size=&Multiply_Vector2D_ByFloat(
     Float=$User.GlobalScale
     Vector2D=&RandomRangeVector2D(Minimum=<80 80>, Maximum=<120 120>, Evaluation Type=::NewEnumerator1)
    )
    Sprite Size Mode=::NewEnumerator4
    UsePositionOffset=false
    Write Material Random=true
   )
   &SphereLocation(
    MaskSpawnGroup=false
    Override Seed=false
    Sphere Radius=&RandomRangeFloat(
     Minimum=50
     Maximum=&Multiply_Float(
      A=120
      B=&SelectFloatFromArray(Direct Array Index=$User.Stage, Float Selection Array=DataInterface() {propertyValues {FloatData [0.6,1,1.4,2]}})
     )
    )
    Surface Only=false
    UseNonUniformScale=false
    UseOffset=false
    UseSphereOrientation=false
   ) {module=SphereLocation001}
   &AddVelocityFromPoint(Module.Use Falloff Distance=false, Velocity Strength=&RandomRangeFloat(Minimum=10, Maximum=40))
   &AddVelocity(Velocity=&Multiply_VectorByFloat(
     Vector=$Output.SphereLocation.SphereVector
     Float=&RandomRangeFloat(Minimum=-0.04, Maximum=-0.02, Evaluation Type=::NewEnumerator1)
    ))
   &DynamicMaterialParameters(
    Index 0 Param 1=&RandomRangeFloat(Minimum=0, Maximum=1, Evaluation Type=::NewEnumerator1)
    Index 0 Param 2=&RandomRangeFloat(Minimum=0, Maximum=1, Evaluation Type=::NewEnumerator1)
    Index 0 Param 3=0.2
    Module.Param0WriteEnabled=true
    Module.Param1WriteEnabled=true
    Module.Param2WriteEnabled=true
    Module.Param3WriteEnabled=true
   )
   &CameraOffset(Camera Offset Amount=&RandomRangeFloat(Minimum=50, Maximum=100, Evaluation Type=::NewEnumerator1))
  ]
  particleUpdateScript [
   &ParticleState()
   &Drag(Use Linear Drag=true, Use Rotational Drag=false)
   &SolveForcesAndVelocity(Clamp Velocity=false, Limit Acceleration=false)
   &DynamicMaterialParameters(
    Index 0 Param 3=1
    Index 0 Param 4=1
    Module.Param0WriteEnabled=true
    Module.Param1WriteEnabled=true
    Module.Param2WriteEnabled=false
    Module.Param3WriteEnabled=true
   ) {module=DynamicMaterialParameters001}
   &SpriteRotationRate(Rotation Rate=$Particles.SpriteRotationRate)
   &ScaleColor(
    Scale RGB=&MakeVectorFromLinearColorRGB(Color=&LerpLinearColors(
      StartColor=$Particles.ColorStart
      EndColor=$Particles.ColorEnd
      LerpFactor=&FloatFromCurve(FloatCurve=DataInterface() {propertyValues {Curve {keys [<at 0 0 RCIM_Cubic>, {interpMode=RCIM_Cubic, tangentMode=RCTM_Break, time=0.5, value=1}]}}})
     ))
    ScaleA=false
    ScaleRGB=true
   ) {module=ScaleColor001}
   &ScaleColor(
    Scale Alpha=&FloatFromCurve(
     FloatCurve=DataInterface() {propertyValues {Curve {keys [<at 0 0 RCIM_Cubic>,<at 0.5 1 RCIM_Cubic>,<at 1 0>]}}}
     Scale Curve=&RandomRangeFloat(Minimum=1.2, Maximum=1.6)
    )
    ScaleA=true
    ScaleRGB=false
   )
   &ScaleSpriteSize(Uniform Curve Sprite Scale=DataInterface() {propertyValues {Curve {keys [<at 0 0.75 RCIM_Cubic>, {interpMode=RCIM_Cubic, tangentMode=RCTM_Break, time=1, value=1}]}}})
  ]
  rendererClasses [&NiagaraSpriteRendererProperties]
  renderers [{rendererClass=&NiagaraSpriteRendererProperties}]
 }
 Shockwave001 {
  emitterSpawnScript [/SetVariables() {isSetParameters=true}]
  emitterUpdateScript [
   &EmitterState(Life Cycle Mode=::NewEnumerator1, Loop Behavior=::NewEnumerator1, Loop Duration=1, UseLoopDelay=true)
   &SpawnBurst_Instantaneous(
    Spawn Count=&SelectIntFromArray(Direct Array Index=$User.Stage, Int Selection Array=DataInterface() {propertyValues {IntData [0,1,1,1]}})
    Spawn Probability=&MakeCustomFloatFromBool(Bool=$User.TurnOnBurstEmitters, True Float=1, False Float=0)
    Use Spawn Probability=true
   )
  ]
  particleSpawnScript [
   &InitializeParticle(
    Color Mode=::NewEnumerator0
    Lifetime=0.5
    Mesh Scale Mode=::NewEnumerator3
    Mesh Uniform Scale=7
    UsePositionOffset=false
    Write Material Random=true
   )
   &InitialMeshOrientation(Orientation Axis=<0 0 1>, Orientation Vector=<0 0 1>, Rotation=<0 0.25 0>, Use Orientation Vector=true, Use Rotation Vector=true)
   &InitialMeshOrientation(
    Orientation Axis=<0 0 0>
    Orientation Vector=<0 0 0>
    Rotation=&RandomRangeVector(Minimum=<0 0 0>, Maximum=<1 0 0>)
    Use Orientation Vector=false
    Use Rotation Vector=true
   ) {module=InitialMeshOrientation001}
   &CameraOffset(Camera Offset Amount=30)
   /Niagara/Modules/Audio/PlayAudio(Audio Player=DataInterface() {propertyValues {
      Attenuation=null
      Concurrency=null
      MaxPlaysPerTick=10
      SoundToPlay=/SidewaysGameplay/Sounds/Grabber/Grabber_Burst_Cue
      bAllowLoopingOneShotSounds=false
      bLimitPlaysPerTick=true
      bOnlyActiveDuringGameplay=true
      bStopWhenComponentIsDestroyed=true
      ConfigurationUserParameter {parameter {name=None, typeDefHandle {registeredTypeIndex=1}, varData []}}
      ParameterNames []
     }})
  ]
  particleUpdateScript [
   &ParticleState()
   &ScaleColor(
    Linear Color Curve=DataInterface() {propertyValues {
      AlphaCurve {keys [<at 0 1>,<at 0.3576 1>,<at 1 0>]}
      BlueCurve {keys [<at 0 1>,<at 0.3654 1>,<at 1 1>]}
      GreenCurve {keys [<at 0 0.935>,<at 0.3654 0.0555>,<at 1 0.5709>]}
      RedCurve {keys [<at 0 0.904>,<at 0.3654 0.4181>,<at 1 0.1818>]}
     }}
    Scale Mode=::NewEnumerator2
   )
   &DynamicMaterialParameters(
    Index 0 Param 1=&FloatFromCurve(FloatCurve=DataInterface() {propertyValues {Curve {keys [{arriveTangent=-2.1015, interpMode=RCIM_Cubic, leaveTangent=-2.1015, tangentMode=RCTM_User, time=0, value=0.5}, <at 1 -0.15>]}}})
    Index 0 Param 2=&FloatFromCurve(FloatCurve=DataInterface() {propertyValues {Curve {keys [{arriveTangent=-0.3464, interpMode=RCIM_Cubic, leaveTangent=-0.3464, tangentMode=RCTM_User, time=0, value=1}, <at 1 0.9>]}}})
    Index 0 Param 3=&FloatFromCurve(FloatCurve=DataInterface() {propertyValues {Curve {keys [<at 0 0>,<at 0.15 1>,<at 1 1.2>]}}})
    Module.Param0WriteEnabled=true
    Module.Param1WriteEnabled=true
    Module.Param2WriteEnabled=true
    Module.Param3WriteEnabled=false
   )
   &ScaleColor(
    Scale Alpha=&FNTimeOfDayScalar(Night Scale=0.7, Morning Scale=0.5, Evening Scale=0.5)
    Scale RGB=&VectorFromFloat(Value=&FloatFromCurve(FloatCurve=DataInterface() {propertyValues {Curve {keys [<at 0 10>,<at 0.15 3>,<at 1 1>]}}}))
    ScaleA=true
    ScaleRGB=true
   ) {module=ScaleColor001}
   &ScaleMeshSize(Scale Factor=&VectorFromFloat(Value=&FloatFromCurve(
      FloatCurve=DataInterface() {propertyValues {Curve {keys [<at 0 0>,<at 0.1 1>,<at 1 2>]}}}
      Scale Curve=&SelectFloatFromArray(Direct Array Index=$User.Stage, Float Selection Array=DataInterface() {propertyValues {FloatData [0,0.5,0.8,1]}})
     )))
  ]
  rendererClasses [&NiagaraMeshRendererProperties]
  renderers [
   {rendererClass=&NiagaraMeshRendererProperties, rendererIndex=0}
   {rendererClass=&NiagaraMeshRendererProperties, rendererIndex=1}
  ]
 }
 SoftWave {
  emitterUpdateScript [
   &EmitterState(Life Cycle Mode=::NewEnumerator1, Loop Behavior=::NewEnumerator1, Loop Delay=0.1, Loop Duration=1, UseLoopDelay=true)
   &SpawnBurst_Instantaneous(
    Spawn Count=&SelectIntFromArray(Direct Array Index=$User.Stage, Int Selection Array=DataInterface() {propertyValues {IntData [0,1,1,1]}})
    Spawn Probability=&MakeCustomFloatFromBool(Bool=$User.TurnOnBurstEmitters, True Float=1, False Float=0)
    Use Spawn Probability=true
   )
  ]
  particleSpawnScript [
   &InitializeParticle(Lifetime=0.4, Mesh Scale Mode=::NewEnumerator3, Mesh Uniform Scale=9, UsePositionOffset=false, Write Material Random=true)
   &SystemLocation(MaskSpawnGroup=false, Offset=<0 0 18>)
   &InitialMeshOrientation(Orientation Axis=<0 0 1>, Orientation Vector=<0 0 1>, Rotation=<0 0.25 0>, Use Orientation Vector=true, Use Rotation Vector=true)
  ]
  particleUpdateScript [
   &ParticleState()
   &ScaleMeshSize(Scale Factor=&VectorFromFloat(Value=&FloatFromCurve(FloatCurve=DataInterface() {propertyValues {Curve {keys [
          {arriveTangent=1.9852, time=0, value=0.2}
          <at 0.2 0.8>
          {arriveTangent=0.0045, interpMode=RCIM_Cubic, leaveTangent=0.0045, tangentMode=RCTM_User, time=1, value=1}
         ]}}})))
   &DynamicMaterialParameters(
    Index 0 Param 2=1
    Index 0 Param 3=1
    Index 0 Param 4=&FloatFromCurve(FloatCurve=DataInterface() {propertyValues {Curve {keys [<at 0 0.7>,<at 1 0.05>]}}})
    Module.Param0WriteEnabled=true
    Module.Param1WriteEnabled=true
    Module.Param2WriteEnabled=true
    Module.Param3WriteEnabled=true
   )
   &ScaleColor(
    Scale Alpha=&FloatFromCurve(FloatCurve=DataInterface() {propertyValues {Curve {keys [{arriveTangent=-2.1056, interpMode=RCIM_Cubic, leaveTangent=-2.1056, tangentMode=RCTM_User, time=0, value=1}, <at 1 0>]}}})
    ScaleA=true
    ScaleRGB=true
   )
   &ScaleColor(
    Linear Color Curve=DataInterface() {propertyValues {
      AlphaCurve {keys [<at 0 2>,<at 1 1>]}
      BlueCurve {keys [<at 0 0.151>,<at 0.4478 0.4765>,<at 1 1>]}
      GreenCurve {keys [<at 0 1>,<at 0.4478 1>,<at 1 0.9859>]}
      RedCurve {keys [<at 0 0.3491>,<at 0.4478 0.151>,<at 1 0.151>]}
     }}
    Scale Mode=::NewEnumerator2
   ) {module=ScaleColor001}
   /SetVariables(
    Particles.Tile2Color:LinearColor=&ColorFromCurve(ColorCurve=DataInterface() {propertyValues {
       AlphaCurve {keys [<at 0 1 RCIM_Cubic>,<at 1 0 RCIM_Cubic>]}
       BlueCurve {keys [<at 0 0.276>, {arriveTangent=0.724, interpMode=RCIM_Cubic, leaveTangent=0.724, time=0.2691, value=0.6438}, <at 1 1 RCIM_Cubic>]}
       GreenCurve {keys [<at 0 1>, {arriveTangent=-0.5406, interpMode=RCIM_Cubic, leaveTangent=-0.5406, time=0.2691, value=1}, <at 1 0.4594 RCIM_Cubic>]}
       RedCurve {keys [<at 0 0.6742>, {arriveTangent=-0.5753, interpMode=RCIM_Cubic, leaveTangent=-0.5753, time=0.2691, value=0.1094}, <at 1 0.099 RCIM_Cubic>]}
      }})
    Particles.Tile2Size:Vector3f=&VectorFromFloat(Value=&FloatFromCurve(
      FloatCurve=DataInterface() {propertyValues {Curve {keys [
          {arriveTangent=1.9852, time=0, value=0.2}
          <at 0.2 0.8>
          {arriveTangent=0.0045, interpMode=RCIM_Cubic, leaveTangent=0.0045, tangentMode=RCTM_User, time=1, value=1}
         ]}}}
      Scale Curve=15
     ))
   ) {isSetParameters=true}
  ]
  rendererClasses [&NiagaraMeshRendererProperties]
  renderers [{rendererClass=&NiagaraMeshRendererProperties, rendererIndex=0}]
 }
 Spark {
  simTarget=GPUComputeSim
  emitterUpdateScript [
   &EmitterState(Life Cycle Mode=::NewEnumerator1, Loop Behavior=::NewEnumerator1, Loop Duration=1, UseLoopDelay=true)
   &SpawnBurst_Instantaneous(
    Spawn Count=&Multiply_Int(
     A=100
     B=&SelectIntFromArray(Direct Array Index=$User.Stage, Int Selection Array=DataInterface() {propertyValues {IntData [0,1,2,3]}})
    )
    Spawn Probability=&MakeCustomFloatFromBool(Bool=$User.TurnOnBurstEmitters, True Float=1, False Float=0)
    Use Spawn Probability=true
   )
  ]
  particleSpawnScript [
   &InitializeParticle(
    Color=<rgba 10 10 10 1>
    Lifetime Min=0.3
    Lifetime Mode=::NewEnumerator1
    Sprite Size Max=<10 20>
    Sprite Size Min=<5 10>
    Sprite Size Mode=::NewEnumerator2
    UsePositionOffset=false
    Write Material Random=true
   ) {module=InitializeParticle001}
   &SphereLocation(
    MaskSpawnGroup=false
    Override Seed=false
    Sphere Radius=&RandomRangeFloat(Minimum=10, Maximum=450)
    Surface Expansion Mode=::NewEnumerator1
    Surface Only=true
    Surface Only Band Thickness=1
    UseNonUniformScale=false
    UseOffset=false
    UseSphereOrientation=false
   ) {module=SphereLocation001}
   &AddVelocityFromPoint(Module.Use Falloff Distance=false, Velocity Strength=&UniformRangedFloat(Minimum=100, Maximum=1500))
   &CameraOffset(Camera Offset Amount=100) {module=CameraOffset001}
  ]
  particleUpdateScript [
   &ParticleState()
   &CurlNoiseForce(
    Noise Frequency=5
    Noise Strength=&Multiply_Float(
     A=2000
     B=&FloatFromCurve(FloatCurve=DataInterface() {propertyValues {Curve {keys [<at 0 0>,<at 0.1 1>,<at 1 1.5>]}}})
    )
    Pan Noise=false
   )
   &Drag(Drag=&UniformRangedFloat(Minimum=4, Maximum=8), Use Linear Drag=true, Use Rotational Drag=false)
   &AccelerationForce(Acceleration=<0 0 500>)
   &SolveForcesAndVelocity(Clamp Velocity=false, Limit Acceleration=false)
   &ScaleSpriteSize(Scale Factor=&Vector2DFromFloat(Value=&FloatFromCurve(
      FloatCurve=DataInterface() {propertyValues {Curve {keys [<at 0 1>,<at 1 0>]}}}
      Scale Curve=&ClampFloat(
       Float=&FN_LOD_Distance_Remap(Start Distance=0, End Distance=2000, Minimum Output=0.3, Maximum Output=1)
       Min=0.3
      )
     )))
   &ScaleSpriteSizeBySpeed(Max Scale Factor=<0.5 5>, Min Scale Factor=<1 0.5>)
   &ScaleColor(
    Scale RGB=&MakeVectorFromLinearColorRGB(Color=&ColorFromCurve(ColorCurve=DataInterface() {propertyValues {
        AlphaCurve {keys []}
        BlueCurve {keys [<at 0 0.2652 RCIM_Cubic>,<at 0.1464 0.1002>,<at 1 0.1335>]}
        GreenCurve {keys [<at 0 0.1856 RCIM_Cubic>,<at 0.1464 0.405>,<at 1 0.526>]}
        RedCurve {keys [<at 0 2 RCIM_Cubic>,<at 0.1464 1>,<at 1 1>]}
       }}))
    ScaleA=true
    ScaleRGB=true
   )
  ]
  rendererClasses [&NiagaraSpriteRendererProperties]
  renderers [{rendererClass=&NiagaraSpriteRendererProperties}]
 }
 Spark001 {
  simTarget=GPUComputeSim
  emitterUpdateScript [
   &EmitterState(Life Cycle Mode=::NewEnumerator1, Loop Behavior=::NewEnumerator1, Loop Duration=1, UseLoopDelay=true)
   &SpawnBurst_Instantaneous(
    Spawn Count=&Multiply_Int(
     A=1000
     B=&SelectIntFromArray(Direct Array Index=$User.Stage, Int Selection Array=DataInterface() {propertyValues {IntData [0,1,2,3]}})
    )
    Spawn Probability=&MakeCustomFloatFromBool(Bool=$User.TurnOnBurstEmitters, True Float=1, False Float=0)
    Use Spawn Probability=true
   )
   &SpawnBurst_Instantaneous(
    Spawn Count=1000
    Spawn Probability=&MakeCustomFloatFromBool(Bool=$User.TurnOnBurstEmitters, True Float=1, False Float=0)
    Spawn Time=0.2
    Use Spawn Probability=true
   ) {enabled=false, module=SpawnBurst_Instantaneous001}
  ]
  particleSpawnScript [
   &InitializeParticle(
    Color=<rgba 5 5 5 1>
    Lifetime Min=0.3
    Lifetime Mode=::NewEnumerator1
    Sprite Size Max=<10 20>
    Sprite Size Min=<5 10>
    Sprite Size Mode=::NewEnumerator2
    UsePositionOffset=false
    Write Material Random=true
   ) {module=InitializeParticle001}
   &SphereLocation(
    MaskSpawnGroup=false
    Override Seed=false
    Sphere Radius=&RandomRangeFloat(Minimum=1, Maximum=5)
    Surface Expansion Mode=::NewEnumerator1
    Surface Only=true
    Surface Only Band Thickness=1
    UseNonUniformScale=false
    UseOffset=false
    UseSphereOrientation=false
   ) {module=SphereLocation001}
   &AddVelocityFromPoint(Module.Use Falloff Distance=false, Velocity Strength=&UniformRangedFloat(Minimum=100, Maximum=1500)) {enabled=false}
   &CameraOffset(Camera Offset Amount=100) {module=CameraOffset001}
   &AddVelocityInCone(
    Cone Angle=180
    Cone Axis=<0 0 1>
    Module.Use Velocity Falloff On Cone Axis=false
    Velocity Strength=&Multiply_Float(
     A=&UniformRangedFloat(Minimum=1000, Maximum=1500)
     B=&SelectFloatFromArray(Direct Array Index=$User.Stage, Float Selection Array=DataInterface() {propertyValues {FloatData [0,0.7,1,1.3]}})
    )
   )
  ]
  particleUpdateScript [
   &ParticleState()
   &CurlNoiseForce(
    Noise Frequency=10
    Noise Strength=&Multiply_Float(
     A=500
     B=&FloatFromCurve(FloatCurve=DataInterface() {propertyValues {Curve {keys [<at 0 0>,<at 0.1 1>,<at 1 1.5>]}}})
    )
    Pan Noise=false
   )
   &Drag(
    Drag=&FloatFromCurve(
     FloatCurve=DataInterface() {propertyValues {Curve {keys [<at 0 1>,<at 1 3.3>]}}}
     Scale Curve=&RandomRangeFloat(Minimum=1, Maximum=2)
    )
    Use Linear Drag=true
    Use Rotational Drag=false
   )
   &AccelerationForce(Acceleration=<0 0 50>)
   &SolveForcesAndVelocity(Clamp Velocity=false, Limit Acceleration=false)
   &ScaleSpriteSize(Scale Factor=&Vector2DFromFloat(Value=&FloatFromCurve(
      FloatCurve=DataInterface() {propertyValues {Curve {keys [<at 0 1>,<at 1 0>]}}}
      Scale Curve=&ClampFloat(
       Float=&FN_LOD_Distance_Remap(Start Distance=0, End Distance=2000, Minimum Output=0.3, Maximum Output=1)
       Min=0.3
      )
     )))
   &ScaleSpriteSizeBySpeed(Max Scale Factor=<0.5 5>, Min Scale Factor=<1 0.5>)
   &ScaleColor(
    Scale RGB=&MakeVectorFromLinearColorRGB(Color=&ColorFromCurve(ColorCurve=DataInterface() {propertyValues {
        AlphaCurve {keys []}
        BlueCurve {keys [<at 0 0.2652 RCIM_Cubic>,<at 0.212 0.0464>,<at 0.6056 0.2111>,<at 1 0.0986>]}
        GreenCurve {keys [<at 0 0.1856 RCIM_Cubic>,<at 0.212 0.3929>,<at 0.6056 0.1314>,<at 1 0.4353>]}
        RedCurve {keys [<at 0 2 RCIM_Cubic>,<at 0.212 1>,<at 0.6056 1>,<at 1 1>]}
       }}))
    ScaleA=true
    ScaleRGB=true
   )
  ]
  rendererClasses [&NiagaraSpriteRendererProperties]
  renderers [{rendererClass=&NiagaraSpriteRendererProperties}]
 }
 Tendrils001 {
  emitterSpawnScript [/SetVariables(
    Emitter.EmissiveSphereRadius:NiagaraFloat=&Multiply_Float(
     A=240
     B=&SelectFloatFromArray(Direct Array Index=$User.Stage, Float Selection Array=DataInterface() {propertyValues {FloatData [1,1,1.8,2.5]}})
    )
    Emitter.WPO_IntensityInput:NiagaraFloat=&Multiply_Float(
     A=40
     B=&SelectFloatFromArray(Direct Array Index=$User.Stage, Float Selection Array=DataInterface() {propertyValues {FloatData [1,1,1.2,2]}})
    )
   ) {isSetParameters=true}]
  emitterUpdateScript [
   &EmitterState(
    Inactive Response=::NewEnumerator1
    Life Cycle Mode=::NewEnumerator1
    Loop Behavior=::NewEnumerator1
    Loop Duration Mode=::NewEnumerator1
    UseLoopDelay=false
   )
   &SpawnRate(
    SpawnRate=&SelectFloatFromArray(Direct Array Index=$User.Stage, Float Selection Array=DataInterface() {propertyValues {FloatData [0,2,4,7]}})
    Use Spawn Probability=false
   )
   &SpawnRate(
    SpawnGroup=1
    SpawnRate=&SelectFloatFromArray(Direct Array Index=$User.Stage, Float Selection Array=DataInterface() {propertyValues {FloatData [0,0,0,4]}})
    Use Spawn Probability=false
   ) {module=SpawnRate001}
  ]
  particleSpawnScript [
   /SetVariables(
    Particles.IniRandom:NiagaraFloat=&RandomRangeFloat(Minimum=0.1, Maximum=1)
    Particles.NewSpawnGroup:NiagaraInt32=$Emitter.SpawnGroup
   ) {isSetParameters=true}
   &InitializeParticle(
    Color=<rgba 0.9792 1 0.9583 1>
    Lifetime Max=&SelectFloatFromArray(Direct Array Index=$User.Stage, Float Selection Array=DataInterface() {propertyValues {FloatData [6,6,3,1]}})
    Lifetime Min=4
    Lifetime Mode=::NewEnumerator1
    Mesh Scale=&MakeVector(X=$Particles.IniRandom, Y=$Particles.IniRandom, Z=&RandomRangeFloat(Minimum=0.5, Maximum=7))
    Mesh Scale Mode=::NewEnumerator4
    UsePositionOffset=false
    Write Material Random=true
   )
   &Color(Color=&MakeLinearColorFromVectorAndFloat("Vector (RGB)"=<1.25 1.25 1.25>, "Float (Alpha)"=&ClampFloat(Float=1)))
   /SetVariables(
    Particles.NewUpwardTendrilsOri:Quat4f=&MakeQuatFromAxisAngle(Axis=<0 0 1>, AngleInDegrees=&RandomRangeFloat(Minimum=0, Maximum=360))
    Particles.NewUpwardTenrilsSize:Vector3f=&MakeVector(
     X=&Multiply_Float(A=$Particles.IniRandom, B=1.5)
     Y=&Multiply_Float(A=$Particles.IniRandom, B=1.5)
     Z=&RandomRangeFloat(Minimum=12, Maximum=25)
    )
   ) {isSetParameters=true}
   &InitialMeshOrientation(
    Orientation Axis=<0 0 1>
    Rotation=&RandomRangeVector(Minimum=<-1 -1 -1>, Maximum=<1 1 1>)
    Use Orientation Vector=true
    Use Rotation Vector=true
   )
   /SetVariables(Particles.VisibilityTag=&SelectIntFromArray(Direct Array Index=$Particles.NewSpawnGroup, Int Selection Array=DataInterface() {propertyValues {IntData [0,1]}})) {isSetParameters=true}
  ]
  particleUpdateScript [
   &ParticleState()
   &ScaleColor(
    Scale Alpha=&FloatFromCurve(FloatCurve=DataInterface() {propertyValues {Curve {keys [
         <at 0 2 RCIM_Cubic>
         <at 0.0791 1>
         {arriveTangent=-1.0859, interpMode=RCIM_Cubic, leaveTangent=-1.0859, time=0.7833, value=0.8041}
         <at 1 0>
        ]}}})
    ScaleA=true
    ScaleRGB=true
   )
   &UpdateMeshOrientation(
    Rotation Rate=&RandomRangeFloat(Minimum=-0.05, Maximum=0.1)
    Rotation Vector=&ReturnMeshOrientationAxis(Mesh Orientation Axis=::ZAxis, Mesh Orientation Quat=$Particles.MeshOrientation)
   )
   /SetVariables(Particles.MeshRotation:Vector3f=&TransformVector(
     Vector=&RotateVectorByQuaternion(Vector To Rotate=<0 0 1>, Quaternion=$Particles.MeshOrientation)
     Destination Space=::World
    )) {isSetParameters=true}
   &DynamicMaterialParameters(
    Index 0 Param 1=&MakeFloatFromVector(Vector=$Particles.MeshRotation, Channel=::NewEnumerator0)
    Index 0 Param 2=&MakeFloatFromVector(Vector=$Particles.MeshRotation, Channel=::NewEnumerator1)
    Index 0 Param 3=&MakeFloatFromVector(Vector=$Particles.MeshRotation, Channel=::NewEnumerator2)
    Index 0 Param 4=&FloatFromCurve(FloatCurve=DataInterface() {propertyValues {Curve {keys [<at 0 1>,<at 1 -1>]}}}, Scale Curve=1.5)
    Index 1 Param 1=&FloatFromCurve(FloatCurve=DataInterface() {propertyValues {Curve {keys [<at 0 0>,<at 1 1>]}}})
    Index 1 Param 2=&SelectFloatFromArray(Direct Array Index=$User.Stage, Float Selection Array=DataInterface() {propertyValues {FloatData [1,1,1.25,1.7]}})
    Module.Param0WriteEnabled=true
    Module.Param0WriteEnabled1=true
    Module.Param1WriteEnabled=true
    Module.Param1WriteEnabled1=true
    Module.Param2WriteEnabled=true
    Module.Param2WriteEnabled1=true
    Module.Param3WriteEnabled=true
    Module.Param3WriteEnabled1=true
    Write Parameter Index 1=true
   )
   /SetVariables(Particles.NewUpwardTenrilsSize:Vector3f=&Multiply_Vector(
     A=$Particles.Initial.NewUpwardTenrilsSize
     B=&Multiply_VectorByFloat(
      Vector=&MakeVector(X=1, Y=1, Z=&Lerp_Float(A=0.4, B=2, Alpha=$System.OpenTheCore))
      Float=$User.GlobalScale
     )
    )) {isSetParameters=true}
   &ScaleMeshSize(Scale Factor=&Multiply_VectorByFloat(
     Vector=&MakeVector(X=1, Y=1, Z=&Lerp_Float(A=0.4, B=1.5, Alpha=$System.OpenTheCore))
     Float=&Multiply_Float(
      A=&Multiply_Float(A=$User.GlobalScale, B=&Lerp_Float(A=1, B=1.1, Alpha=$User.CorePlusingScalar))
      B=&SelectFloatFromArray(Direct Array Index=$User.Stage, Float Selection Array=DataInterface() {propertyValues {FloatData [1,1,1.5,2]}})
     )
    ))
  ]
  rendererClasses [&NiagaraMeshRendererProperties]
  renderers [
   {rendererClass=&NiagaraMeshRendererProperties, rendererIndex=0}
   {rendererClass=&NiagaraMeshRendererProperties, rendererIndex=1}
  ]
 }
 TinyFrags {
  emitterSpawnScript [/SetVariables(Emitter.Corruption=1, Emitter.Refraction=2, Emitter.Strength=1.1) {isSetParameters=true}]
  emitterUpdateScript [
   &EmitterState(
    Enable Distance Culling=true
    Life Cycle Mode=::NewEnumerator1
    Loop Duration=&RandomRangeFloat(Minimum=1, Maximum=2, Evaluation Type=::NewEnumerator1)
    MaxDistance=2500
    Module.bUseMaxDistance=true
    Module.bUseMinDistance=false
    Recalculate Duration Each Loop=true
    Scalability Mode=::NewEnumerator1
    UseLoopDelay=true
   )
   &SpawnRate(
    SpawnRate=&Multiply_Float(
     A=10
     B=&SelectFloatFromArray(Direct Array Index=$User.Stage, Float Selection Array=DataInterface() {propertyValues {FloatData [0,1,1.2,1.4]}})
    )
    Use Spawn Probability=false
   )
  ]
  particleSpawnScript [
   /SetVariables(Particles.MeshOrientation=&MakeQuatFromAxisAngle(
     Axis=&RandomRangeVector(Minimum=<0 0 0>, Maximum=<1 1 1>, Evaluation Type=::NewEnumerator1)
     AngleInDegrees=90
    )) {isSetParameters=true}
   /SetVariables(Particles.SpawnRotationVector=&RandomRangeVector(Minimum=<-0.15 -0.15 -0.15>, Maximum=<0.15 0.15 0.15>, Evaluation Type=::NewEnumerator1)) {isSetParameters=true}
   &InitializeParticle(
    Color Maximum=<rgba 0.25 0 0.138 1>
    Color Minimum=<rgba 0.25 0 0.0581 1>
    Color Mode=::NewEnumerator2
    Lifetime Max=3
    Lifetime Min=2
    Lifetime Mode=::NewEnumerator1
    Mesh Scale Max=<0.7 0.7 1>
    Mesh Scale Min=<0.2 0.2 1>
    Mesh Scale Mode=::NewEnumerator2
    UsePositionOffset=false
    Write Material Random=true
   )
   &AddVelocity(Velocity=&RandomRangeVector(Minimum=<-8 -8 -8>, Maximum=<8 8 8>, Evaluation Type=::NewEnumerator1))
   &SphereLocation(
    MaskSpawnGroup=false
    Override Seed=false
    Sphere Radius=&Multiply_Float(A=&RandomRangeFloat(Minimum=70, Maximum=120), B=$User.GlobalScale)
    Surface Only=false
    UseNonUniformScale=false
    UseOffset=false
    UseSphereOrientation=false
   ) {module=SphereLocation001}
   &AddVelocity(Velocity=&Multiply_VectorByFloat(
     Vector=$Output.SphereLocation.SphereVector
     Float=&RandomRangeFloat(Minimum=-0.1, Maximum=0.2, Evaluation Type=::NewEnumerator1)
    )) {module=AddVelocity001}
  ]
  particleUpdateScript [
   &ParticleState()
   &CurlNoiseForce(Pan Noise=false)
   &Drag(Drag=0.2, Use Linear Drag=true, Use Rotational Drag=false)
   &SolveForcesAndVelocity(Clamp Velocity=false, Limit Acceleration=false)
   &ScaleColor(
    Scale Alpha=&FloatFromCurve(
     FloatCurve=DataInterface() {propertyValues {Curve {keys [
         <at 0 0 RCIM_Cubic>
         {arriveTangent=1.4286, interpMode=RCIM_Cubic, leaveTangent=1.4286, time=0.2, value=1}
         {arriveTangent=-1.25, interpMode=RCIM_Cubic, leaveTangent=-1.25, time=0.7, value=1}
         <at 1 0>
        ]}}}
     Scale Curve=2
    )
    Scale RGB=&VectorFromFloat(Value=&RandomRangeFloat(Minimum=1, Maximum=1))
    ScaleA=true
    ScaleRGB=true
   )
   &UpdateMeshOrientation(
    Rotation Rate=&RandomRangeFloat(Minimum=-1.5, Maximum=1.5)
    Rotation Vector=$Particles.SpawnRotationVector
   ) {module=UpdateMeshOrientation001}
   &ScaleMeshSize(Scale Factor=&VectorFromFloat(Value=&FloatFromCurve(
      FloatCurve=DataInterface() {propertyValues {Curve {keys [<at 0 0 RCIM_Cubic>, <at 0.2 1>, {arriveTangent=-1.25, interpMode=RCIM_Cubic, leaveTangent=-1.25, time=0.8, value=1}, <at 1 0>]}}}
      Scale Curve=3
     )))
  ]
  rendererClasses [&NiagaraMeshRendererProperties]
  renderers [{rendererClass=&NiagaraMeshRendererProperties, rendererIndex=0}]
 }
}
summary {
 systemName=NS_Sideways_Encounter_Core
 emitters [
  {emitterName=Shimmers, rendererClasses [&NiagaraSpriteRendererProperties]}
  {emitterName=TinyFrags, rendererClasses [&NiagaraMeshRendererProperties]}
  {
   emitterName=Idle_FlareAndShard
   rendererClasses [
    &NiagaraSpriteRendererProperties
    &NiagaraMeshRendererProperties
    &NiagaraLightRendererProperties
    &NiagaraComponentRendererProperties
   ]
  }
  {emitterName=Embers_Dark, simTarget=GPUComputeSim, rendererClasses [&NiagaraSpriteRendererProperties]}
  {emitterName=Tendrils001, rendererClasses [&NiagaraMeshRendererProperties]}
  {emitterName=Embers_Hot001, simTarget=GPUComputeSim, rendererClasses [&NiagaraSpriteRendererProperties]}
  {emitterName=SoftWave, rendererClasses [&NiagaraMeshRendererProperties]}
  {emitterName=Spark, simTarget=GPUComputeSim, rendererClasses [&NiagaraSpriteRendererProperties]}
  {emitterName=Shockwave001, rendererClasses [&NiagaraMeshRendererProperties]}
  {emitterName=Spark001, simTarget=GPUComputeSim, rendererClasses [&NiagaraSpriteRendererProperties]}
 ]
 userVariables [
  User.CorePlusingScalar {type=NiagaraFloat, defaultValue=1, description=""}
  User.GlobalScale {type=NiagaraFloat, defaultValue=1, description=""}
  User.InWater {type=NiagaraBool}
  User.Open {type=NiagaraFloat, defaultValue=1, description=""}
  User.Stage {type=NiagaraInt32, defaultValue=1}
  User.TurnOnBurstEmitters {type=NiagaraBool}
  User.WaveLength {type=NiagaraFloat, defaultValue=0, description=""}
  User.WaveScale {type=NiagaraFloat, defaultValue=0, description=""}
 ]
}
