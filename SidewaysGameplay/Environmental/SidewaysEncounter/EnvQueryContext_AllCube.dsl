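; EnvQueryContext_AllCube is an Environment Query System context that supplies every Corruption gameplay Cube actor to queries, filed under the Sideways encounter gameplay.
; Its ProvideActorsSet override calls GetAllActorsOfClass on the BGA_Athena_CorruptionGameplay_Cube class and returns the whole array.
; The querier object and actor are ignored, so the result has no distance or ownership filtering.
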(fn ProvideActorsSet
  (params (QuerierObject :type Object%20Reference) (QuerierActor :type Actor%20Object%20Reference))
  (bind _outactors (Actor|GetAllActorsOfClass "/CorruptionGameplay/Gameplay/Environmental/Cube/BGA_Athena_CorruptionGameplay_Cube.BGA_Athena_CorruptionGameplay_Cube_C"))
  (return _outactors))
