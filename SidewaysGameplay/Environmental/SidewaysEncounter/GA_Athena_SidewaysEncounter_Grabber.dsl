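; Gameplay ability that makes the Sideways encounter Grabber repeatedly fire projectiles at pawns in its range.
; On activation it caches the Grabber avatar and an effect container spec for hits, then starts a looping fire timer.
; Each shot picks a random pawn, leading the aim by its previous velocity, or drops it if too far away.
; Encounter shutdown clears the timer; an empty pawn list or invalid avatar ends the ability.
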(event Ability|EventActivateAbility
  (bind _returnvalue (Ability|GetAvatarActorfromActorInfo))
  ;[1 1 1] Begin Play - Save out  a reference to the Grabber actor, make the container to give to the projectile and begin firing
  (Utilities|IsValid _returnvalue
    (:Is%20Valid
      (bind _asbga_athena_sideways_encounter_grabber (Utilities|Casting|CastToBGA_Athena_SidewaysEncounter_Grabber _returnvalue)
        (:then
          (Variables|Default|SetGrabberParent _asbga_athena_sideways_encounter_grabber)
          (Variables|Default|SetGECS_Projectile (Ability|MakeGameplayEffectContainerSpec "(TagName=\"Targeting.ApplicationTag.0\")"))
          (bind _returnvalue_2 (Utilities|Time|SetTimerbyEvent (delegate Custom|FireProjectile) (ScalableFloat|GetValueatLevel (Variables|Default|GetRow_FireInterval)) true))
          (Variables|Default|SetTimer_LaunchProjectile _returnvalue_2))
        (:CastFailed
          _)))
    (:Is%20Not%20Valid
      _))
  (Ability|EndAbility))

(event Custom|FireProjectile
  (bind _grabberparent (Variables|Default|GetGrabberParent))
  (bind _returnvalue (Transformation|GetActorLocation _grabberparent))
  (bind _grabberparent_2 (Variables|Default|GetGrabberParent))
  (bind _pawnsinrange (Variables|/SidewaysGameplay/Environmental/SidewaysEncounter/BGA_Athena_SidewaysEncounter_Grabber.BGA_Athena_SidewaysEncounter_Grabber_C|GetPawnsInRange _grabberparent_2))
  ;[1 1 1] Fire Projectile - fire the projectile at the player. If they are too far don't fire at them and remove them from the array. If the array is empty end the ability.
  (if (Variables|/SidewaysGameplay/Environmental/SidewaysEncounter/BGA_Athena_SidewaysEncounter_Grabber.BGA_Athena_SidewaysEncounter_Grabber_C|GetEncounterShutdown (Variables|Default|GetGrabberParent))
    (Utilities|Time|ClearandInvalidateTimerbyHandle (Variables|Default|GetTimer_LaunchProjectile))
    (else
      (if (> (Utilities|Array|Length _pawnsinrange) 0)
        (bind _output_get (Variables|Default|SetCurrentTarget (Utilities|Array|RandomArrayItem _pawnsinrange)))
        (if (>= (+ (Variables|/Script/Engine.SphereComponent|GetSphereRadius (Variables|/SidewaysGameplay/Environmental/SidewaysEncounter/BGA_Athena_SidewaysEncounter_Grabber.BGA_Athena_SidewaysEncounter_Grabber_C|GetGrabberRange _grabberparent)) 1e+02) (Math|Vector|Distance(Vector) (Transformation|GetActorLocation _output_get) _returnvalue))
          (Projectile|SpawnProjectile :ProjectileClass "/SidewaysGameplay/Environmental/SidewaysEncounter/Prj_Athena_SidewaysEncounter_Grabber.Prj_Athena_SidewaysEncounter_Grabber_C" :RequestedBy (Variables|Default|GetGrabberParent) :SpawnLocation (Math|Vector|vector+vector _returnvalue (Math|Vector|MakeVector 0.0 0.0 (Variables|Default|GetProjectileSpawnOffsetHeight))) :SpawnRotation (Math|Conversions|RotationFromXVector (Math|Vector|Normalize (- (Math|Vector|vector+vector (+ (Transformation|GetActorLocation _output_get) (* (FortMovement|GetPreviousVelocity _output_get) (ScalableFloat|GetValueatLevel (Variables|Default|GetRow_ProjectleAddedTargetVelocity)))) (Math|Vector|MakeVector 0.0 0.0 (Variables|Default|GetProjectileAimOffsetHeight))) (Math|Vector|vector+vector _returnvalue (Math|Vector|MakeVector 0.0 0.0 (Variables|Default|GetProjectileSpawnOffsetHeight)))))) :SpawnDirection (Math|Conversions|RotationFromXVector (Math|Vector|Normalize (- (Math|Vector|vector+vector (+ (Transformation|GetActorLocation _output_get) (* (FortMovement|GetPreviousVelocity _output_get) (ScalableFloat|GetValueatLevel (Variables|Default|GetRow_ProjectleAddedTargetVelocity)))) (Math|Vector|MakeVector 0.0 0.0 (Variables|Default|GetProjectileAimOffsetHeight))) (Math|Vector|vector+vector _returnvalue (Math|Vector|MakeVector 0.0 0.0 (Variables|Default|GetProjectileSpawnOffsetHeight)))))) :EffectContainerSpecToApplyOnHit (Variables|Default|GetGECS_Projectile) :bUseDefaultPhysics false :InitialSpeed (ScalableFloat|GetValueatLevel (Variables|Default|GetRow_ProjectileSpeed)) :GravityScale (ScalableFloat|GetValueatLevel (Variables|Default|GetRow_ProjectleGrav)) :ChargePercent 0.0)
          (else
            (Utilities|Array|RemoveItem (Variables|/SidewaysGameplay/Environmental/SidewaysEncounter/BGA_Athena_SidewaysEncounter_Grabber.BGA_Athena_SidewaysEncounter_Grabber_C|GetPawnsInRange (Variables|Default|GetGrabberParent)) _output_get)))
        (else
          (Ability|EndAbility))))))
