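# Environment query for the Sideways encounter that picks locations relative to players, using two options in fallback order.
# The primary option combines two donut generators for areas in front of players within the dome, filtered to a query-parameter-driven 3D distance band from any player.
# The secondary option uses two donuts as pie slices aimed at players, without the distance filter.
# Both options filter points with an upward geometry box trace requiring clear space.
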EnvQuery /SidewaysGameplay/Environmental/SidewaysEncounter/Lager_EQS_Donut_Params
options [
 {
  generator {
   name=EnvQueryGenerator_Composite_0
   class=EnvQueryGenerator_Composite
   properties {
    optionName=Primary - Areas in front of Players within Dome
    generators [
     &self:EnvQueryGenerator_Composite_0.EnvQueryGenerator_Donut_0
     &self:EnvQueryGenerator_Composite_0.EnvQueryGenerator_Donut_1
    ]
   }
  }
  tests [
   EnvQueryTest_Project_0 {
    class=EnvQueryTest_Project
    properties {
     boolValue=false
     scoringFactor=-1
     testComment=Require clear sweep up
     testPurpose=Filter
     projectionData {
      extentY=100
      extentZ=100
      navigationFilter=&RecastFilter_UseDefaultArea
      postProjectionVerticalOffset=0
      projectDown=0
      projectUp=500
      traceMode=GeometryByChannel
      traceShape=Box
     }
    }
   }
   EnvQueryTest_Distance_0 {
    class=EnvQueryTest_Distance
    properties {
     clampMaxType=SpecifiedValue
     clampMinType=SpecifiedValue
     distanceTo=/Script/FortniteAIServer.FortQueryContext_AllPlayers
     filterType=Range
     multipleContextFilterOp=AnyPass
     scoreClampMax=2000
     scoreClampMin=1000
     scoringEquation=Constant
     scoringFactor=1
     testComment=Not too close or too far from Players
     testMode=Distance3D
     testPurpose=Filter
     floatValueMax {
      dataBinding=&self:EnvQueryTest_Distance_0.AIDataProvider_QueryParams_1
      dataField=FloatValue
      defaultValue=1500
     }
     floatValueMin {
      dataBinding=&self:EnvQueryTest_Distance_0.AIDataProvider_QueryParams_0
      dataField=FloatValue
      defaultValue=700
     }
    }
   }
  ]
 }
 {
  generator {
   name=EnvQueryGenerator_Composite_1
   class=EnvQueryGenerator_Composite
   properties {
    optionName=Secondary - Pie slices aimed at Players
    generators [
     &self:EnvQueryGenerator_Composite_1.EnvQueryGenerator_Donut_0
     &self:EnvQueryGenerator_Composite_1.EnvQueryGenerator_Donut_1
    ]
   }
  }
  tests [EnvQueryTest_Project_5 {
    class=EnvQueryTest_Project
    properties {
     boolValue=false
     scoringFactor=-1
     testComment=Require clear sweep up
     testPurpose=Filter
     projectionData {
      extentY=100
      extentZ=100
      navigationFilter=&RecastFilter_UseDefaultArea
      postProjectionVerticalOffset=0
      projectDown=0
      projectUp=500
      traceMode=GeometryByChannel
      traceShape=Box
     }
    }
   }]
 }
]
