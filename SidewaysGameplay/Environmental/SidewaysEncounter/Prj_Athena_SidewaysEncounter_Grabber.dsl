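; Grabber projectile for the Sideways encounter that starts its owning encounter when it hits a qualifying player pawn.
; BeginPlay randomizes particle offset, sets lifespan from a scalable float, plays audio with a UI-ducking mix and sound indicator, then deactivates the effect after a delay.
; On hit, a gameplay tag check gates a once-only StartEncounter call on the owning BGA grabber, skipped if shut down.
; Vehicle and animal checks are mostly stubs.
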(event Collision|EventHit
  (params (MyComp :type Primitive%20Component%20Object%20Reference) (Other :type Actor%20Object%20Reference) (OtherComp :type Primitive%20Component%20Object%20Reference) (bSelfMoved :type Boolean) (HitLocation :type Vector :default "0, 0, 0") (HitNormal :type Vector :default "0, 0, 0") (NormalImpulse :type Vector :default "0, 0, 0") (Hit :type Hit%20Result%20Structure%20%28by%20ref%29))
  ;[1 1 1] Hit - Trigger the Encounter
  (bind _asfort_player_pawn (Utilities|Casting|CastToFortPlayerPawn Other)
    (:then
      (if (GameplayTags|HasMatchingGameplayTag _asfort_player_pawn (Variables|Default|GetT_OstTag))
        (bind _asfort_vehicle_interface (Utilities|Casting|CastToFortVehicleInterface Other)
          (:then
            (if (Vehicle|ContainsPlayers _asfort_vehicle_interface)
              (Development|PrintString "Hit")))
          (:CastFailed
            (if (GameplayTags|HasMatchingGameplayTag (Variables|Default|GetT_Animal))
              _)))
        (Utilities|FlowControl|DoOnce false
          (:Completed
            (bind _asbga_athena_sideways_encounter_grabber (Utilities|Casting|CastToBGA_Athena_SidewaysEncounter_Grabber (Actor|GetOwner)))
            (if (not (Variables|/SidewaysGameplay/Environmental/SidewaysEncounter/BGA_Athena_SidewaysEncounter_Grabber.BGA_Athena_SidewaysEncounter_Grabber_C|GetEncounterShutdown _asbga_athena_sideways_encounter_grabber))
              (Class|BGAAthenaSidewaysEncounterGrabber|StartEncounter _asbga_athena_sideways_encounter_grabber))))
        (else
          (if (Vehicle|IsinVehicle _asfort_player_pawn)
            _
            (else
              _)))))
    (:CastFailed
      _)))

(event EventBeginPlay
  (bind _returnvalue (Transformation|GetActorLocation))
  (bind _projectilefiresound (Variables|Default|GetProjectileFireSound))
  ;[1 1 1] Set the actors lifespawn from sheets
  (Audio|PushSoundMixModifier "/Game/Sounds/FortSoundClassesAndModes/Fort_Sideways_TendrilAttack_Duck_UI.Fort_Sideways_TendrilAttack_Duck_UI")
  (|CallParentFunction)
  (Effects|Components|ParticleSystem|SetVectorParameter (Variables|Default|GetNS_Sideways_Encounter_CoreGrabber) "RandomPosOffset" (Math|Vector|Lerp(Vector) "-300.000000,-300.000000,-100.000000" "300.000000,300.000000,300.000000" (Math|Random|RandomFloat)))
  (Actor|SetLifeSpan (ScalableFloat|GetValueatLevel (Variables|Default|GetRow_ProjectileLifespan)))
  (Audio|Components|Audio|FadeIn (Variables|Default|GetLoopingAudio) 0.2 1.0 0.0 "Linear" :!class /Script/Engine.AudioComponent)
  (Audio|PlaySoundatLocation _projectilefiresound _returnvalue)
  (SoundIndicator|ShowSoundIndicator (Variables|Default|GetFortSoundIndicator) _returnvalue _projectilefiresound 1.0 0)
  (Utilities|FlowControl|Delay 1.0)
  (Components|Activation|Deactivate (Variables|Default|GetNS_Sideways_Encounter_CoreGrabber)))

(fn ConstructionScript
  (|CallParentFunction))
